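# Control Rig function library of posing helpers that procedurally builds and drives FK controls on a humanoid skeleton.
# addFKCtrl spawns tagged controls for selected body regions and custom joints, mirroring left to right and parenting to the nearest existing region control.
# addNullObject creates reference nulls at initial bone transforms.
# setFKCtrl copies control transforms to bones and corrects limb twist bones using swing-twist against those nulls.
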ControlRigBlueprint /Game/Animation/Libraries/CRFL_PosingHelpers
graphs [
 New Function_ContainedGraph {nodes [
   n1 = RigVMFunctionEntryNode()
   > RigVMVariableNode(Value=@n1.Custom_FKCtrl) {variable=custom_item_ctrls}
   > RigVMVariableNode(Value=@n1.Options_FKCtrl) {variable=fkCtrl_bool_structure}
   > RigVMVariableNode(Value=@n1.Shape_Settings) {variable=fkCtrl_shapeSettings_structure}
   > RigVMFunction_Sequence = RigVMAggregateNode()
   RigVMFunction_Sequence_1 = RigVMAggregateNode(ExecuteContext=@RigVMFunction_Sequence.A)
   createFKCtrl = RigVMFunctionReferenceNode(
    FKChain=true
    JointNames=("root")
    ShapeSettings=(bVisible=True,Name="Triangle_Thick",Color=(R=0.3025,G=1,B=0.0685,A=1),Transform=(Rotation=(X=0,Y=0,Z=1,W=0),Translation=(X=0,Y=0,Z=0),Scale3D=(X=10,Y=10,Z=1)))
    ExecuteContext=@RigVMFunction_Sequence_1.A
   )
   RigVMFunction_ControlFlowBranch = &RigVMFunction_ControlFlowBranch(
    ExecuteContext=@RigVMFunction_Sequence_1.B
    Condition=RigVMVariableNode() {variable=fkCtrl_bool_structure}.Value.pelvis_1_#1
   )
   getCurrentParentCtrl_1 = RigVMFunctionReferenceNode(
    FKCtrl_bool_structure="(pelvis_1_#1=true,spine_3_#2=False,clavicle_5_#3=False,arm_7_#4=False,leg_9_#5=False,head_11_46DC15EF465CD ...(truncated)"
    ExecuteContext=@RigVMFunction_ControlFlowBranch.True
   )
   > createFKCtrl_1 = RigVMFunctionReferenceNode(
    FKChain=true
    JointNames=RigVMVariableNode() {variable=fkCtrl_boneNames_structure}.Value.pelvis_14_#6
    ShapeSettings=RigVMVariableNode() {variable=fkCtrl_shapeSettings_structure}.Value.pelvis_2_#7
    Parent=@getCurrentParentCtrl_1.currentParentCtrl.Name
   )
   n2 = &RigVMFunction_ControlFlowBranch(
    ExecuteContext=@RigVMFunction_Sequence_1.C
    Condition=RigVMVariableNode() {variable=fkCtrl_bool_structure}.Value.spine_3_#2
   )
   getCurrentParentCtrl_2 = RigVMFunctionReferenceNode(
    FKCtrl_bool_structure="(pelvis_1_#1=False,spine_3_#2=true,clavicle_5_#3=False,arm_7_#4=False,leg_9_#5=False,head_11_46DC15EF465CD ...(truncated)"
    ExecuteContext=@n2.True
   )
   n3 = RigVMVariableNode() {variable=fkCtrl_shapeSettings_structure}
   > createFKCtrl_0 = RigVMFunctionReferenceNode(
    FKChain=true
    JointNames=RigVMVariableNode() {variable=fkCtrl_boneNames_structure}.Value.spine_15_#8
    ShapeSettings=@n3.Value.spine_4_#9
    Parent=@getCurrentParentCtrl_2.currentParentCtrl.Name
   )
   > /Script/ControlRig.RigUnit_HierarchySetShapeSettings(Item=(Type=Control,Name="spine_05_ctrl"), Settings=@n3.Value.spine_4_#9) {is=struct RigVMUnitNode}
   Multiply = &RigVMFunction_MathDoubleMul(B=0.8, A=@n3.Value.spine_4_#9.Transform.Scale3D.X)
   > &RigUnit_HierarchySetShapeSettings(
    Item=(Type=Control,Name="spine_04_ctrl")
    Settings=(bVisible=True,Name="Box_Thick",Color=(R=1,G=1,B=0,A=1),Transform=(Rotation=(X=0,Y=-0.7071,Z=0,W=0.7071),Translation=(X=2,Y=2,Z=0),Scale3D=(X=3,Y=3,Z=2.5)))
    Settings.Transform.Scale3D.X=@Multiply.Result
    Settings.Transform.Scale3D.Y=@Multiply.Result
    Settings.Transform.Scale3D.Z=&RigVMFunction_MathDoubleMul(B=0.33, A=@n3.Value.spine_4_#9.Transform.Scale3D.Z).Result
    Settings.Transform.Translation.X=&RigVMFunction_MathDoubleMul(B=4, A=@n3.Value.spine_4_#9.Transform.Translation.X).Result
    Settings.Color=@n3.Value.spine_4_#9.Color
    Settings.Transform.Rotation=@n3.Value.spine_4_#9.Transform.Rotation
   )
   n4 = &RigVMFunction_ControlFlowBranch(
    ExecuteContext=@RigVMFunction_Sequence_1.D
    Condition=RigVMVariableNode() {variable=fkCtrl_bool_structure}.Value.clavicle_5_#3
   )
   getCurrentParentCtrl_3 = RigVMFunctionReferenceNode(
    FKCtrl_bool_structure="(pelvis_1_#1=False,spine_3_#2=False,clavicle_5_#3=true,arm_7_#4=False,leg_9_#5=False,head_11_46DC15EF465CD ...(truncated)"
    ExecuteContext=@n4.True
   )
   n5 = RigVMVariableNode() {variable=fkCtrl_boneNames_structure}
   n6 = RigVMVariableNode() {variable=fkCtrl_shapeSettings_structure}
   > createFKCtrl_2 = RigVMFunctionReferenceNode(
    FKChain=true
    JointNames=@n5.Value.clavicle_16_#10
    ShapeSettings=@n6.Value.clavicle_6_#11
    Parent=@getCurrentParentCtrl_3.currentParentCtrl.Name
   )
   > mirrorJointNames = RigVMFunctionReferenceNode(jointNames=@n5.Value.clavicle_16_#10)
   > mirrorShapeCtrls = RigVMFunctionReferenceNode(shapeControl=@n6.Value.clavicle_6_#11)
   Replace = /Script/RigVM.RigVMFunction_NameReplace(New=_r_, Old=_l_, Name=@getCurrentParentCtrl_3.currentParentCtrl.Name) {is=struct RigVMUnitNode, template="Replace::Execute(in Name,in Old,in New,out Result)"}
   > createFKCtrl_3 = RigVMFunctionReferenceNode(FKChain=true, Parent=@Replace.Result, JointNames=@mirrorJointNames.jointNames_mirror, ShapeSettings=@mirrorShapeCtrls.shapeControl_mirror)
   RigVMFunction_Sequence_2 = RigVMAggregateNode(ExecuteContext=@RigVMFunction_Sequence.B)
   n7 = &RigVMFunction_ControlFlowBranch(
    ExecuteContext=@RigVMFunction_Sequence_2.A
    Condition=RigVMVariableNode() {variable=fkCtrl_bool_structure}.Value.arm_7_#4
   )
   getCurrentParentCtrl = RigVMFunctionReferenceNode(
    FKCtrl_bool_structure="(pelvis_1_#1=False,spine_3_#2=False,clavicle_5_#3=False,arm_7_#4=true,leg_9_#5=False,head_11_46DC15EF465CD ...(truncated)"
    ExecuteContext=@n7.True
   )
   n8 = RigVMVariableNode() {variable=fkCtrl_boneNames_structure}
   n9 = RigVMVariableNode() {variable=fkCtrl_shapeSettings_structure}
   > createFKCtrl_4 = RigVMFunctionReferenceNode(
    FKChain=true
    JointNames=@n8.Value.arm_17_#12
    ShapeSettings=@n9.Value.arm_8_#13
    Parent=@getCurrentParentCtrl.currentParentCtrl.Name
   )
   > adjustShapeCtrls_5 = RigVMFunctionReferenceNode(scaleFactor=20, ctrlNames=@createFKCtrl_4.CtrlNames)
   > mirrorJointNames_1 = RigVMFunctionReferenceNode(jointNames=@n8.Value.arm_17_#12)
   > mirrorShapeCtrls_1 = RigVMFunctionReferenceNode(shapeControl=@n9.Value.arm_8_#13)
   > createFKCtrl_5 = RigVMFunctionReferenceNode(
    FKChain=true
    Parent=&RigVMFunction_NameReplace(New=_r_, Old=_l_, Name=@getCurrentParentCtrl.currentParentCtrl.Name) {template="Replace::Execute(in Name,in Old,in New,out Result)"}.Result
    JointNames=@mirrorJointNames_1.jointNames_mirror
    ShapeSettings=@mirrorShapeCtrls_1.shapeControl_mirror
   )
   > adjustShapeCtrls_1 = RigVMFunctionReferenceNode(scaleFactor=15, ctrlNames=@createFKCtrl_5.CtrlNames)
   n10 = &RigVMFunction_ControlFlowBranch(
    ExecuteContext=@RigVMFunction_Sequence_2.B
    Condition=RigVMVariableNode() {variable=fkCtrl_bool_structure}.Value.leg_9_#5
   )
   getCurrentParentCtrl_4 = RigVMFunctionReferenceNode(
    FKCtrl_bool_structure="(pelvis_1_#1=False,spine_3_#2=False,clavicle_5_#3=False,arm_7_#4=False,leg_9_#5=true,head_11_46DC15EF465CD ...(truncated)"
    ExecuteContext=@n10.True
   )
   n11 = RigVMVariableNode() {variable=fkCtrl_boneNames_structure}
   n12 = RigVMVariableNode() {variable=fkCtrl_shapeSettings_structure}
   > createFKCtrl_6 = RigVMFunctionReferenceNode(
    FKChain=true
    JointNames=@n11.Value.leg_18_#14
    ShapeSettings=@n12.Value.leg_10_#15
    Parent=@getCurrentParentCtrl_4.currentParentCtrl.Name
   )
   > adjustShapeCtrls_2 = RigVMFunctionReferenceNode(scaleFactor=15, ctrlNames=@createFKCtrl_6.CtrlNames)
   > mirrorJointNames_2 = RigVMFunctionReferenceNode(jointNames=@n11.Value.leg_18_#14)
   > mirrorShapeCtrls_2 = RigVMFunctionReferenceNode(shapeControl=@n12.Value.leg_10_#15)
   > createFKCtrl_7 = RigVMFunctionReferenceNode(
    FKChain=true
    Parent=&RigVMFunction_NameReplace(New=_r_, Old=_l_, Name=@getCurrentParentCtrl_4.currentParentCtrl.Name) {template="Replace::Execute(in Name,in Old,in New,out Result)"}.Result
    JointNames=@mirrorJointNames_2.jointNames_mirror
    ShapeSettings=@mirrorShapeCtrls_2.shapeControl_mirror
   )
   > adjustShapeCtrls_3 = RigVMFunctionReferenceNode(scaleFactor=10, ctrlNames=@createFKCtrl_7.CtrlNames)
   n13 = &RigVMFunction_ControlFlowBranch(
    ExecuteContext=@RigVMFunction_Sequence_2.C
    Condition=RigVMVariableNode() {variable=fkCtrl_bool_structure}.Value.head_11_#16
   )
   getCurrentParentCtrl_5 = RigVMFunctionReferenceNode(
    FKCtrl_bool_structure="(pelvis_1_#1=False,spine_3_#2=False,clavicle_5_#3=False,arm_7_#4=False,leg_9_#5=False,head_11_46DC15EF465C ...(truncated)"
    ExecuteContext=@n13.True
   )
   > createFKCtrl_8 = RigVMFunctionReferenceNode(
    FKChain=true
    JointNames=RigVMVariableNode() {variable=fkCtrl_boneNames_structure}.Value.head_19_#17
    ShapeSettings=RigVMVariableNode() {variable=fkCtrl_shapeSettings_structure}.Value.head_12_#18
    Parent=@getCurrentParentCtrl_5.currentParentCtrl.Name
   )
   n14 = RigVMVariableNode() {variable=fkCtrl_shapeSettings_structure}
   n15 = &RigVMFunction_MathDoubleMul(B=0.7, A=@n14.Value.head_12_#18.Transform.Scale3D.X)
   > &RigUnit_HierarchySetShapeSettings(
    Item=(Type=Control,Name="head_ctrl")
    Settings=(bVisible=True,Name="Box_Thick",Color=(R=1,G=1,B=0,A=1),Transform=(Rotation=(X=0,Y=-0.7071,Z=0,W=0.7071),Translation=(X=2,Y=2,Z=0),Scale3D=(X=3,Y=3,Z=2.5)))
    Settings.Transform.Scale3D.X=@n15.Result
    Settings.Transform.Scale3D.Y=@n15.Result
    Settings.Transform.Scale3D.Z=&RigVMFunction_MathDoubleMul(B=0.25, A=@n14.Value.head_12_#18.Transform.Scale3D.Z).Result
    Settings.Transform.Translation.X=&RigVMFunction_MathDoubleMul(B=1.5, A=@n14.Value.head_12_#18.Transform.Translation.X).Result
    Settings.Color=@n14.Value.head_12_#18.Color
    Settings.Transform.Rotation=@n14.Value.head_12_#18.Transform.Rotation
   )
   n16 = &RigVMFunction_Sequence(ExecuteContext=@RigVMFunction_Sequence.C)
   For_Each = RigVMDispatchNode(ExecuteContext=@n16.A, Array=RigVMVariableNode() {variable=custom_item_ctrls}.Value) {template="DISPATCH_RigVMDispatch_ArrayIterator(in Array,out Element,out Index,out Count,out Ratio)"}
   > RigVMVariableNode(Value=@For_Each.Element.parent_2_#19) {variable=custom_parent_ctrl}
   > RigVMVariableNode(Value=@For_Each.Element.mirror_8_#20) {variable=custom_mirror}
   > RigVMVariableNode(Value=@For_Each.Element.FKChain_16_#21) {variable=custom_fkChain}
   > RigVMVariableNode(Value=@For_Each.Element.factor_scale_21_#22) {variable=custom_factor_scale}
   > RigVMVariableNode(Value=@For_Each.Element.shape_settings_24_#23) {variable=custom_Shape_Settings}
   > For_Each_1 = RigVMDispatchNode(Array=@For_Each.Element.joint_names_14_#24) {template="DISPATCH_RigVMDispatch_ArrayIterator(in Array,out Element,out Index,out Count,out Ratio)"}
   > Add = RigVMDispatchNode(Element=@For_Each_1.Element.Name, Array=RigVMVariableNode() {variable=custom_names_joint}.Value) {template="DISPATCH_RigVMDispatch_ArrayAdd(io Array,in Element,out Index)"}
   createFKCtrl_9 = RigVMFunctionReferenceNode(
    ExecuteContext=@For_Each_1.Completed
    ShapeSettings=RigVMVariableNode() {variable=custom_Shape_Settings}.Value
    JointNames=RigVMVariableNode() {variable=custom_names_joint}.Value
    Parent=RigVMVariableNode() {variable=custom_parent_ctrl}.Value.Name
    FKChain=RigVMVariableNode() {variable=custom_fkChain}.Value
   )
   > adjustShapeCtrls_0 = RigVMFunctionReferenceNode(scaleFactor=RigVMVariableNode() {variable=custom_factor_scale}.Value, ctrlNames=@createFKCtrl_9.CtrlNames)
   n17 = &RigVMFunction_ControlFlowBranch(Condition=RigVMVariableNode() {variable=custom_mirror}.Value, ExecuteContext=@adjustShapeCtrls_0.ExecuteContext)
   RigVMDispatchNode(ExecuteContext=@n17.Completed, Array=RigVMVariableNode() {variable=custom_names_joint}.Value) {template="DISPATCH_RigVMDispatch_ArrayReset(io Array)"}
   mirrorJointNames_0 = RigVMFunctionReferenceNode(ExecuteContext=@n17.True, jointNames=RigVMVariableNode() {variable=custom_names_joint}.Value)
   > createFKCtrl_10 = RigVMFunctionReferenceNode(
    Parent=RigVMVariableNode() {variable=custom_parent_ctrl}.Value.Name
    ShapeSettings=RigVMVariableNode() {variable=custom_Shape_Settings}.Value
    FKChain=RigVMVariableNode() {variable=custom_fkChain}.Value
    JointNames=@mirrorJointNames_0.jointNames_mirror
   )
   > adjustShapeCtrls_4 = RigVMFunctionReferenceNode(scaleFactor=RigVMVariableNode() {variable=custom_factor_scale}.Value, ctrlNames=@createFKCtrl_10.CtrlNames)
   RigVMFunctionReturnNode(ExecuteContext=@RigVMFunction_Sequence.D)
  ]}
 New Function_ContainedGraph {nodes [
   n1 = RigVMFunctionEntryNode()
   RigVMFunction_ControlFlowBranch = &RigVMFunction_ControlFlowBranch(ExecuteContext=@n1.ExecuteContext, Condition=@n1.setFK_Ctrl)
   RigVMFunctionReturnNode(ExecuteContext=@RigVMFunction_ControlFlowBranch.Completed)
   RigVMFunction_Sequence = RigVMAggregateNode(ExecuteContext=@RigVMFunction_ControlFlowBranch.True)
   n2 = &RigVMFunction_Sequence(ExecuteContext=@RigVMFunction_Sequence.B)
   fixTwistRotation = RigVMFunctionReferenceNode(joint_name=upperarm_l, joint_to_fix=upperarm_twist_01_l, multiplier=0, reverse=true, ExecuteContext=@n2.A)
   > fixTwistRotation_1 = RigVMFunctionReferenceNode(joint_name=upperarm_l, joint_to_fix=upperarm_twist_02_l, multiplier=0.5, reverse=true)
   > fixTwistRotation_4 = RigVMFunctionReferenceNode(joint_name=hand_l, joint_to_fix=lowerarm_twist_01_l, multiplier=0)
   > fixTwistRotation_5 = RigVMFunctionReferenceNode(joint_name=hand_l, joint_to_fix=lowerarm_twist_02_l, multiplier=0.5)
   > fixTwistRotation_2 = RigVMFunctionReferenceNode(joint_name=upperarm_r, joint_to_fix=upperarm_twist_01_r, multiplier=0, reverse=true)
   > fixTwistRotation_3 = RigVMFunctionReferenceNode(joint_name=upperarm_r, joint_to_fix=upperarm_twist_02_r, multiplier=0.5, reverse=true)
   > fixTwistRotation_6 = RigVMFunctionReferenceNode(joint_name=hand_r, joint_to_fix=lowerarm_twist_01_r, multiplier=0)
   > fixTwistRotation_7 = RigVMFunctionReferenceNode(joint_name=hand_r, joint_to_fix=lowerarm_twist_02_r, multiplier=0.5)
   fixTwistRotation_8 = RigVMFunctionReferenceNode(joint_name=thigh_l, joint_to_fix=thigh_twist_01_l, multiplier=0, reverse=true, ExecuteContext=@n2.B)
   > fixTwistRotation_9 = RigVMFunctionReferenceNode(joint_name=foot_l, joint_to_fix=calf_twist_01_l, multiplier=0, reverse=false)
   > fixTwistRotation_10 = RigVMFunctionReferenceNode(joint_name=foot_l, joint_to_fix=calf_twist_02_l, multiplier=0.5, reverse=false)
   > fixTwistRotation_11 = RigVMFunctionReferenceNode(joint_name=thigh_r, joint_to_fix=thigh_twist_01_r, multiplier=0, reverse=true)
   > fixTwistRotation_12 = RigVMFunctionReferenceNode(joint_name=foot_r, joint_to_fix=calf_twist_01_r, multiplier=0, reverse=false)
   > fixTwistRotation_13 = RigVMFunctionReferenceNode(joint_name=foot_r, joint_to_fix=calf_twist_02_r, multiplier=0.5, reverse=false)
   RigVMVariableNode(
    Value=/Script/ControlRig.RigUnit_FindItemsWithMetadataTag(NameSpace=Self, Tag=BodyCtrl) {is=struct RigVMUnitNode}.Items
    ExecuteContext=@RigVMFunction_Sequence.A
   ) {variable=ctrl_Items}
   > For_Each = RigVMDispatchNode(Array=RigVMVariableNode() {variable=ctrl_Items}.Value) {template="DISPATCH_RigVMDispatch_ArrayIterator(in Array,out Element,out Index,out Count,out Ratio)"}
   n3 = &RigVMFunction_ControlFlowBranch(
    ExecuteContext=@For_Each.ExecuteContext
    Condition=/Script/ControlRig.RigUnit_ItemExists(Item=@For_Each.Element) {is=struct RigVMUnitNode}.Exists
   )
   Chop = /Script/RigVM.RigVMFunction_NameTruncate(Count=5, FromEnd=True, Name=@For_Each.Element.Name) {is=struct RigVMUnitNode, template="Chop::Execute(in Name,in Count,in FromEnd,out Remainder,out Chopped)"}
   &RigUnit_SetTransform(
    Item=(Type=Bone,Name="None")
    Weight=1
    bInitial=False
    Item.Name=@Chop.Remainder
    Value=&RigUnit_GetTransform(bInitial=False, Item=@For_Each.Element).Transform
    ExecutePin=@n3.True
   )
  ]}
 New Function_ContainedGraph {nodes [
   n1 = RigVMFunctionEntryNode()
   > RigVMVariableNode(Value=@n1.FKChain) {variable=FKChain_ctrl}
   > RigVMVariableNode(Value=@n1.Parent) {variable=parent_ctrl}
   > RigVMVariableNode(Value=@n1.JointNames) {variable=joint_names}
   > n2 = RigVMVariableNode(Value=@n1.ShapeSettings) {variable=ShapeCtrl_Settings}
   RigVMFunction_Sequence = &RigVMFunction_Sequence(ExecuteContext=@n2.ExecuteContext)
   For_Each = RigVMDispatchNode(ExecuteContext=@RigVMFunction_Sequence.A, Array=RigVMVariableNode() {variable=joint_names}.Value) {template="DISPATCH_RigVMDispatch_ArrayIterator(in Array,out Element,out Index,out Count,out Ratio)"}
   > RigVMVariableNode(Value=(Type=Bone,Name="None"), Value.Name=@For_Each.Element) {variable=temp_Bone}
   Concat = /Script/RigVM.RigVMFunction_NameConcat(B=_ctrl, A=@For_Each.Element) {is=struct RigVMUnitNode, template="Concat::Execute(in A,in B,out Result)"}
   > n3 = RigVMVariableNode(Value=@Concat.Result) {variable=currentName_ctrl}
   Greater = /Script/RigVM.RigVMFunction_MathIntGreater(B=0, A=@For_Each.Index) {is=struct RigVMUnitNode, template="Greater::Execute(in A,in B,out Result)"}
   RigVMFunction_ControlFlowBranch = &RigVMFunction_ControlFlowBranch(Condition=@Greater.Result, ExecuteContext=@n3.ExecuteContext)
   n4 = /Script/ControlRig.RigUnit_HierarchyAddControlTransform(
    InitialValue=(Rotation=(X=0,Y=0,Z=0,W=1),Translation=(X=0,Y=0,Z=0),Scale3D=(X=1,Y=1,Z=1))
    OffsetSpace=GlobalSpace
    Settings="(InitialSpace=LocalSpace,bUsePreferredRotationOrder=False,PreferredRotationOrder=YZX,Limits=(LimitTranslationX=(bMinimum=False,bMaximum=False),LimitTranslationY=(bMinimum=False,bMaximum=False),LimitTranslationZ=(bMinimum=False,bMaximum=False),LimitPitch=(b ...(truncated)"
    OffsetTransform=&RigUnit_GetTransform(bInitial=False, Item=RigVMVariableNode() {variable=temp_Bone}.Value).Transform
    ExecutePin=@RigVMFunction_ControlFlowBranch.Completed
    Parent=RigVMVariableNode() {variable=currentParent_ctrl}.Value
    Name=RigVMVariableNode() {variable=currentName_ctrl}.Value
    Settings.Shape=RigVMVariableNode() {variable=ShapeCtrl_Settings}.Value
   ) {
    is=struct RigVMUnitNode
    template="SpawnControl::Execute(in InitialValue,in Settings,in OffsetTransform,in OffsetSpace,in Parent,in Name,out Item)"
   }
   > /Script/ControlRig.RigUnit_SetMetadataTag(NameSpace=Self, Tag=BodyCtrl, Item=@n4.Item) {is=struct RigVMUnitNode}
   n5 = RigVMVariableNode() {variable=currentName_ctrl}
   > RigVMVariableNode(Value=(Type=Control,Name="None"), Value.Name=@n5.Value) {variable=temp_Ctrl}
   > Add = RigVMDispatchNode(Array=RigVMVariableNode() {variable=ctrl_names}.Value, Element=@n5.Value) {template="DISPATCH_RigVMDispatch_ArrayAdd(io Array,in Element,out Index)"}
   RigVMVariableNode(
    Value=(Type=Control,Name="None")
    ExecuteContext=@RigVMFunction_ControlFlowBranch.False
    Value.Name=RigVMVariableNode() {variable=parent_ctrl}.Value
   ) {variable=currentParent_ctrl}
   n6 = &RigVMFunction_ControlFlowBranch(ExecuteContext=@RigVMFunction_ControlFlowBranch.True, Condition=RigVMVariableNode() {variable=FKChain_ctrl}.Value)
   RigVMVariableNode(Value=(Type=Control,Name="None"), ExecuteContext=@n6.False, Value.Name=RigVMVariableNode() {variable=parent_ctrl}.Value) {variable=currentParent_ctrl}
   RigVMVariableNode(ExecuteContext=@n6.True, Value=RigVMVariableNode() {variable=temp_Ctrl}.Value) {variable=currentParent_ctrl}
   RigVMFunctionReturnNode(ExecuteContext=@RigVMFunction_Sequence.B, CtrlNames=RigVMVariableNode() {variable=ctrl_names}.Value)
  ]}
 New Function_ContainedGraph {nodes [
   n1 = RigVMFunctionEntryNode()
   > For_Each = RigVMDispatchNode(Array=@n1.jointNames) {template="DISPATCH_RigVMDispatch_ArrayIterator(in Array,out Element,out Index,out Count,out Ratio)"}
   Replace = &RigVMFunction_NameReplace(New=_r, Old=_l, Name=@For_Each.Element) {template="Replace::Execute(in Name,in Old,in New,out Result)"}
   > Add = RigVMDispatchNode(Element=@Replace.Result, Array=RigVMVariableNode() {variable=jointNames_input}.Value) {template="DISPATCH_RigVMDispatch_ArrayAdd(io Array,in Element,out Index)"}
   RigVMFunctionReturnNode(ExecuteContext=@For_Each.Completed, jointNames_mirror=RigVMVariableNode() {variable=jointNames_input}.Value)
  ]}
 New Function_ContainedGraph {nodes [
   n1 = RigVMFunctionEntryNode()
   Scale = /Script/RigVM.RigVMFunction_MathVectorScale(Factor=-1, Value=@n1.shapeControl.Transform.Translation) {is=struct RigVMUnitNode, template="Scale::Execute(in Value,in Factor,out Result)"}
   > RigVMFunctionReturnNode(
    shapeControl_mirror=(bVisible=True,Name="Default",Color=(R=1,G=0,B=0,A=1),Transform=(Rotation=(X=0,Y=0,Z=0,W=1),Translation=(X=0,Y=0,Z=0),Scale3D=(X=1,Y=1,Z=1)))
    shapeControl_mirror.Name=@n1.shapeControl.Name
    shapeControl_mirror.Transform.Scale3D=@n1.shapeControl.Transform.Scale3D
    shapeControl_mirror.Color=RigVMDispatchNode(Elements=(R=1,G=0,B=0,A=1)) {template="DISPATCH_RigVMDispatch_MakeStruct(in Elements,out Struct)"}.Struct
    shapeControl_mirror.Transform.Translation=@Scale.Result
    shapeControl_mirror.Transform.Rotation=/Script/RigVM.RigVMFunction_MathQuaternionScaleV2(Factor=-1, Value=@n1.shapeControl.Transform.Rotation) {is=struct RigVMUnitNode, template="Scale::Execute(in Value,in Factor,out Result)"}.Result
   )
  ]}
 New Function_ContainedGraph {nodes [
   n1 = RigVMFunctionEntryNode()
   > RigVMDispatchNode(Enabled=True, ScreenColor=(R=1,G=1,B=1,A=1), ScreenDuration=0.05, Value=@n1.ctrlNames) {template="DISPATCH_RigVMDispatch_Print(in Prefix,in Value,in Enabled,in ScreenDuration,in ScreenColor)"}
   > For_Each = RigVMDispatchNode(Array=@n1.ctrlNames) {template="DISPATCH_RigVMDispatch_ArrayIterator(in Array,out Element,out Index,out Count,out Ratio)"}
   n2 = /Script/ControlRig.RigUnit_GetShapeTransform(Control=@For_Each.Element) {is=struct RigVMUnitNode}
   RigVMFunction_MathIntToDouble = /Script/RigVM.RigVMFunction_MathIntToFloat(Value=@For_Each.Index) {is=struct RigVMUnitNode, template="Cast::Execute(in Value,out Result)"}
   Add = &RigVMFunction_MathDoubleAdd(B=1, A=@RigVMFunction_MathIntToDouble.Result)
   Multiply = &RigVMFunction_MathDoubleMul(A=@Add.Result, B=@n1.scaleFactor)
   Divide = &RigVMFunction_MathDoubleDiv(B=100, A=@Multiply.Result)
   Subtract = &RigVMFunction_MathDoubleSub(A=1, B=@Divide.Result)
   > /Script/ControlRig.RigUnit_SetShapeTransform(
    Transform=(Rotation=(X=0,Y=0,Z=0,W=1),Translation=(X=0,Y=0,Z=0),Scale3D=(X=1,Y=1,Z=1))
    Control=@For_Each.Element
    Transform.Rotation=@n2.Transform.Rotation
    Transform.Scale3D.Z=@n2.Transform.Scale3D.Z
    Transform.Translation=@n2.Transform.Translation
    Transform.Scale3D.X=&RigVMFunction_MathDoubleMul(A=@n2.Transform.Scale3D.X, B=@Subtract.Result).Result
    Transform.Scale3D.Y=&RigVMFunction_MathDoubleMul(A=@n2.Transform.Scale3D.Y, B=@Subtract.Result).Result
   ) {is=struct RigVMUnitNode}
   RigVMFunctionReturnNode(ExecuteContext=@For_Each.Completed)
  ]}
 New Function_ContainedGraph {nodes [
   n1 = RigVMFunctionEntryNode()
   > RigVMVariableNode(Value=@n1.FKCtrl_bool_structure) {variable=input_fKCtrl_bool_structure}
   > RigVMFunction_Sequence = RigVMAggregateNode()
   RigVMFunction_Sequence_1 = RigVMAggregateNode(ExecuteContext=@RigVMFunction_Sequence.A)
   n2 = &RigVMFunction_ControlFlowBranch(
    ExecuteContext=@RigVMFunction_Sequence_1.A
    Condition=RigVMVariableNode() {variable=input_fKCtrl_bool_structure}.Value.pelvis_1_#1
   )
   RigVMVariableNode(Value=(Type=Control,Name="root_ctrl"), ExecuteContext=@n2.True) {variable=output_currentParentCtrl}
   n3 = &RigVMFunction_ControlFlowBranch(
    ExecuteContext=@RigVMFunction_Sequence_1.B
    Condition=RigVMVariableNode() {variable=input_fKCtrl_bool_structure}.Value.spine_3_#2
   )
   n4 = &RigVMFunction_ControlFlowBranch(Condition=&RigUnit_ItemExists(Item=(Type=Control,Name="pelvis_ctrl")).Exists, ExecuteContext=@n3.True)
   RigVMVariableNode(Value=(Type=Control,Name="root_ctrl"), ExecuteContext=@n4.False) {variable=output_currentParentCtrl}
   RigVMVariableNode(Value=(Type=Control,Name="pelvis_ctrl"), ExecuteContext=@n4.True) {variable=output_currentParentCtrl}
   n5 = &RigVMFunction_ControlFlowBranch(
    ExecuteContext=@RigVMFunction_Sequence_1.C
    Condition=RigVMVariableNode() {variable=input_fKCtrl_bool_structure}.Value.clavicle_5_#3
   )
   n6 = &RigVMFunction_ControlFlowBranch(Condition=&RigUnit_ItemExists(Item=(Type=Control,Name="spine_05_ctrl")).Exists, ExecuteContext=@n5.True)
   n7 = &RigVMFunction_ControlFlowBranch(Condition=&RigUnit_ItemExists(Item=(Type=Control,Name="pelvis_ctrl")).Exists, ExecuteContext=@n6.False)
   RigVMVariableNode(Value=(Type=Control,Name="root_ctrl"), ExecuteContext=@n7.False) {variable=output_currentParentCtrl}
   RigVMVariableNode(Value=(Type=Control,Name="pelvis_ctrl"), ExecuteContext=@n7.True) {variable=output_currentParentCtrl}
   RigVMVariableNode(Value=(Type=Control,Name="spine_05_ctrl"), ExecuteContext=@n6.True) {variable=output_currentParentCtrl}
   RigVMFunction_Sequence_2 = RigVMAggregateNode(ExecuteContext=@RigVMFunction_Sequence.B)
   n8 = &RigVMFunction_ControlFlowBranch(
    ExecuteContext=@RigVMFunction_Sequence_2.A
    Condition=RigVMVariableNode() {variable=input_fKCtrl_bool_structure}.Value.arm_7_#4
   )
   n9 = &RigVMFunction_ControlFlowBranch(Condition=&RigUnit_ItemExists(Item=(Type=Control,Name="clavicle_l_ctrl")).Exists, ExecuteContext=@n8.True)
   n10 = &RigVMFunction_ControlFlowBranch(Condition=&RigUnit_ItemExists(Item=(Type=Control,Name="spine_05_ctrl")).Exists, ExecuteContext=@n9.False)
   n11 = &RigVMFunction_ControlFlowBranch(Condition=&RigUnit_ItemExists(Item=(Type=Control,Name="pelvis_ctrl")).Exists, ExecuteContext=@n10.False)
   RigVMVariableNode(Value=(Type=Control,Name="root_ctrl"), ExecuteContext=@n11.False) {variable=output_currentParentCtrl}
   RigVMVariableNode(Value=(Type=Control,Name="pelvis_ctrl"), ExecuteContext=@n11.True) {variable=output_currentParentCtrl}
   RigVMVariableNode(Value=(Type=Control,Name="spine_05_ctrl"), ExecuteContext=@n10.True) {variable=output_currentParentCtrl}
   RigVMVariableNode(Value=(Type=Control,Name="clavicle_l_ctrl"), ExecuteContext=@n9.True) {variable=output_currentParentCtrl}
   n12 = &RigVMFunction_ControlFlowBranch(
    ExecuteContext=@RigVMFunction_Sequence_2.B
    Condition=RigVMVariableNode() {variable=input_fKCtrl_bool_structure}.Value.leg_9_#5
   )
   n13 = &RigVMFunction_ControlFlowBranch(Condition=&RigUnit_ItemExists(Item=(Type=Control,Name="pelvis_ctrl")).Exists, ExecuteContext=@n12.True)
   RigVMVariableNode(Value=(Type=Control,Name="root_ctrl"), ExecuteContext=@n13.False) {variable=output_currentParentCtrl}
   RigVMVariableNode(Value=(Type=Control,Name="pelvis_ctrl"), ExecuteContext=@n13.True) {variable=output_currentParentCtrl}
   n14 = &RigVMFunction_ControlFlowBranch(
    ExecuteContext=@RigVMFunction_Sequence_2.C
    Condition=RigVMVariableNode() {variable=input_fKCtrl_bool_structure}.Value.head_11_#16
   )
   n15 = &RigVMFunction_ControlFlowBranch(Condition=&RigUnit_ItemExists(Item=(Type=Control,Name="spine_05_ctrl")).Exists, ExecuteContext=@n14.True)
   n16 = &RigVMFunction_ControlFlowBranch(Condition=&RigUnit_ItemExists(Item=(Type=Control,Name="pelvis_ctrl")).Exists, ExecuteContext=@n15.False)
   RigVMVariableNode(Value=(Type=Control,Name="root_ctrl"), ExecuteContext=@n16.False) {variable=output_currentParentCtrl}
   RigVMVariableNode(Value=(Type=Control,Name="pelvis_ctrl"), ExecuteContext=@n16.True) {variable=output_currentParentCtrl}
   RigVMVariableNode(Value=(Type=Control,Name="spine_05_ctrl"), ExecuteContext=@n15.True) {variable=output_currentParentCtrl}
   RigVMFunctionReturnNode(ExecuteContext=@RigVMFunction_Sequence.F, currentParentCtrl=RigVMVariableNode() {variable=output_currentParentCtrl}.Value)
  ]}
 New Function_ContainedGraph {nodes [
   n1 = RigVMFunctionEntryNode()
   > RigVMVariableNode(Value=@n1.Options_Null) {variable=bool_null_structure}
   > RigVMVariableNode(Value=@n1.Custom_Joints) {variable=Current_JointItems}
   > RigVMFunction_Sequence = RigVMAggregateNode()
   RigVMFunction_Sequence_1 = RigVMAggregateNode(ExecuteContext=@RigVMFunction_Sequence.A)
   RigVMFunction_ControlFlowBranch = &RigVMFunction_ControlFlowBranch(
    ExecuteContext=@RigVMFunction_Sequence_1.A
    Condition=RigVMVariableNode() {variable=bool_null_structure}.Value.pelvis_1_#1
   )
   createNullObjects = RigVMFunctionReferenceNode(
    ExecuteContext=@RigVMFunction_ControlFlowBranch.True
    joint_names=RigVMVariableNode() {variable=joint_names_structure}.Value.pelvis_14_#6
   )
   n2 = &RigVMFunction_ControlFlowBranch(
    ExecuteContext=@RigVMFunction_Sequence_1.B
    Condition=RigVMVariableNode() {variable=bool_null_structure}.Value.spine_3_#2
   )
   createNullObjects_1 = RigVMFunctionReferenceNode(ExecuteContext=@n2.True, joint_names=RigVMVariableNode() {variable=joint_names_structure}.Value.spine_15_#8)
   n3 = &RigVMFunction_ControlFlowBranch(
    ExecuteContext=@RigVMFunction_Sequence_1.C
    Condition=RigVMVariableNode() {variable=bool_null_structure}.Value.clavicle_5_#3
   )
   n4 = RigVMVariableNode() {variable=joint_names_structure}
   createNullObjects_2 = RigVMFunctionReferenceNode(ExecuteContext=@n3.True, joint_names=@n4.Value.clavicle_16_#10)
   > mirrorJointNames_0 = RigVMFunctionReferenceNode(jointNames=@n4.Value.clavicle_16_#10)
   > createNullObjects_3 = RigVMFunctionReferenceNode(joint_names=@mirrorJointNames_0.jointNames_mirror)
   RigVMFunction_Sequence_2 = RigVMAggregateNode(ExecuteContext=@RigVMFunction_Sequence.B)
   n5 = &RigVMFunction_ControlFlowBranch(
    ExecuteContext=@RigVMFunction_Sequence_2.A
    Condition=RigVMVariableNode() {variable=bool_null_structure}.Value.arm_7_#4
   )
   n6 = RigVMVariableNode() {variable=joint_names_structure}
   createNullObjects_5 = RigVMFunctionReferenceNode(ExecuteContext=@n5.True, joint_names=@n6.Value.arm_17_#12)
   > mirrorJointNames_1 = RigVMFunctionReferenceNode(jointNames=@n6.Value.arm_17_#12)
   > createNullObjects_4 = RigVMFunctionReferenceNode(joint_names=@mirrorJointNames_1.jointNames_mirror)
   n7 = &RigVMFunction_ControlFlowBranch(
    ExecuteContext=@RigVMFunction_Sequence_2.B
    Condition=RigVMVariableNode() {variable=bool_null_structure}.Value.leg_9_#5
   )
   n8 = RigVMVariableNode() {variable=joint_names_structure}
   createNullObjects_7 = RigVMFunctionReferenceNode(ExecuteContext=@n7.True, joint_names=@n8.Value.leg_18_#14)
   > mirrorJointNames_2 = RigVMFunctionReferenceNode(jointNames=@n8.Value.leg_18_#14)
   > createNullObjects_6 = RigVMFunctionReferenceNode(joint_names=@mirrorJointNames_2.jointNames_mirror)
   n9 = &RigVMFunction_ControlFlowBranch(
    ExecuteContext=@RigVMFunction_Sequence_2.C
    Condition=RigVMVariableNode() {variable=bool_null_structure}.Value.head_11_#16
   )
   createNullObjects_8 = RigVMFunctionReferenceNode(ExecuteContext=@n9.True, joint_names=RigVMVariableNode() {variable=joint_names_structure}.Value.head_19_#17)
   For_Each = RigVMDispatchNode(ExecuteContext=@RigVMFunction_Sequence.C, Array=RigVMVariableNode() {variable=Current_JointItems}.Value) {template="DISPATCH_RigVMDispatch_ArrayIterator(in Array,out Element,out Index,out Count,out Ratio)"}
   > Add = RigVMDispatchNode(Element=@For_Each.Element.Name, Array=RigVMVariableNode() {variable=current_jointNames}.Value) {template="DISPATCH_RigVMDispatch_ArrayAdd(io Array,in Element,out Index)"}
   createNullObjects_0 = RigVMFunctionReferenceNode(ExecuteContext=@For_Each.Completed, joint_names=RigVMVariableNode() {variable=current_jointNames}.Value)
   RigVMFunctionReturnNode(ExecuteContext=@RigVMFunction_Sequence.D)
  ]}
 New Function_ContainedGraph {nodes [
   n1 = RigVMFunctionEntryNode()
   > For_Each = RigVMDispatchNode(Array=@n1.joint_names) {template="DISPATCH_RigVMDispatch_ArrayIterator(in Array,out Element,out Index,out Count,out Ratio)"}
   n2 = /Script/ControlRig.RigUnit_HierarchyGetParent(Child=(Type=Bone,Name="None"), bDefaultParent=True, Child.Name=@For_Each.Element) {is=struct RigVMUnitNode}
   RigVMFunction_ControlFlowBranch = &RigVMFunction_ControlFlowBranch(ExecuteContext=@For_Each.ExecuteContext, Condition=&RigUnit_ItemExists(Item=@n2.Parent).Exists)
   Concat = &RigVMFunction_NameConcat(B=_null, A=@For_Each.Element) {template="Concat::Execute(in A,in B,out Result)"}
   n3 = &RigUnit_HierarchyAddNull(
    Space=LocalSpace
    Name=@Concat.Result
    Transform=&RigUnit_GetTransform(Item=(Type=Bone,Name="None"), Space=LocalSpace, bInitial=true, Item.Name=@For_Each.Element).Transform
    Parent=@n2.Parent
    ExecutePin=@RigVMFunction_ControlFlowBranch.True
   )
   > &RigUnit_SetMetadataTag(NameSpace=Self, Tag=BodyNull, Item=@n3.Item)
   RigVMFunctionReturnNode(ExecuteContext=@For_Each.Completed)
  ]}
 New Function_ContainedGraph {nodes [
   n1 = RigVMFunctionEntryNode()
   > RigVMVariableNode(Value=@n1.joint_to_fix) {variable=joint_to_fix_name}
   > n2 = RigVMVariableNode(Value=@n1.reverse) {variable=reverse_bool}
   Concat = &RigVMFunction_NameConcat(B=_null, A=@n1.joint_name) {template="Concat::Execute(in A,in B,out Result)"}
   RigVMFunction_ControlFlowBranch = &RigVMFunction_ControlFlowBranch(
    Condition=&RigUnit_ItemExists(Item=(Type=Null,Name="None"), Item.Name=@Concat.Result).Exists
    ExecuteContext=@n2.ExecuteContext
   )
   RigVMFunctionReturnNode(ExecuteContext=@RigVMFunction_ControlFlowBranch.Completed)
   n3 = &RigVMFunction_ControlFlowBranch(ExecuteContext=@RigVMFunction_ControlFlowBranch.True, Condition=RigVMVariableNode() {variable=reverse_bool}.Value)
   Interpolate = /Script/RigVM.RigVMFunction_MathQuaternionSlerp(B=(X=0,Y=0,Z=0,W=1), A=RigVMVariableNode() {variable=localRotation_twist}.Value, T=@n1.multiplier) {is=struct RigVMUnitNode, template="Interpolate::Execute(in A,in B,in T,out Result)"}
   &RigUnit_ModifyTransforms(
    ItemToModify=((Item=(Type=Bone,Name="None"),Transform=(Rotation=(X=0,Y=0,Z=0,W=1),Translation=(X=0,Y=0,Z=0),Scale3D=(X=1,Y=1,Z=1))))
    Weight=1
    WeightMinimum=0
    ItemToModify.0.Transform.Rotation=@Interpolate.Result
    ExecutePin=@n3.Completed
    ItemToModify.0.Item.Name=RigVMVariableNode() {variable=joint_to_fix_name}.Value
   )
   RigVMFunction_MathQuaternionSwingTwist = &RigVMFunction_MathQuaternionSwingTwist(
    TwistAxis=(X=1,Y=0,Z=0)
    Input=&RigUnit_GetRelativeTransformForItem(
     Child=(Type=Bone,Name="None")
     Parent=(Type=Null,Name="None")
     bChildInitial=False
     bParentInitial=False
     Parent.Name=@Concat.Result
     Child.Name=@n1.joint_name
    ).RelativeTransform.Rotation
   )
   RigVMVariableNode(ExecuteContext=@n3.False, Value=@RigVMFunction_MathQuaternionSwingTwist.Twist) {variable=localRotation_twist}
   Inverse = /Script/RigVM.RigVMFunction_MathQuaternionInverse(Value=@RigVMFunction_MathQuaternionSwingTwist.Twist) {is=struct RigVMUnitNode, template="Inverse::Execute(in Value,out Result)"}
   RigVMVariableNode(Value=@Inverse.Result, ExecuteContext=@n3.True) {variable=localRotation_twist}
  ]}
 RigVMFunctionLibrary {nodes [
   addFKCtrl = RigVMCollapseNode(
    Custom_FKCtrl=()
    ExecuteContext=()
    Options_FKCtrl="(pelvis_1_#1=False,spine_3_#2=False,clavicle_5_#3=False,arm_7_#4=False,leg_9_#5=False,head_11_46DC15EF465C ...(truncated)"
    Shape_Settings="(pelvis_2_#7=(bVisible=True,Name=\"Hexagon_Thin\",Color=(R=0.0117,G=0.8281,B=1,A=1),Transform=(Rotation=(X=0,Y=-0.7071,Z=0,W=0.7071),Translation=(X=2,Y=0,Z=0),Scale3D=(X=5,Y=5,Z=7))),spine_4_#9=(bVi ...(truncated)"
   )
   addNullObject = RigVMCollapseNode(
    Custom_Joints=()
    ExecuteContext=()
    Options_Null="(pelvis_1_#1=False,spine_3_#2=False,clavicle_5_#3=False,arm_7_#4=False,leg_9_#5=False,head_11_46DC15EF465C ...(truncated)"
   )
   adjustShapeCtrls = RigVMCollapseNode(ExecuteContext=(), ctrlNames=(), scaleFactor=10)
   createFKCtrl = RigVMCollapseNode(
    ExecuteContext=()
    FKChain=true
    JointNames=()
    ShapeSettings=(bVisible=True,Name="Default",Color=(R=1,G=0,B=0,A=1),Transform=(Rotation=(X=0,Y=0,Z=0,W=1),Translation=(X=0,Y=0,Z=0),Scale3D=(X=1,Y=1,Z=1)))
   )
   createNullObjects = RigVMCollapseNode(ExecuteContext=(), joint_names=())
   fixTwistRotation = RigVMCollapseNode(ExecuteContext=())
   getCurrentParentCtrl = RigVMCollapseNode(
    ExecuteContext=()
    FKCtrl_bool_structure="(pelvis_1_#1=False,spine_3_#2=False,clavicle_5_#3=False,arm_7_#4=False,leg_9_#5=False,head_11_46DC15EF465C ...(truncated)"
   )
   mirrorJointNames = RigVMCollapseNode(ExecuteContext=(), jointNames=())
   mirrorShapeCtrls = RigVMCollapseNode(
    ExecuteContext=()
    shapeControl=(bVisible=True,Name="Default",Color=(R=1,G=0,B=0,A=1),Transform=(Rotation=(X=0,Y=0,Z=0,W=1),Translation=(X=0,Y=0,Z=0),Scale3D=(X=1,Y=1,Z=1)))
   )
   setFKCtrl = RigVMCollapseNode(ExecuteContext=(), setFK_Ctrl=False)
  ]}
 RigVMFunction_Sequence_1_ContainedGraph {nodes [
   n1 = RigVMFunctionEntryNode()
   RigVMFunction_Sequence = &RigVMFunction_Sequence(ExecuteContext=@n1.ExecuteContext)
   n2 = &RigVMFunction_Sequence(ExecuteContext=@RigVMFunction_Sequence.B)
   n3 = &RigVMFunction_Sequence(ExecuteContext=@n2.B)
   RigVMFunctionReturnNode(A=@RigVMFunction_Sequence.A, B=@n2.A, C=@n3.A, D=@n3.B)
  ]}
 RigVMFunction_Sequence_1_ContainedGraph {nodes [
   n1 = RigVMFunctionEntryNode()
   RigVMFunction_Sequence = &RigVMFunction_Sequence(ExecuteContext=@n1.ExecuteContext)
   n2 = &RigVMFunction_Sequence(ExecuteContext=@RigVMFunction_Sequence.B)
   RigVMFunctionReturnNode(A=@RigVMFunction_Sequence.A, B=@n2.A, C=@n2.B)
  ]}
 RigVMFunction_Sequence_1_ContainedGraph {nodes [
   n1 = RigVMFunctionEntryNode()
   RigVMFunction_Sequence = &RigVMFunction_Sequence(ExecuteContext=@n1.ExecuteContext)
   n2 = &RigVMFunction_Sequence(ExecuteContext=@RigVMFunction_Sequence.B)
   RigVMFunctionReturnNode(A=@RigVMFunction_Sequence.A, B=@n2.A, F=@n2.B)
  ]}
 RigVMFunction_Sequence_1_ContainedGraph {nodes [
   n1 = RigVMFunctionEntryNode()
   RigVMFunction_Sequence = &RigVMFunction_Sequence(ExecuteContext=@n1.ExecuteContext)
   n2 = &RigVMFunction_Sequence(ExecuteContext=@RigVMFunction_Sequence.B)
   n3 = &RigVMFunction_Sequence(ExecuteContext=@n2.B)
   RigVMFunctionReturnNode(A=@RigVMFunction_Sequence.A, B=@n2.A, C=@n3.A, D=@n3.B)
  ]}
 RigVMFunction_Sequence_2_ContainedGraph {nodes [
   n1 = RigVMFunctionEntryNode()
   n2 = &RigVMFunction_Sequence(ExecuteContext=@n1.ExecuteContext)
   n3 = &RigVMFunction_Sequence(ExecuteContext=@n2.B)
   n4 = &RigVMFunction_Sequence(ExecuteContext=@n3.B)
   RigVMFunctionReturnNode(A=@n2.A, B=@n3.A, C=@n4.A, D=@n4.B)
  ]}
 RigVMFunction_Sequence_2_ContainedGraph {nodes [
   n1 = RigVMFunctionEntryNode()
   n2 = &RigVMFunction_Sequence(ExecuteContext=@n1.ExecuteContext)
   n3 = &RigVMFunction_Sequence(ExecuteContext=@n2.B)
   RigVMFunctionReturnNode(A=@n2.A, B=@n3.A, C=@n3.B)
  ]}
 RigVMFunction_Sequence_2_ContainedGraph {nodes [
   n1 = RigVMFunctionEntryNode()
   n2 = &RigVMFunction_Sequence(ExecuteContext=@n1.ExecuteContext)
   n3 = &RigVMFunction_Sequence(ExecuteContext=@n2.B)
   RigVMFunctionReturnNode(A=@n2.A, B=@n3.A, C=@n3.B)
  ]}
 RigVMFunction_Sequence_3_ContainedGraph {nodes [
   n1 = RigVMFunctionEntryNode()
   n2 = &RigVMFunction_Sequence(ExecuteContext=@n1.ExecuteContext)
   n3 = &RigVMFunction_Sequence(ExecuteContext=@n2.B)
   RigVMFunctionReturnNode(B=@n3.A, C=@n3.B, A=@n2.A)
  ]}
 RigVMFunction_Sequence_3_ContainedGraph {nodes [
   n1 = RigVMFunctionEntryNode()
   n2 = &RigVMFunction_Sequence(ExecuteContext=@n1.ExecuteContext)
   n3 = &RigVMFunction_Sequence(ExecuteContext=@n2.B)
   RigVMFunctionReturnNode(B=@n3.A, C=@n3.B, A=@n2.A)
  ]}
 RigVMFunction_Sequence_3_ContainedGraph {nodes [
   n1 = RigVMFunctionEntryNode()
   n2 = &RigVMFunction_Sequence(ExecuteContext=@n1.ExecuteContext)
   n3 = &RigVMFunction_Sequence(ExecuteContext=@n2.B)
   RigVMFunctionReturnNode(B=@n3.A, C=@n3.B, A=@n2.A)
  ]}
 RigVMModel {nodes [
   n1 = &RigUnit_PrepareForExecution()
   addFKCtrl_0 = RigVMFunctionReferenceNode(
    Custom_FKCtrl="((parent_2_#19=(Type=None,Name=\"None\"),shape_settings_24_#23=(bVisible=True,Name=\"Default\",Color=(R=1,G=0,B=0,A=1),Transform=(Rotation=(X=0,Y=0,Z=0,W=1),Translation=(X=0,Y=0,Z=0),Scale3D=(X=1,Y=1,Z= ...(truncated)"
    Options_FKCtrl="(pelvis_1_#1=False,spine_3_#2=False,clavicle_5_#3=False,arm_7_#4=True,leg_9_#5=True,head_11_46DC15EF465CD2 ...(truncated)"
    Shape_Settings="(pelvis_2_#7=(bVisible=True,Name=\"Hexagon_Thin\",Color=(R=0.0117,G=0.8281,B=1,A=1),Transform=(Rotation=(X=0,Y=-0.7071,Z=0,W=0.7071),Translation=(X=2,Y=0,Z=0),Scale3D=(X=5,Y=5,Z=7))),spine_4_#9=(bVi ...(truncated)"
    ExecuteContext=@n1.ExecutePin
   )
   RigUnit_ItemArray = &RigUnit_ItemArray(Items=((Type=Bone,Name="pelvis"),(Type=Bone,Name="thigh_l"),(Type=Bone,Name="thigh_r")))
   > addNullObject = RigVMFunctionReferenceNode(
    Options_Null="(pelvis_1_#1=False,spine_3_#2=False,clavicle_5_#3=False,arm_7_#4=False,leg_9_#5=False,head_11_46DC15EF465C ...(truncated)"
    Custom_Joints=@RigUnit_ItemArray.Items
   )
   RigUnit_BeginExecution = &RigUnit_BeginExecution()
   setFKCtrl_0 = RigVMFunctionReferenceNode(setFK_Ctrl=true, ExecuteContext=@RigUnit_BeginExecution.ExecutePin)
  ]}
]
hierarchy [
 root {loc=<0 0 0>, rot=<0 0 0 1>}
 attach {parent=root, loc=<0 0 0>, rot=<0 0 0 1>}
 VB attach_ik_hand_gun {parent=attach, loc=<-33.8029 3.8283 90.5694>, rot=<0.4834 0.3223 0.5887 -0.5621>}
 VB VB attach_ik_hand_gun_ik_hand_l {parent=VB attach_ik_hand_gun, loc=<6.7067 65.8041 13.978>, rot=<-0.0992 -0.9733 -0.2068 0>}
 VB VB attach_ik_hand_gun_ik_hand_r {parent=VB attach_ik_hand_gun, loc=<0 0 0>, rot=<0 0 0 -1>}
 VB attach_ik_foot_l {parent=attach, loc=<15.2608 -4.4176 9.4433>, rot=<0.0357 0.7045 0.0284 -0.7082>}
 VB attach_ik_foot_r {parent=attach, loc=<-15.2609 -4.4176 9.4431>, rot=<0.7082 0.0284 -0.7045 0.0357>}
 pelvis {parent=root, loc=<0 -0.7071 93.1191>, rot=<0.0127 0.707 -0.0127 -0.707>}
 spine_01 {parent=pelvis, loc=<6.7494 -0.6388 0>, rot=<0 0 0.034 -0.9994>}
 spine_02 {parent=spine_01, loc=<7.2518 -0.035 0>, rot=<0 0 0.0238 -0.9997>}
 spine_03 {parent=spine_02, loc=<6.0104 0 0>, rot=<0 0 0.0149 -0.9999>}
 spine_04 {parent=spine_03, loc=<8.1024 0 0>, rot=<0 0 -0.0132 -0.9999>}
 spine_05 {parent=spine_04, loc=<10.0146 0 0>, rot=<0 0 -0.0015 -1>}
 clavicle_l {parent=spine_05, loc=<4.3575 -0.1466 -2.3223>, rot=<0.0465 -0.7289 -0.009 -0.6829>}
 upperarm_l {parent=clavicle_l, loc=<15.5111 0 0>, rot=<0.0305 -0.536 -0.0057 -0.8437>}
 lowerarm_l {parent=upperarm_l, loc=<26.6514 0 0>, rot=<0.0391 -0.0099 -0.2453 -0.9686>}
 hand_l {parent=lowerarm_l, loc=<22.3491 0 0>, rot=<-0.7081 -0.0507 0.0975 -0.6975>}
 index_metacarpal_l {parent=hand_l, loc=<1.6535 0.0183 -2.96>, rot=<0.0158 -0.0124 -0.0173 -0.9996>}
 index_01_l {parent=index_metacarpal_l, loc=<8.3008 0.0357 -0.0524>, rot=<-0.0044 0.0485 0.0924 -0.9945>}
 index_02_l {parent=index_01_l, loc=<3.3587 -0.3004 0.0452>, rot=<-0.002 0.0027 0.3225 -0.9466>}
 index_03_l {parent=index_02_l, loc=<2.5578 -0.123 -0.0252>, rot=<-0.002 0.0067 0.2558 -0.9667>}
 middle_metacarpal_l {parent=hand_l, loc=<2.1856 -0.1725 -0.5984>, rot=<0.006 0.0168 -0.0329 -0.9993>}
 middle_01_l {parent=middle_metacarpal_l, loc=<7.3275 -0.2447 0.0129>, rot=<-0.0208 0.0042 0.0858 -0.9961>}
 middle_02_l {parent=middle_01_l, loc=<4.0123 0.1401 0.1175>, rot=<-0.0218 0.0258 0.3512 -0.9357>}
 middle_03_l {parent=middle_02_l, loc=<2.6077 0.0534 0.1373>, rot=<0.0082 0.0188 0.1708 -0.9851>}
 pinky_metacarpal_l {parent=hand_l, loc=<1.7616 -0.9753 3.3814>, rot=<-0.0397 0.0814 -0.069 -0.9935>}
 pinky_01_l {parent=pinky_metacarpal_l, loc=<7.1709 -0.4831 -0.1094>, rot=<-0.0654 -0.0035 0.2242 -0.9723>}
 pinky_02_l {parent=pinky_01_l, loc=<2.8274 0.0893 -0.1043>, rot=<0.0037 0.0011 0.2854 -0.9584>}
 pinky_03_l {parent=pinky_02_l, loc=<2.2046 0.0307 -0.0078>, rot=<-0.0002 0.0098 0.1445 -0.9895>}
 ring_metacarpal_l {parent=hand_l, loc=<2.2074 -0.3625 1.3638>, rot=<-0.0284 0.0722 -0.0458 -0.9959>}
 ring_01_l {parent=ring_metacarpal_l, loc=<7.0295 -0.6197 -0.1446>, rot=<-0.0489 -0.0252 0.1892 -0.9804>}
 ring_02_l {parent=ring_01_l, loc=<3.6174 0.4647 -0.1027>, rot=<0 0 0.2806 -0.9598>}
 ring_03_l {parent=ring_02_l, loc=<2.4152 0.114 0.0199>, rot=<0 0 0.1189 -0.9929>}
 thumb_01_l {parent=hand_l, loc=<1.3118 -0.3191 -2.459>, rot=<0.6546 -0.4042 0.0664 -0.6354>}
 thumb_02_l {parent=thumb_01_l, loc=<5.9336 -0.0448 0.0401>, rot=<-0.0037 0.0213 0.298 -0.9543>}
 thumb_03_l {parent=thumb_02_l, loc=<2.2223 0.3682 -0.0907>, rot=<0.0118 -0.013 0.0066 -0.9998>}
 weapon_l {parent=hand_l, loc=<-0.1974 -3.4111 0.3811>, rot=<0 0 0 1>}
 lowerarm_twist_01_l {parent=lowerarm_l, loc=<17.8549 0 0>, rot=<0 0 0 1>}
 lowerarm_twist_02_l {parent=lowerarm_l, loc=<9.6199 0 0>, rot=<0 0 0 1>}
 upperarm_twist_01_l {parent=upperarm_l, loc=<0.0041 0 0>, rot=<0 0 0 1>}
 upperarm_twist_02_l {parent=upperarm_l, loc=<13.7941 0 0>, rot=<0 0 0 1>}
 clavicle_r {parent=spine_05, loc=<4.3575 -0.1466 2.3223>, rot=<0.7289 0.0465 -0.6829 0.009>}
 upperarm_r {parent=clavicle_r, loc=<-15.5111 0 0>, rot=<0.0305 -0.536 -0.0057 -0.8437>}
 lowerarm_r {parent=upperarm_r, loc=<-26.6515 0 0>, rot=<0.0391 -0.0099 -0.2453 -0.9686>}
 hand_r {parent=lowerarm_r, loc=<-22.3492 0 0>, rot=<-0.7081 -0.0507 0.0975 -0.6975>}
 index_metacarpal_r {parent=hand_r, loc=<-1.6534 -0.0181 2.9601>, rot=<0.0158 -0.0124 -0.0173 -0.9996>}
 index_01_r {parent=index_metacarpal_r, loc=<-8.3008 -0.0357 0.0524>, rot=<-0.0044 0.0485 0.0924 -0.9945>}
 index_02_r {parent=index_01_r, loc=<-3.3589 0.3003 -0.0453>, rot=<-0.002 0.0027 0.3225 -0.9466>}
 index_03_r {parent=index_02_r, loc=<-2.5578 0.1231 0.0251>, rot=<-0.002 0.0067 0.2558 -0.9667>}
 middle_metacarpal_r {parent=hand_r, loc=<-2.1856 0.1726 0.5984>, rot=<0.006 0.0168 -0.0329 -0.9993>}
 middle_01_r {parent=middle_metacarpal_r, loc=<-7.3273 0.2447 -0.0129>, rot=<-0.0208 0.0042 0.0858 -0.9961>}
 middle_02_r {parent=middle_01_r, loc=<-4.0142 -0.1411 -0.07>, rot=<-0.0217 0.0259 0.3512 -0.9357>}
 middle_03_r {parent=middle_02_r, loc=<-2.6074 -0.0534 -0.1312>, rot=<0.0083 0.0188 0.1708 -0.9851>}
 pinky_metacarpal_r {parent=hand_r, loc=<-1.7617 0.9753 -3.3814>, rot=<-0.0397 0.0814 -0.069 -0.9935>}
 pinky_01_r {parent=pinky_metacarpal_r, loc=<-7.1707 0.4832 0.1094>, rot=<-0.0654 -0.0035 0.2242 -0.9723>}
 pinky_02_r {parent=pinky_01_r, loc=<-2.8274 -0.0892 0.1043>, rot=<0.0037 0.0011 0.2854 -0.9584>}
 pinky_03_r {parent=pinky_02_r, loc=<-2.2045 -0.031 0.0078>, rot=<-0.0002 0.0098 0.1445 -0.9895>}
 ring_metacarpal_r {parent=hand_r, loc=<-2.2073 0.3627 -1.3638>, rot=<-0.0284 0.0722 -0.0458 -0.9959>}
 ring_01_r {parent=ring_metacarpal_r, loc=<-7.0297 0.6197 0.1446>, rot=<-0.0489 -0.0252 0.1892 -0.9804>}
 ring_02_r {parent=ring_01_r, loc=<-3.6171 -0.4648 0.1027>, rot=<0 0 0.2806 -0.9598>}
 ring_03_r {parent=ring_02_r, loc=<-2.4153 -0.1141 -0.0199>, rot=<0 0 0.1189 -0.9929>}
 thumb_01_r {parent=hand_r, loc=<-1.3119 0.3192 2.459>, rot=<0.6546 -0.4042 0.0664 -0.6354>}
 thumb_02_r {parent=thumb_01_r, loc=<-5.9335 0.0447 -0.0401>, rot=<-0.0037 0.0213 0.298 -0.9543>}
 thumb_03_r {parent=thumb_02_r, loc=<-2.2223 -0.3682 0.0907>, rot=<0.0118 -0.013 0.0066 -0.9998>}
 weapon_r {parent=hand_r, loc=<0.1975 3.4112 -0.3811>, rot=<0 0 0 -1>}
 lowerarm_twist_01_r {parent=lowerarm_r, loc=<-17.855 0 0>, rot=<0 0 0 1>}
 lowerarm_twist_02_r {parent=lowerarm_r, loc=<-9.6199 0 0>, rot=<0 0 0 1>}
 upperarm_twist_01_r {parent=upperarm_r, loc=<-0.0041 0 0>, rot=<0 0 0 1>}
 upperarm_twist_02_r {parent=upperarm_r, loc=<-13.7942 0 0>, rot=<0 0 0 1>}
 neck_01 {parent=spine_05, loc=<11.3138 0.1529 0>, rot=<0 0 -0.2003 -0.9797>}
 neck_02 {parent=neck_01, loc=<4.5627 0 0>, rot=<0 0 0.0504 -0.9987>}
 head {parent=neck_02, loc=<5.3914 0 0>, rot=<0 0 0.115 -0.9934>}
 VB head_fx {parent=head, loc=<0 0 0>, rot=<0 0 0 1>}
 VB spine_05_weapon_r {parent=spine_05, loc=<-41.4719 3.1166 30.5419>, rot=<0.7642 -0.139 0.0493 -0.6278>}
 VB VB spine_05_weapon_r_ik_hand_gun {parent=VB spine_05_weapon_r, loc=<-0.1975 -3.4112 0.3811>, rot=<0 0 0 -1>}
 VB VB spine_05_weapon_r_ik_hand_l {parent=VB spine_05_weapon_r, loc=<6.5092 62.3929 14.3591>, rot=<0.0992 0.9733 0.2068 0>}
 VB spine_05_upperarm_r {parent=spine_05, loc=<3.3828 -1.007 17.7788>, rot=<-0.981 -0.0606 0.184 -0.0088>}
 VB VB spine_05_upperarm_r_lowerarm_r {parent=VB spine_05_upperarm_r, loc=<-26.6514 0 0>, rot=<0.0391 -0.0099 -0.2453 -0.9686>}
 VB VB VB spine_05_upperarm_r_lowerarm_r_hand_r {parent=VB VB spine_05_upperarm_r_lowerarm_r, loc=<-22.3492 0 0>, rot=<-0.7081 -0.0507 0.0975 -0.6975>}
 thigh_l {parent=pelvis, loc=<-2.5693 -0.0109 -7.7501>, rot=<0.0012 0.0572 0.0155 -0.9982>}
 calf_l {parent=thigh_l, loc=<-39.0484 0 0>, rot=<0.0006 -0.0161 -0.0391 -0.9991>}
 calf_twist_01_l {parent=calf_l, loc=<-35.813 -0.0967 0.0397>, rot=<0.0237 0.0095 -0.0035 -0.9997>}
 calf_twist_02_l {parent=calf_l, loc=<-18.8064 -0.0967 0.0397>, rot=<0.0237 0.0095 -0.0035 -0.9997>}
 foot_l {parent=calf_l, loc=<-42.561 0.1077 0.4292>, rot=<0.0418 -0.0455 0.0343 -0.9975>}
 ball_l {parent=foot_l, loc=<-6.1599 13.8264 0.6806>, rot=<0.0001 -0.0001 -0.7063 -0.7079>}
 thigh_twist_01_l {parent=thigh_l, loc=<-21.6303 0 0>, rot=<0.0013 0 0 -1>}
 thigh_r {parent=pelvis, loc=<-2.5691 -0.0109 7.7501>, rot=<0.0572 -0.0012 0.9982 0.0155>}
 calf_r {parent=thigh_r, loc=<39.0487 0 0>, rot=<0.0006 -0.0161 -0.0391 -0.9991>}
 calf_twist_01_r {parent=calf_r, loc=<35.8132 0.0967 -0.0397>, rot=<0.0237 0.0095 -0.0035 -0.9997>}
 calf_twist_02_r {parent=calf_r, loc=<18.8064 0.0967 -0.0397>, rot=<0.0237 0.0095 -0.0035 -0.9997>}
 foot_r {parent=calf_r, loc=<42.5612 -0.1077 -0.4292>, rot=<0.0418 -0.0455 0.0343 -0.9975>}
 ball_r {parent=foot_r, loc=<6.1598 -13.8264 -0.6806>, rot=<0.0001 -0.0001 -0.7063 -0.7079>}
 thigh_twist_01_r {parent=thigh_r, loc=<21.6303 0 0>, rot=<0.0013 0 0 -1>}
 VB pelvis_ik_hand_gun {parent=pelvis, loc=<-2.3857 4.6238 33.8029>, rot=<0.7549 -0.1831 0.0856 -0.6239>}
 VB pelvis_ik_hand_r {parent=pelvis, loc=<-2.3857 4.6238 33.8029>, rot=<-0.7549 0.1831 -0.0856 0.6239>}
 VB pelvis_ik_hand_l {parent=pelvis, loc=<-2.3856 4.6238 -33.8029>, rot=<0.1831 0.7549 -0.6239 -0.0856>}
 ik_foot_root {parent=root, loc=<0 0 0>, rot=<0 0 0 1>}
 ik_foot_l {parent=ik_foot_root, loc=<15.2608 -4.4176 9.4433>, rot=<0.0357 0.7045 0.0284 -0.7082>}
 ik_foot_r {parent=ik_foot_root, loc=<-15.2609 -4.4176 9.4431>, rot=<0.7082 0.0284 -0.7045 0.0357>}
 VB ik_foot_root_weapon_l {parent=ik_foot_root, loc=<30.5419 2.8604 90.0703>, rot=<0.5621 0.5887 -0.3223 0.4834>}
 VB ik_foot_root_weapon_r {parent=ik_foot_root, loc=<-30.5419 2.8605 90.0703>, rot=<-0.4834 -0.3223 -0.5887 0.5621>}
 ik_hand_root {parent=root, loc=<0 0 0>, rot=<0 0 0 1>}
 ik_hand_gun {parent=ik_hand_root, loc=<-33.8029 3.8283 90.5694>, rot=<0.4834 0.3223 0.5887 -0.5621>}
 ik_hand_l {parent=ik_hand_gun, loc=<6.7067 65.8041 13.978>, rot=<-0.0992 -0.9733 -0.2068 0>}
 ik_hand_r {parent=ik_hand_gun, loc=<0 0 0>, rot=<0 0 0 1>}
 VB ik_hand_root_ik_hand_l {parent=ik_hand_root, loc=<33.8029 3.8282 90.5694>, rot=<0.5621 0.5887 -0.3223 0.4834>}
 VB root_prop {parent=root, loc=<0 0 0>, rot=<0 0 0 1>}
 VB Root_hand_l {parent=root, loc=<33.8029 3.8282 90.5695>, rot=<0.5621 0.5887 -0.3223 0.4834>}
 VB Root_hand_r {parent=root, loc=<-33.8029 3.8282 90.5694>, rot=<-0.4834 -0.3223 -0.5887 0.5621>}
 r_knee_corrective {type=Curve, loc=<0 0 0>, rot=<0 0 0 1>}
 l_knee_corrective {type=Curve, loc=<0 0 0>, rot=<0 0 0 1>}
 Male_Med_Head_01_Pout {type=Curve, loc=<0 0 0>, rot=<0 0 0 1>}
 Male_Med_Head_02_Pout {type=Curve, loc=<0 0 0>, rot=<0 0 0 1>}
 Male_Med_Head_03_Pout {type=Curve, loc=<0 0 0>, rot=<0 0 0 1>}
 Male_Med_Head_05_Effort {type=Curve, loc=<0 0 0>, rot=<0 0 0 1>}
 Stretch {type=Curve, loc=<0 0 0>, rot=<0 0 0 1>}
 Squash {type=Curve, loc=<0 0 0>, rot=<0 0 0 1>}
 polySurface151 {type=Curve, loc=<0 0 0>, rot=<0 0 0 1>}
 l_cheek_puff {type=Curve, loc=<0 0 0>, rot=<0 0 0 1>}
 r_cheek_puff {type=Curve, loc=<0 0 0>, rot=<0 0 0 1>}
 l_smile_fix {type=Curve, loc=<0 0 0>, rot=<0 0 0 1>}
 r_smile_fix {type=Curve, loc=<0 0 0>, rot=<0 0 0 1>}
 jaw_open_fix {type=Curve, loc=<0 0 0>, rot=<0 0 0 1>}
 r_lip_sneer_fix {type=Curve, loc=<0 0 0>, rot=<0 0 0 1>}
 l_lip_sneer_fix {type=Curve, loc=<0 0 0>, rot=<0 0 0 1>}
 L_Blink {type=Curve, loc=<0 0 0>, rot=<0 0 0 1>}
 R_Blink {type=Curve, loc=<0 0 0>, rot=<0 0 0 1>}
 L_Arm_Stretch {type=Curve, loc=<0 0 0>, rot=<0 0 0 1>}
 Hair_Hat {type=Curve, loc=<0 0 0>, rot=<0 0 0 1>}
 Hair_Helmet {type=Curve, loc=<0 0 0>, rot=<0 0 0 1>}
 Male_Med_Head_02_What {type=Curve, loc=<0 0 0>, rot=<0 0 0 1>}
 Male_Med_Head_02_Smile {type=Curve, loc=<0 0 0>, rot=<0 0 0 1>}
 Male_Med_Head_02_Scowl {type=Curve, loc=<0 0 0>, rot=<0 0 0 1>}
 Male_Med_Head_02_Hmmph {type=Curve, loc=<0 0 0>, rot=<0 0 0 1>}
 Male_Med_Head_02_Grrr {type=Curve, loc=<0 0 0>, rot=<0 0 0 1>}
 Male_Med_Head_02_Effort {type=Curve, loc=<0 0 0>, rot=<0 0 0 1>}
 Male_Med_Head_03_What {type=Curve, loc=<0 0 0>, rot=<0 0 0 1>}
 Male_Med_Head_03_Grrr {type=Curve, loc=<0 0 0>, rot=<0 0 0 1>}
 Male_Med_Head_03_Effort {type=Curve, loc=<0 0 0>, rot=<0 0 0 1>}
 Male_Med_Head_03_Hmmph {type=Curve, loc=<0 0 0>, rot=<0 0 0 1>}
 Male_Med_Head_03_Scowl {type=Curve, loc=<0 0 0>, rot=<0 0 0 1>}
 Male_Med_Head_03_Smile {type=Curve, loc=<0 0 0>, rot=<0 0 0 1>}
 Male_Med_Head_05_Teeth_Low_What {type=Curve, loc=<0 0 0>, rot=<0 0 0 1>}
 Male_Med_Head_05_Teeth_Low_Pout {type=Curve, loc=<0 0 0>, rot=<0 0 0 1>}
 Male_Med_Head_05_Teeth_Low_Hmmph {type=Curve, loc=<0 0 0>, rot=<0 0 0 1>}
 Male_Med_Head_05_Teeth_Low_Grrr {type=Curve, loc=<0 0 0>, rot=<0 0 0 1>}
 Male_Med_Head_05_Teeth_Low_Effort {type=Curve, loc=<0 0 0>, rot=<0 0 0 1>}
 Male_Med_Head_05_What {type=Curve, loc=<0 0 0>, rot=<0 0 0 1>}
 Male_Med_Head_05_Grrr {type=Curve, loc=<0 0 0>, rot=<0 0 0 1>}
 Male_Med_Head_05_Hmmph {type=Curve, loc=<0 0 0>, rot=<0 0 0 1>}
 Male_Med_Head_05_Pout {type=Curve, loc=<0 0 0>, rot=<0 0 0 1>}
 Male_Med_Head_05_Tongue_What {type=Curve, loc=<0 0 0>, rot=<0 0 0 1>}
 Male_Med_Head_05_Tongue_Pout {type=Curve, loc=<0 0 0>, rot=<0 0 0 1>}
 Male_Med_Head_05_Tongue_Hmmph {type=Curve, loc=<0 0 0>, rot=<0 0 0 1>}
 Male_Med_Head_05_Tongue_Grrr {type=Curve, loc=<0 0 0>, rot=<0 0 0 1>}
 Male_Med_Head_05_Tongue_Effort {type=Curve, loc=<0 0 0>, rot=<0 0 0 1>}
 AdditiveLayer {type=Curve, loc=<0 0 0>, rot=<0 0 0 1>}
 RightHandIK {type=Curve, loc=<0 0 0>, rot=<0 0 0 1>}
 LeftHandIK {type=Curve, loc=<0 0 0>, rot=<0 0 0 1>}
 LeftHandFK {type=Curve, loc=<0 0 0>, rot=<0 0 0 1>}
 Male_Med_Head_01_Grrr {type=Curve, loc=<0 0 0>, rot=<0 0 0 1>}
 WeaponAiming {type=Curve, loc=<0 0 0>, rot=<0 0 0 1>}
 Male_Med_Head_01_Hmmph {type=Curve, loc=<0 0 0>, rot=<0 0 0 1>}
 DragonSlash {type=Curve, loc=<0 0 0>, rot=<0 0 0 1>}
 Male_Med_Head_01_Effort {type=Curve, loc=<0 0 0>, rot=<0 0 0 1>}
 LeftHankFK {type=Curve, loc=<0 0 0>, rot=<0 0 0 1>}
 blendOrient1 {type=Curve, loc=<0 0 0>, rot=<0 0 0 1>}
 blendParent1 {type=Curve, loc=<0 0 0>, rot=<0 0 0 1>}
 bottom_R_Corner_Down {type=Curve, loc=<0 0 0>, rot=<0 0 0 1>}
 bottom_R_Corner_Up {type=Curve, loc=<0 0 0>, rot=<0 0 0 1>}
 top_L_Corner_Up {type=Curve, loc=<0 0 0>, rot=<0 0 0 1>}
 top_L_Corner_Down {type=Curve, loc=<0 0 0>, rot=<0 0 0 1>}
 top_R_Corner_Down {type=Curve, loc=<0 0 0>, rot=<0 0 0 1>}
 top_R_Corner_Up {type=Curve, loc=<0 0 0>, rot=<0 0 0 1>}
 bottom_L_Corner_Down {type=Curve, loc=<0 0 0>, rot=<0 0 0 1>}
 bottomt_L_Corner_Up {type=Curve, loc=<0 0 0>, rot=<0 0 0 1>}
 TrailCurve1 {type=Curve, loc=<0 0 0>, rot=<0 0 0 1>}
 BothHandFK {type=Curve, loc=<0 0 0>, rot=<0 0 0 1>}
 bottom_R_Corner_Down1 {type=Curve, loc=<0 0 0>, rot=<0 0 0 1>}
 blendPoint1 {type=Curve, loc=<0 0 0>, rot=<0 0 0 1>}
 WeaponAimingFreeze {type=Curve, loc=<0 0 0>, rot=<0 0 0 1>}
 l_calf_top_taper {type=Curve, loc=<0 0 0>, rot=<0 0 0 1>}
 JogBlendSpaceRotation {type=Curve, loc=<0 0 0>, rot=<0 0 0 1>}
 ZeroOutPitchWeight {type=Curve, loc=<0 0 0>, rot=<0 0 0 1>}
 MeleeTwist {type=Curve, loc=<0 0 0>, rot=<0 0 0 1>}
 MeleeAnim {type=Curve, loc=<0 0 0>, rot=<0 0 0 1>}
 l_sneer_fix {type=Curve, loc=<0 0 0>, rot=<0 0 0 1>}
 jaw_fix {type=Curve, loc=<0 0 0>, rot=<0 0 0 1>}
 r_sneer_fix {type=Curve, loc=<0 0 0>, rot=<0 0 0 1>}
 TurnRotationAmount {type=Curve, loc=<0 0 0>, rot=<0 0 0 1>}
 TurnEnd {type=Curve, loc=<0 0 0>, rot=<0 0 0 1>}
 BlendIn {type=Curve, loc=<0 0 0>, rot=<0 0 0 1>}
 BlendOut {type=Curve, loc=<0 0 0>, rot=<0 0 0 1>}
 BodyRotateX {type=Curve, loc=<0 0 0>, rot=<0 0 0 1>}
 BodyTranslateX {type=Curve, loc=<0 0 0>, rot=<0 0 0 1>}
 CrescentKickScale {type=Curve, loc=<0 0 0>, rot=<0 0 0 1>}
 mask_geo {type=Curve, loc=<0 0 0>, rot=<0 0 0 1>}
 RemoveJumpAdditive {type=Curve, loc=<0 0 0>, rot=<0 0 0 1>}
 RightLeg_IK {type=Curve, loc=<0 0 0>, rot=<0 0 0 1>}
 SnapWeapon_LHand {type=Curve, loc=<0 0 0>, rot=<0 0 0 1>}
 PivotAnim {type=Curve, loc=<0 0 0>, rot=<0 0 0 1>}
 DefaultAdditiveStateWeight {type=Curve, loc=<0 0 0>, rot=<0 0 0 1>}
 StartAnim {type=Curve, loc=<0 0 0>, rot=<0 0 0 1>}
 StopAnim {type=Curve, loc=<0 0 0>, rot=<0 0 0 1>}
 DisableSpeedWarping {type=Curve, loc=<0 0 0>, rot=<0 0 0 1>}
 DisableDBNOSlopeWarp {type=Curve, loc=<0 0 0>, rot=<0 0 0 1>}
 MaskRightArm {type=Curve, loc=<0 0 0>, rot=<0 0 0 1>}
 TurnInPlaceAnim {type=Curve, loc=<0 0 0>, rot=<0 0 0 1>}
 Head_MESH {type=Curve, loc=<0 0 0>, rot=<0 0 0 1>}
 Hair_MESH {type=Curve, loc=<0 0 0>, rot=<0 0 0 1>}
 LaunchPadTransitionOff {type=Curve, loc=<0 0 0>, rot=<0 0 0 1>}
 C_glabella_up_pose {type=Curve, loc=<0 0 0>, rot=<0 0 0 1>}
 jaw_open_pose {type=Curve, loc=<0 0 0>, rot=<0 0 0 1>}
 L_blink_pose {type=Curve, loc=<0 0 0>, rot=<0 0 0 1>}
 L_brow_down_pose {type=Curve, loc=<0 0 0>, rot=<0 0 0 1>}
 L_brow_squeeze_in_pose {type=Curve, loc=<0 0 0>, rot=<0 0 0 1>}
 L_brow_up_pose {type=Curve, loc=<0 0 0>, rot=<0 0 0 1>}
 L_frown_pose {type=Curve, loc=<0 0 0>, rot=<0 0 0 1>}
 L_lip_corner_narrow_pose {type=Curve, loc=<0 0 0>, rot=<0 0 0 1>}
 L_squint_inner_pose {type=Curve, loc=<0 0 0>, rot=<0 0 0 1>}
 L_wide_pose {type=Curve, loc=<0 0 0>, rot=<0 0 0 1>}
 look_down_pose {type=Curve, loc=<0 0 0>, rot=<0 0 0 1>}
 look_left_pose {type=Curve, loc=<0 0 0>, rot=<0 0 0 1>}
 look_right_pose {type=Curve, loc=<0 0 0>, rot=<0 0 0 1>}
 look_up_pose {type=Curve, loc=<0 0 0>, rot=<0 0 0 1>}
 phoneme_fv_pose {type=Curve, loc=<0 0 0>, rot=<0 0 0 1>}
 phoneme_oo_pose {type=Curve, loc=<0 0 0>, rot=<0 0 0 1>}
 R_blink_pose {type=Curve, loc=<0 0 0>, rot=<0 0 0 1>}
 R_brow_down_pose {type=Curve, loc=<0 0 0>, rot=<0 0 0 1>}
 R_brow_squeeze_in_pose {type=Curve, loc=<0 0 0>, rot=<0 0 0 1>}
 R_brow_up_pose {type=Curve, loc=<0 0 0>, rot=<0 0 0 1>}
 R_frown_pose {type=Curve, loc=<0 0 0>, rot=<0 0 0 1>}
 R_lip_corner_narrow_pose {type=Curve, loc=<0 0 0>, rot=<0 0 0 1>}
 R_squint_inner_pose {type=Curve, loc=<0 0 0>, rot=<0 0 0 1>}
 R_wide_pose {type=Curve, loc=<0 0 0>, rot=<0 0 0 1>}
 DisableArmsHeadAdditive {type=Curve, loc=<0 0 0>, rot=<0 0 0 1>}
 FrontendRotation {type=Curve, loc=<0 0 0>, rot=<0 0 0 1>}
 BothLegFK {type=Curve, loc=<0 0 0>, rot=<0 0 0 1>}
 AllowTurnInPlace {type=Curve, loc=<0 0 0>, rot=<0 0 0 1>}
 MontageTimeStretchCurve {type=Curve, loc=<0 0 0>, rot=<0 0 0 1>}
 ParaGlideLeanAlpha {type=Curve, loc=<0 0 0>, rot=<0 0 0 1>}
 WeaponRInChestSpace {type=Curve, loc=<0 0 0>, rot=<0 0 0 1>}
 JimBeam_AO {type=Curve, loc=<0 0 0>, rot=<0 0 0 1>}
 R_squeeze_pose {type=Curve, loc=<0 0 0>, rot=<0 0 0 1>}
 L_squeeze_pose {type=Curve, loc=<0 0 0>, rot=<0 0 0 1>}
 R_upper_lip_raiser_pose {type=Curve, loc=<0 0 0>, rot=<0 0 0 1>}
 L_cheek_up_pose {type=Curve, loc=<0 0 0>, rot=<0 0 0 1>}
 R_cheek_up_pose {type=Curve, loc=<0 0 0>, rot=<0 0 0 1>}
 DisableFaceOverride {type=Curve, loc=<0 0 0>, rot=<0 0 0 1>}
 EnableFullBodyOverride {type=Curve, loc=<0 0 0>, rot=<0 0 0 1>}
 C_glabella_down_pose {type=Curve, loc=<0 0 0>, rot=<0 0 0 1>}
 C_nose_wrinkler_pose {type=Curve, loc=<0 0 0>, rot=<0 0 0 1>}
 eye_convergence_pose {type=Curve, loc=<0 0 0>, rot=<0 0 0 1>}
 jaw_back_pose {type=Curve, loc=<0 0 0>, rot=<0 0 0 1>}
 jaw_forward_pose {type=Curve, loc=<0 0 0>, rot=<0 0 0 1>}
 jaw_left_pose {type=Curve, loc=<0 0 0>, rot=<0 0 0 1>}
 jaw_right_pose {type=Curve, loc=<0 0 0>, rot=<0 0 0 1>}
 jaw_up_pose {type=Curve, loc=<0 0 0>, rot=<0 0 0 1>}
 L_lip_corner_wide_pose {type=Curve, loc=<0 0 0>, rot=<0 0 0 1>}
 L_lower_lip_down_pose {type=Curve, loc=<0 0 0>, rot=<0 0 0 1>}
 L_lower_lip_up_pose {type=Curve, loc=<0 0 0>, rot=<0 0 0 1>}
 L_smile_pose {type=Curve, loc=<0 0 0>, rot=<0 0 0 1>}
 L_upper_lip_lower_pose {type=Curve, loc=<0 0 0>, rot=<0 0 0 1>}
 L_upper_lip_raiser_pose {type=Curve, loc=<0 0 0>, rot=<0 0 0 1>}
 L_upperLidDown_pose {type=Curve, loc=<0 0 0>, rot=<0 0 0 1>}
 mouth_down_pose {type=Curve, loc=<0 0 0>, rot=<0 0 0 1>}
 mouth_up_pose {type=Curve, loc=<0 0 0>, rot=<0 0 0 1>}
 phoneme_ch_pose {type=Curve, loc=<0 0 0>, rot=<0 0 0 1>}
 phoneme_mbp_pose {type=Curve, loc=<0 0 0>, rot=<0 0 0 1>}
 R_lip_corner_wide_pose {type=Curve, loc=<0 0 0>, rot=<0 0 0 1>}
 R_lower_lip_down_pose {type=Curve, loc=<0 0 0>, rot=<0 0 0 1>}
 R_lower_lip_up_pose {type=Curve, loc=<0 0 0>, rot=<0 0 0 1>}
 R_smile_pose {type=Curve, loc=<0 0 0>, rot=<0 0 0 1>}
 R_upper_lip_lower_pose {type=Curve, loc=<0 0 0>, rot=<0 0 0 1>}
 R_upperLidDown_pose {type=Curve, loc=<0 0 0>, rot=<0 0 0 1>}
 tongue_back_pose {type=Curve, loc=<0 0 0>, rot=<0 0 0 1>}
 tongue_down_pose {type=Curve, loc=<0 0 0>, rot=<0 0 0 1>}
 tongue_forward_pose {type=Curve, loc=<0 0 0>, rot=<0 0 0 1>}
 tongue_up_pose {type=Curve, loc=<0 0 0>, rot=<0 0 0 1>}
 LiveLink {type=Curve, loc=<0 0 0>, rot=<0 0 0 1>}
 ParaGlideDeploy {type=Curve, loc=<0 0 0>, rot=<0 0 0 1>}
 HideLeftWeapon {type=Curve, loc=<0 0 0>, rot=<0 0 0 1>}
 DisableLeftHandIKGolfCart {type=Curve, loc=<0 0 0>, rot=<0 0 0 1>}
 DisableRightHandIKGolfCart {type=Curve, loc=<0 0 0>, rot=<0 0 0 1>}
 DisableHandIKRetargeting {type=Curve, loc=<0 0 0>, rot=<0 0 0 1>}
 DisableConsumableOffsetPose {type=Curve, loc=<0 0 0>, rot=<0 0 0 1>}
 UpperBodyNoAdditive {type=Curve, loc=<0 0 0>, rot=<0 0 0 1>}
 MaskLeftArm {type=Curve, loc=<0 0 0>, rot=<0 0 0 1>}
 ForceUpperBody {type=Curve, loc=<0 0 0>, rot=<0 0 0 1>}
 RightHandFK {type=Curve, loc=<0 0 0>, rot=<0 0 0 1>}
 ParaGlideLeanTransition {type=Curve, loc=<0 0 0>, rot=<0 0 0 1>}
 InMotionEmote {type=Curve, loc=<0 0 0>, rot=<0 0 0 1>}
 ForceFullBodyInFrontEnd {type=Curve, loc=<0 0 0>, rot=<0 0 0 1>}
 WeaponLinFloorSpace {type=Curve, loc=<0 0 0>, rot=<0 0 0 1>}
 FrontEnd_Idle_Bypass {type=Curve, loc=<0 0 0>, rot=<0 0 0 1>}
 DisableSlopeWarping {type=Curve, loc=<0 0 0>, rot=<0 0 0 1>}
 Disable_Custom {type=Curve, loc=<0 0 0>, rot=<0 0 0 1>}
 EmoteRequiresWeapon {type=Curve, loc=<0 0 0>, rot=<0 0 0 1>}
 DisableIKRootAdditive {type=Curve, loc=<0 0 0>, rot=<0 0 0 1>}
 DisableFullBodySprint {type=Curve, loc=<0 0 0>, rot=<0 0 0 1>}
 DetachFeet_Vehicle {type=Curve, loc=<0 0 0>, rot=<0 0 0 1>}
 IndigoBoost {type=Curve, loc=<0 0 0>, rot=<0 0 0 1>}
 DisableUpperBodyAdditiveMeshSpace {type=Curve, loc=<0 0 0>, rot=<0 0 0 1>}
 MeleeAOWeight {type=Curve, loc=<0 0 0>, rot=<0 0 0 1>}
 HandsInRootSpace {type=Curve, loc=<0 0 0>, rot=<0 0 0 1>}
 CrouchPelvisAdjustment {type=Curve, loc=<0 0 0>, rot=<0 0 0 1>}
 EnableAccelYawCorrection {type=Curve, loc=<0 0 0>, rot=<0 0 0 1>}
 ParaGlideModifyRoot {type=Curve, loc=<0 0 0>, rot=<0 0 0 1>}
 UseScalarParameterForBlink {type=Curve, loc=<0 0 0>, rot=<0 0 0 1>}
 BlinkControl {type=Curve, loc=<0 0 0>, rot=<0 0 0 1>}
 ForceTargeting {type=Curve, loc=<0 0 0>, rot=<0 0 0 1>}
 BlinkControlProp {type=Curve, loc=<0 0 0>, rot=<0 0 0 1>}
 UseScalarParameterForBlinkProp {type=Curve, loc=<0 0 0>, rot=<0 0 0 1>}
 CrackshotCozyHeadCorrection {type=Curve, loc=<0 0 0>, rot=<0 0 0 1>}
 ForwardMotion {type=Curve, loc=<0 0 0>, rot=<0 0 0 1>}
 EmoteForcedRootYaw {type=Curve, loc=<0 0 0>, rot=<0 0 0 1>}
 MeleeTrailWidth {type=Curve, loc=<0 0 0>, rot=<0 0 0 1>}
 space_world {type=Curve, loc=<0 0 0>, rot=<0 0 0 1>}
 PeelyBlenderFace {type=Curve, loc=<0 0 0>, rot=<0 0 0 1>}
 Width1 {type=Curve, loc=<0 0 0>, rot=<0 0 0 1>}
 OverrideIKHands {type=Curve, loc=<0 0 0>, rot=<0 0 0 1>}
 ParaGlideAddAlpha {type=Curve, loc=<0 0 0>, rot=<0 0 0 1>}
 ParaGlideTurn {type=Curve, loc=<0 0 0>, rot=<0 0 0 1>}
 DisablePelvisAdjustment {type=Curve, loc=<0 0 0>, rot=<0 0 0 1>}
 UncleBrollyDeploy {type=Curve, loc=<0 0 0>, rot=<0 0 0 1>}
 ForceRootSlopeOrientation {type=Curve, loc=<0 0 0>, rot=<0 0 0 1>}
 DisableAdditiveOffsetPose {type=Curve, loc=<0 0 0>, rot=<0 0 0 1>}
 DisableIdleNoiseAdditive {type=Curve, loc=<0 0 0>, rot=<0 0 0 1>}
 UsingConsumable {type=Curve, loc=<0 0 0>, rot=<0 0 0 1>}
 DisableRightArmAdditive {type=Curve, loc=<0 0 0>, rot=<0 0 0 1>}
 NoGlider {type=Curve, loc=<0 0 0>, rot=<0 0 0 1>}
 ParaGlideModifyPelvis {type=Curve, loc=<0 0 0>, rot=<0 0 0 1>}
 ParaGlideModifyLegs {type=Curve, loc=<0 0 0>, rot=<0 0 0 1>}
 HT_L1_TransitionCurve {type=Curve, loc=<0 0 0>, rot=<0 0 0 1>}
 KeepPelvisAdjustment {type=Curve, loc=<0 0 0>, rot=<0 0 0 1>}
 PatrolDisableIdleDuringWalk {type=Curve, loc=<0 0 0>, rot=<0 0 0 1>}
 PatrolWalkUseFullbodyBlendspace {type=Curve, loc=<0 0 0>, rot=<0 0 0 1>}
 RemoveJumpUpAdditive {type=Curve, loc=<0 0 0>, rot=<0 0 0 1>}
 RemoveJumpLoopAdditive {type=Curve, loc=<0 0 0>, rot=<0 0 0 1>}
 RemoveJumpFallAdditive {type=Curve, loc=<0 0 0>, rot=<0 0 0 1>}
 JH_TransformCurve {type=Curve, loc=<0 0 0>, rot=<0 0 0 1>}
 PhysicsOff {type=Curve, loc=<0 0 0>, rot=<0 0 0 1>}
 DisableWeaponAdditivePose {type=Curve, loc=<0 0 0>, rot=<0 0 0 1>}
 Lexa_TransformationCurve {type=Curve, loc=<0 0 0>, rot=<0 0 0 1>}
 HideProp {type=Curve, loc=<0 0 0>, rot=<0 0 0 1>}
 CopyPose {type=Curve, loc=<0 0 0>, rot=<0 0 0 1>}
 MaterialCurveA {type=Curve, loc=<0 0 0>, rot=<0 0 0 1>}
 MaterialCurveB {type=Curve, loc=<0 0 0>, rot=<0 0 0 1>}
 MaterialCurveC {type=Curve, loc=<0 0 0>, rot=<0 0 0 1>}
 MaterialCurveD {type=Curve, loc=<0 0 0>, rot=<0 0 0 1>}
 MaterialCurveE {type=Curve, loc=<0 0 0>, rot=<0 0 0 1>}
 UseEmoteAnimOffset {type=Curve, loc=<0 0 0>, rot=<0 0 0 1>}
 emissive mult 2 {type=Curve, loc=<0 0 0>, rot=<0 0 0 1>}
 MaxHandle {type=Curve, loc=<0 0 0>, rot=<0 0 0 1>}
 FootPhase {type=Curve, loc=<0 0 0>, rot=<0 0 0 1>}
 TacticalDoorBash {type=Curve, loc=<0 0 0>, rot=<0 0 0 1>}
 Custom_Frontend_Face_Idle {type=Curve, loc=<0 0 0>, rot=<0 0 0 1>}
 HandIKInAttachBoneSpace {type=Curve, loc=<0 0 0>, rot=<0 0 0 1>}
 SwingPhase {type=Curve, loc=<0 0 0>, rot=<0 0 0 1>}
 Moving {type=Curve, loc=<0 0 0>, rot=<0 0 0 1>}
 HandAttachRInChestSpace {type=Curve, loc=<0 0 0>, rot=<0 0 0 1>}
 emote_BuiltIn {type=Curve, loc=<0 0 0>, rot=<0 0 0 1>}
 OhanaPackAdj {type=Curve, loc=<0 0 0>, rot=<0 0 0 1>}
 UseRootSlopeOrientation {type=Curve, loc=<0 0 0>, rot=<0 0 0 1>}
 Backpack_Anim_Override {type=Curve, loc=<0 0 0>, rot=<0 0 0 1>}
 PhysicsGroundOffset {type=Curve, loc=<0 0 0>, rot=<0 0 0 1>}
 AllowRidingNoiseAdditive {type=Curve, loc=<0 0 0>, rot=<0 0 0 1>}
 LeftLegIKOnGroundVehicle {type=Curve, loc=<0 0 0>, rot=<0 0 0 1>}
 RightLegIKOnGroundVehicle {type=Curve, loc=<0 0 0>, rot=<0 0 0 1>}
 BaseHeadDefaultFaceIdle {type=Curve, loc=<0 0 0>, rot=<0 0 0 1>}
 Hurdle_100fwd {type=Curve, loc=<0 0 0>, rot=<0 0 0 1>}
 Hurdle_250fwd {type=Curve, loc=<0 0 0>, rot=<0 0 0 1>}
 Hurdle_100up {type=Curve, loc=<0 0 0>, rot=<0 0 0 1>}
 Hurdle_200up {type=Curve, loc=<0 0 0>, rot=<0 0 0 1>}
 Hurdle_50up {type=Curve, loc=<0 0 0>, rot=<0 0 0 1>}
 Hurdle_Mantle {type=Curve, loc=<0 0 0>, rot=<0 0 0 1>}
 Idle {type=Curve, loc=<0 0 0>, rot=<0 0 0 1>}
 Interruptible {type=Curve, loc=<0 0 0>, rot=<0 0 0 1>}
 CustomFaceIdle {type=Curve, loc=<0 0 0>, rot=<0 0 0 1>}
 Hurdle_Land {type=Curve, loc=<0 0 0>, rot=<0 0 0 1>}
 Hurdle_Fall {type=Curve, loc=<0 0 0>, rot=<0 0 0 1>}
 TacticalJump {type=Curve, loc=<0 0 0>, rot=<0 0 0 1>}
 copyWeaponRtoAttachL {type=Curve, loc=<0 0 0>, rot=<0 0 0 1>}
 CTRL_expressions_browDownL {type=Curve, loc=<0 0 0>, rot=<0 0 0 1>}
 CTRL_expressions_browDownR {type=Curve, loc=<0 0 0>, rot=<0 0 0 1>}
 CTRL_expressions_browLateralL {type=Curve, loc=<0 0 0>, rot=<0 0 0 1>}
 CTRL_expressions_browLateralR {type=Curve, loc=<0 0 0>, rot=<0 0 0 1>}
 CTRL_expressions_browRaiseInL {type=Curve, loc=<0 0 0>, rot=<0 0 0 1>}
 CTRL_expressions_browRaiseInR {type=Curve, loc=<0 0 0>, rot=<0 0 0 1>}
 CTRL_expressions_browRaiseOuterL {type=Curve, loc=<0 0 0>, rot=<0 0 0 1>}
 CTRL_expressions_browRaiseOuterR {type=Curve, loc=<0 0 0>, rot=<0 0 0 1>}
 CTRL_expressions_browUINL {type=Curve, loc=<0 0 0>, rot=<0 0 0 1>}
 CTRL_expressions_browUINR {type=Curve, loc=<0 0 0>, rot=<0 0 0 1>}
 CTRL_expressions_browUOUTL {type=Curve, loc=<0 0 0>, rot=<0 0 0 1>}
 CTRL_expressions_browUOUTR {type=Curve, loc=<0 0 0>, rot=<0 0 0 1>}
 CTRL_expressions_browDINL {type=Curve, loc=<0 0 0>, rot=<0 0 0 1>}
 CTRL_expressions_browDINR {type=Curve, loc=<0 0 0>, rot=<0 0 0 1>}
 CTRL_expressions_browDOUTL {type=Curve, loc=<0 0 0>, rot=<0 0 0 1>}
 CTRL_expressions_browDOUTR {type=Curve, loc=<0 0 0>, rot=<0 0 0 1>}
 CTRL_expressions_eyeBlinkL {type=Curve, loc=<0 0 0>, rot=<0 0 0 1>}
 CTRL_expressions_eyeBlinkR {type=Curve, loc=<0 0 0>, rot=<0 0 0 1>}
 CTRL_expressions_eyeWidenL {type=Curve, loc=<0 0 0>, rot=<0 0 0 1>}
 CTRL_expressions_eyeWidenR {type=Curve, loc=<0 0 0>, rot=<0 0 0 1>}
 CTRL_expressions_eyeSquintInnerL {type=Curve, loc=<0 0 0>, rot=<0 0 0 1>}
 CTRL_expressions_eyeSquintInnerR {type=Curve, loc=<0 0 0>, rot=<0 0 0 1>}
 CTRL_expressions_eyeCheekRaiseL {type=Curve, loc=<0 0 0>, rot=<0 0 0 1>}
 CTRL_expressions_eyeCheekRaiseR {type=Curve, loc=<0 0 0>, rot=<0 0 0 1>}
 CTRL_expressions_eyeUpperLidUpL {type=Curve, loc=<0 0 0>, rot=<0 0 0 1>}
 CTRL_expressions_eyeUpperLidUpR {type=Curve, loc=<0 0 0>, rot=<0 0 0 1>}
 CTRL_expressions_eyeRelaxL {type=Curve, loc=<0 0 0>, rot=<0 0 0 1>}
 CTRL_expressions_eyeRelaxR {type=Curve, loc=<0 0 0>, rot=<0 0 0 1>}
 CTRL_expressions_eyeLowerLidUpL {type=Curve, loc=<0 0 0>, rot=<0 0 0 1>}
 CTRL_expressions_eyeLowerLidUpR {type=Curve, loc=<0 0 0>, rot=<0 0 0 1>}
 CTRL_expressions_eyeLowerLidDownL {type=Curve, loc=<0 0 0>, rot=<0 0 0 1>}
 CTRL_expressions_eyeLowerLidDownR {type=Curve, loc=<0 0 0>, rot=<0 0 0 1>}
 CTRL_expressions_eyelidRollPosUL {type=Curve, loc=<0 0 0>, rot=<0 0 0 1>}
 CTRL_expressions_eyelidRollPosUR {type=Curve, loc=<0 0 0>, rot=<0 0 0 1>}
 CTRL_expressions_eyelidRollPosDL {type=Curve, loc=<0 0 0>, rot=<0 0 0 1>}
 CTRL_expressions_eyelidRollPosDR {type=Curve, loc=<0 0 0>, rot=<0 0 0 1>}
 CTRL_expressions_eyelidRollNegUL {type=Curve, loc=<0 0 0>, rot=<0 0 0 1>}
 CTRL_expressions_eyelidRollNegUR {type=Curve, loc=<0 0 0>, rot=<0 0 0 1>}
 CTRL_expressions_eyelidRollNegDL {type=Curve, loc=<0 0 0>, rot=<0 0 0 1>}
 CTRL_expressions_eyelidRollNegDR {type=Curve, loc=<0 0 0>, rot=<0 0 0 1>}
 CTRL_expressions_eyeLookUpL {type=Curve, loc=<0 0 0>, rot=<0 0 0 1>}
 CTRL_expressions_eyeLookUpR {type=Curve, loc=<0 0 0>, rot=<0 0 0 1>}
 CTRL_expressions_eyeLookDownL {type=Curve, loc=<0 0 0>, rot=<0 0 0 1>}
 CTRL_expressions_eyeLookDownR {type=Curve, loc=<0 0 0>, rot=<0 0 0 1>}
 CTRL_expressions_eyeLookLeftL {type=Curve, loc=<0 0 0>, rot=<0 0 0 1>}
 CTRL_expressions_eyeLookLeftR {type=Curve, loc=<0 0 0>, rot=<0 0 0 1>}
 CTRL_expressions_eyeLookRightL {type=Curve, loc=<0 0 0>, rot=<0 0 0 1>}
 CTRL_expressions_eyeLookRightR {type=Curve, loc=<0 0 0>, rot=<0 0 0 1>}
 CTRL_expressions_eyeEyelidsReduce {type=Curve, loc=<0 0 0>, rot=<0 0 0 1>}
 CTRL_expressions_noseWrinkleL {type=Curve, loc=<0 0 0>, rot=<0 0 0 1>}
 CTRL_expressions_noseWrinkleR {type=Curve, loc=<0 0 0>, rot=<0 0 0 1>}
 CTRL_expressions_noseWrinkleUpperL {type=Curve, loc=<0 0 0>, rot=<0 0 0 1>}
 CTRL_expressions_noseWrinkleUpperR {type=Curve, loc=<0 0 0>, rot=<0 0 0 1>}
 CTRL_expressions_noseNostrilDepressL {type=Curve, loc=<0 0 0>, rot=<0 0 0 1>}
 CTRL_expressions_noseNostrilDepressR {type=Curve, loc=<0 0 0>, rot=<0 0 0 1>}
 CTRL_expressions_noseNostrilDilateL {type=Curve, loc=<0 0 0>, rot=<0 0 0 1>}
 CTRL_expressions_noseNostrilDilateR {type=Curve, loc=<0 0 0>, rot=<0 0 0 1>}
 CTRL_expressions_noseNostrilCompressL {type=Curve, loc=<0 0 0>, rot=<0 0 0 1>}
 CTRL_expressions_noseNostrilCompressR {type=Curve, loc=<0 0 0>, rot=<0 0 0 1>}
 CTRL_expressions_noseRotateL {type=Curve, loc=<0 0 0>, rot=<0 0 0 1>}
 CTRL_expressions_noseRotateR {type=Curve, loc=<0 0 0>, rot=<0 0 0 1>}
 CTRL_expressions_noseSquash {type=Curve, loc=<0 0 0>, rot=<0 0 0 1>}
 CTRL_expressions_noseStretch {type=Curve, loc=<0 0 0>, rot=<0 0 0 1>}
 CTRL_expressions_mouthCheekSuckL {type=Curve, loc=<0 0 0>, rot=<0 0 0 1>}
 CTRL_expressions_mouthCheekSuckR {type=Curve, loc=<0 0 0>, rot=<0 0 0 1>}
 CTRL_expressions_mouthCheekBlowL {type=Curve, loc=<0 0 0>, rot=<0 0 0 1>}
 CTRL_expressions_mouthCheekBlowR {type=Curve, loc=<0 0 0>, rot=<0 0 0 1>}
 CTRL_expressions_mouthLipsBlowL {type=Curve, loc=<0 0 0>, rot=<0 0 0 1>}
 CTRL_expressions_mouthLipsBlowR {type=Curve, loc=<0 0 0>, rot=<0 0 0 1>}
 CTRL_expressions_mouthLeft {type=Curve, loc=<0 0 0>, rot=<0 0 0 1>}
 CTRL_expressions_mouthRight {type=Curve, loc=<0 0 0>, rot=<0 0 0 1>}
 CTRL_expressions_mouthUp {type=Curve, loc=<0 0 0>, rot=<0 0 0 1>}
 CTRL_expressions_mouthDown {type=Curve, loc=<0 0 0>, rot=<0 0 0 1>}
 CTRL_expressions_mouthUpperLipRaiseL {type=Curve, loc=<0 0 0>, rot=<0 0 0 1>}
 CTRL_expressions_mouthUpperLipRaiseR {type=Curve, loc=<0 0 0>, rot=<0 0 0 1>}
 CTRL_expressions_mouthLowerLipDepressL {type=Curve, loc=<0 0 0>, rot=<0 0 0 1>}
 CTRL_expressions_mouthLowerLipDepressR {type=Curve, loc=<0 0 0>, rot=<0 0 0 1>}
 CTRL_expressions_mouthCornerPullL {type=Curve, loc=<0 0 0>, rot=<0 0 0 1>}
 CTRL_expressions_mouthCornerPullR {type=Curve, loc=<0 0 0>, rot=<0 0 0 1>}
 CTRL_expressions_mouthStretchL {type=Curve, loc=<0 0 0>, rot=<0 0 0 1>}
 CTRL_expressions_mouthStretchR {type=Curve, loc=<0 0 0>, rot=<0 0 0 1>}
 CTRL_expressions_mouthDimpleL {type=Curve, loc=<0 0 0>, rot=<0 0 0 1>}
 CTRL_expressions_mouthDimpleR {type=Curve, loc=<0 0 0>, rot=<0 0 0 1>}
 CTRL_expressions_mouthCornerDepressL {type=Curve, loc=<0 0 0>, rot=<0 0 0 1>}
 CTRL_expressions_mouthCornerDepressR {type=Curve, loc=<0 0 0>, rot=<0 0 0 1>}
 CTRL_expressions_mouthPressUL {type=Curve, loc=<0 0 0>, rot=<0 0 0 1>}
 CTRL_expressions_mouthPressUR {type=Curve, loc=<0 0 0>, rot=<0 0 0 1>}
 CTRL_expressions_mouthPressDL {type=Curve, loc=<0 0 0>, rot=<0 0 0 1>}
 CTRL_expressions_mouthPressDR {type=Curve, loc=<0 0 0>, rot=<0 0 0 1>}
 CTRL_expressions_mouthLipsPurseUL {type=Curve, loc=<0 0 0>, rot=<0 0 0 1>}
 CTRL_expressions_mouthLipsPurseUR {type=Curve, loc=<0 0 0>, rot=<0 0 0 1>}
 CTRL_expressions_mouthLipsPurseDL {type=Curve, loc=<0 0 0>, rot=<0 0 0 1>}
 CTRL_expressions_mouthLipsPurseDR {type=Curve, loc=<0 0 0>, rot=<0 0 0 1>}
 CTRL_expressions_mouthLipsTowardsUL {type=Curve, loc=<0 0 0>, rot=<0 0 0 1>}
 CTRL_expressions_mouthLipsTowardsUR {type=Curve, loc=<0 0 0>, rot=<0 0 0 1>}
 CTRL_expressions_mouthLipsTowardsDL {type=Curve, loc=<0 0 0>, rot=<0 0 0 1>}
 CTRL_expressions_mouthLipsTowardsDR {type=Curve, loc=<0 0 0>, rot=<0 0 0 1>}
 CTRL_expressions_mouthFunnelUL {type=Curve, loc=<0 0 0>, rot=<0 0 0 1>}
 CTRL_expressions_mouthFunnelUR {type=Curve, loc=<0 0 0>, rot=<0 0 0 1>}
 CTRL_expressions_mouthFunnelDL {type=Curve, loc=<0 0 0>, rot=<0 0 0 1>}
 CTRL_expressions_mouthFunnelDR {type=Curve, loc=<0 0 0>, rot=<0 0 0 1>}
 CTRL_expressions_mouthLipsTogetherUL {type=Curve, loc=<0 0 0>, rot=<0 0 0 1>}
 CTRL_expressions_mouthLipsTogetherUR {type=Curve, loc=<0 0 0>, rot=<0 0 0 1>}
 CTRL_expressions_mouthLipsTogetherDL {type=Curve, loc=<0 0 0>, rot=<0 0 0 1>}
 CTRL_expressions_mouthLipsTogetherDR {type=Curve, loc=<0 0 0>, rot=<0 0 0 1>}
 CTRL_expressions_mouthUpperLipBiteL {type=Curve, loc=<0 0 0>, rot=<0 0 0 1>}
 CTRL_expressions_mouthUpperLipBiteR {type=Curve, loc=<0 0 0>, rot=<0 0 0 1>}
 CTRL_expressions_mouthLowerLipBiteL {type=Curve, loc=<0 0 0>, rot=<0 0 0 1>}
 CTRL_expressions_mouthLowerLipBiteR {type=Curve, loc=<0 0 0>, rot=<0 0 0 1>}
 CTRL_expressions_mouthCornerSharpenUL {type=Curve, loc=<0 0 0>, rot=<0 0 0 1>}
 CTRL_expressions_mouthCornerSharpenUR {type=Curve, loc=<0 0 0>, rot=<0 0 0 1>}
 CTRL_expressions_mouthCornerSharpenDL {type=Curve, loc=<0 0 0>, rot=<0 0 0 1>}
 CTRL_expressions_mouthCornerSharpenDR {type=Curve, loc=<0 0 0>, rot=<0 0 0 1>}
 CTRL_expressions_mouthCornerRounderUL {type=Curve, loc=<0 0 0>, rot=<0 0 0 1>}
 CTRL_expressions_mouthCornerRounderUR {type=Curve, loc=<0 0 0>, rot=<0 0 0 1>}
 CTRL_expressions_mouthCornerRounderDL {type=Curve, loc=<0 0 0>, rot=<0 0 0 1>}
 CTRL_expressions_mouthCornerRounderDR {type=Curve, loc=<0 0 0>, rot=<0 0 0 1>}
 CTRL_expressions_mouthUpperLipTowardsTeethL {type=Curve, loc=<0 0 0>, rot=<0 0 0 1>}
 CTRL_expressions_mouthUpperLipTowardsTeethR {type=Curve, loc=<0 0 0>, rot=<0 0 0 1>}
 CTRL_expressions_mouthLowerLipTowardsTeethL {type=Curve, loc=<0 0 0>, rot=<0 0 0 1>}
 CTRL_expressions_mouthLowerLipTowardsTeethR {type=Curve, loc=<0 0 0>, rot=<0 0 0 1>}
 CTRL_expressions_mouthCornerUpL {type=Curve, loc=<0 0 0>, rot=<0 0 0 1>}
 CTRL_expressions_mouthCornerUpR {type=Curve, loc=<0 0 0>, rot=<0 0 0 1>}
 CTRL_expressions_mouthCornerDownL {type=Curve, loc=<0 0 0>, rot=<0 0 0 1>}
 CTRL_expressions_mouthCornerDownR {type=Curve, loc=<0 0 0>, rot=<0 0 0 1>}
 CTRL_expressions_mouthCornerWideL {type=Curve, loc=<0 0 0>, rot=<0 0 0 1>}
 CTRL_expressions_mouthCornerWideR {type=Curve, loc=<0 0 0>, rot=<0 0 0 1>}
 CTRL_expressions_mouthCornerNarrowL {type=Curve, loc=<0 0 0>, rot=<0 0 0 1>}
 CTRL_expressions_mouthCornerNarrowR {type=Curve, loc=<0 0 0>, rot=<0 0 0 1>}
 CTRL_expressions_tongueUp {type=Curve, loc=<0 0 0>, rot=<0 0 0 1>}
 CTRL_expressions_tongueDown {type=Curve, loc=<0 0 0>, rot=<0 0 0 1>}
 CTRL_expressions_tongueLeft {type=Curve, loc=<0 0 0>, rot=<0 0 0 1>}
 CTRL_expressions_tongueRight {type=Curve, loc=<0 0 0>, rot=<0 0 0 1>}
 CTRL_expressions_tongueOut {type=Curve, loc=<0 0 0>, rot=<0 0 0 1>}
 CTRL_expressions_tongueIn {type=Curve, loc=<0 0 0>, rot=<0 0 0 1>}
 CTRL_expressions_tongueRollUp {type=Curve, loc=<0 0 0>, rot=<0 0 0 1>}
 CTRL_expressions_tongueRollDown {type=Curve, loc=<0 0 0>, rot=<0 0 0 1>}
 CTRL_expressions_tongueRollLeft {type=Curve, loc=<0 0 0>, rot=<0 0 0 1>}
 CTRL_expressions_tongueRollRight {type=Curve, loc=<0 0 0>, rot=<0 0 0 1>}
 CTRL_expressions_tongueWide {type=Curve, loc=<0 0 0>, rot=<0 0 0 1>}
 CTRL_expressions_tongueNarrow {type=Curve, loc=<0 0 0>, rot=<0 0 0 1>}
 CTRL_expressions_jawOpen {type=Curve, loc=<0 0 0>, rot=<0 0 0 1>}
 CTRL_expressions_jawLeft {type=Curve, loc=<0 0 0>, rot=<0 0 0 1>}
 CTRL_expressions_jawRight {type=Curve, loc=<0 0 0>, rot=<0 0 0 1>}
 CTRL_expressions_jawFwd {type=Curve, loc=<0 0 0>, rot=<0 0 0 1>}
 CTRL_expressions_jawBack {type=Curve, loc=<0 0 0>, rot=<0 0 0 1>}
 CTRL_expressions_jawChinRaiseDL {type=Curve, loc=<0 0 0>, rot=<0 0 0 1>}
 CTRL_expressions_jawChinRaiseDR {type=Curve, loc=<0 0 0>, rot=<0 0 0 1>}
 CTRL_expressions_jawChinRaiseUL {type=Curve, loc=<0 0 0>, rot=<0 0 0 1>}
 CTRL_expressions_jawChinRaiseUR {type=Curve, loc=<0 0 0>, rot=<0 0 0 1>}
 CTRL_expressions_jawOpenExtreme {type=Curve, loc=<0 0 0>, rot=<0 0 0 1>}
 CTRL_expressions_teethUpU {type=Curve, loc=<0 0 0>, rot=<0 0 0 1>}
 CTRL_expressions_teethUpD {type=Curve, loc=<0 0 0>, rot=<0 0 0 1>}
 CTRL_expressions_teethDownU {type=Curve, loc=<0 0 0>, rot=<0 0 0 1>}
 CTRL_expressions_teethDownD {type=Curve, loc=<0 0 0>, rot=<0 0 0 1>}
 CTRL_expressions_teethLeftU {type=Curve, loc=<0 0 0>, rot=<0 0 0 1>}
 CTRL_expressions_teethLeftD {type=Curve, loc=<0 0 0>, rot=<0 0 0 1>}
 CTRL_expressions_teethRightU {type=Curve, loc=<0 0 0>, rot=<0 0 0 1>}
 CTRL_expressions_teethRightD {type=Curve, loc=<0 0 0>, rot=<0 0 0 1>}
 CTRL_expressions_teethFwdU {type=Curve, loc=<0 0 0>, rot=<0 0 0 1>}
 CTRL_expressions_teethFwdD {type=Curve, loc=<0 0 0>, rot=<0 0 0 1>}
 CTRL_expressions_teethBackU {type=Curve, loc=<0 0 0>, rot=<0 0 0 1>}
 CTRL_expressions_teethBackD {type=Curve, loc=<0 0 0>, rot=<0 0 0 1>}
 CTRL_expressions_teethRotateFwdU {type=Curve, loc=<0 0 0>, rot=<0 0 0 1>}
 CTRL_expressions_teethRotateFwdD {type=Curve, loc=<0 0 0>, rot=<0 0 0 1>}
 CTRL_expressions_teethRotateBackU {type=Curve, loc=<0 0 0>, rot=<0 0 0 1>}
 CTRL_expressions_teethRotateBackD {type=Curve, loc=<0 0 0>, rot=<0 0 0 1>}
 is_3L {type=Curve, loc=<0 0 0>, rot=<0 0 0 1>}
 Mirrored {type=Curve, loc=<0 0 0>, rot=<0 0 0 1>}
 LockMovement {type=Curve, loc=<0 0 0>, rot=<0 0 0 1>}
 Walk {type=Curve, loc=<0 0 0>, rot=<0 0 0 1>}
 Sprint {type=Curve, loc=<0 0 0>, rot=<0 0 0 1>}
 Jog {type=Curve, loc=<0 0 0>, rot=<0 0 0 1>}
 Run {type=Curve, loc=<0 0 0>, rot=<0 0 0 1>}
 LeftFootOnGrindRail {type=Curve, loc=<0 0 0>, rot=<0 0 0 1>}
 RightFootOnGrindRail {type=Curve, loc=<0 0 0>, rot=<0 0 0 1>}
 DisableLeftHandIKHipAttach {type=Curve, loc=<0 0 0>, rot=<0 0 0 1>}
 YawCorrectionDirectionSouth {type=Curve, loc=<0 0 0>, rot=<0 0 0 1>}
 YawCorrectionDirectionEast {type=Curve, loc=<0 0 0>, rot=<0 0 0 1>}
 YawCorrectionDirectionWest {type=Curve, loc=<0 0 0>, rot=<0 0 0 1>}
 YawCorrectionDirectionNorth {type=Curve, loc=<0 0 0>, rot=<0 0 0 1>}
 is_face_bone_anim {type=Curve, loc=<0 0 0>, rot=<0 0 0 1>}
 Hover {type=Curve, loc=<0 0 0>, rot=<0 0 0 1>}
 WeaponRinFloorSpace {type=Curve, loc=<0 0 0>, rot=<0 0 0 1>}
 AlphaBlur_R {type=Curve, loc=<0 0 0>, rot=<0 0 0 1>}
 AlphaBlur_L {type=Curve, loc=<0 0 0>, rot=<0 0 0 1>}
 RotationDirection_R {type=Curve, loc=<0 0 0>, rot=<0 0 0 1>}
 RotationDirection_L {type=Curve, loc=<0 0 0>, rot=<0 0 0 1>}
 PlayLowerBodyLocomotionOverFullBodySlot {type=Curve, loc=<0 0 0>, rot=<0 0 0 1>}
 Distance {type=Curve, loc=<0 0 0>, rot=<0 0 0 1>}
 UserCurve01 {type=Curve, loc=<0 0 0>, rot=<0 0 0 1>}
 UserCurve02 {type=Curve, loc=<0 0 0>, rot=<0 0 0 1>}
 LegIKInAttachBoneSpace {type=Curve, loc=<0 0 0>, rot=<0 0 0 1>}
 MoveData_Speed {type=Curve, loc=<0 0 0>, rot=<0 0 0 1>}
 FootSpeed_L {type=Curve, loc=<0 0 0>, rot=<0 0 0 1>}
 FootSpeed_R {type=Curve, loc=<0 0 0>, rot=<0 0 0 1>}
 RightElbowTwist {type=Curve, loc=<0 0 0>, rot=<0 0 0 1>}
 LeftElbowTwist {type=Curve, loc=<0 0 0>, rot=<0 0 0 1>}
 Intensity {type=Curve, loc=<0 0 0>, rot=<0 0 0 1>}
 head {type=Curve, loc=<0 0 0>, rot=<0 0 0 1>}
 neck_01 {type=Curve, loc=<0 0 0>, rot=<0 0 0 1>}
 neck_02 {type=Curve, loc=<0 0 0>, rot=<0 0 0 1>}
 spine_01 {type=Curve, loc=<0 0 0>, rot=<0 0 0 1>}
 spine_04 {type=Curve, loc=<0 0 0>, rot=<0 0 0 1>}
 spine_03 {type=Curve, loc=<0 0 0>, rot=<0 0 0 1>}
 spine_05 {type=Curve, loc=<0 0 0>, rot=<0 0 0 1>}
 upperarm_r {type=Curve, loc=<0 0 0>, rot=<0 0 0 1>}
 hand_l {type=Curve, loc=<0 0 0>, rot=<0 0 0 1>}
 lowerarm_l {type=Curve, loc=<0 0 0>, rot=<0 0 0 1>}
 lowerarm_r {type=Curve, loc=<0 0 0>, rot=<0 0 0 1>}
 lowerarm_twist_01_l {type=Curve, loc=<0 0 0>, rot=<0 0 0 1>}
 upperarm_l {type=Curve, loc=<0 0 0>, rot=<0 0 0 1>}
 ball_r {type=Curve, loc=<0 0 0>, rot=<0 0 0 1>}
 index_metacarpal_l {type=Curve, loc=<0 0 0>, rot=<0 0 0 1>}
 clavicle_l {type=Curve, loc=<0 0 0>, rot=<0 0 0 1>}
 hand_r {type=Curve, loc=<0 0 0>, rot=<0 0 0 1>}
 spine_02 {type=Curve, loc=<0 0 0>, rot=<0 0 0 1>}
 calf_r {type=Curve, loc=<0 0 0>, rot=<0 0 0 1>}
 calf_twist_01_r {type=Curve, loc=<0 0 0>, rot=<0 0 0 1>}
 calf_twist_02_r {type=Curve, loc=<0 0 0>, rot=<0 0 0 1>}
 clavicle_r {type=Curve, loc=<0 0 0>, rot=<0 0 0 1>}
 ik_hand_r {type=Curve, loc=<0 0 0>, rot=<0 0 0 1>}
 thigh_r {type=Curve, loc=<0 0 0>, rot=<0 0 0 1>}
 thigh_twist_01_r {type=Curve, loc=<0 0 0>, rot=<0 0 0 1>}
 calf_l {type=Curve, loc=<0 0 0>, rot=<0 0 0 1>}
 thigh_l {type=Curve, loc=<0 0 0>, rot=<0 0 0 1>}
 pelvis {type=Curve, loc=<0 0 0>, rot=<0 0 0 1>}
 thumb_01_l {type=Curve, loc=<0 0 0>, rot=<0 0 0 1>}
 thumb_02_l {type=Curve, loc=<0 0 0>, rot=<0 0 0 1>}
 thumb_03_l {type=Curve, loc=<0 0 0>, rot=<0 0 0 1>}
 middle_metacarpal_l {type=Curve, loc=<0 0 0>, rot=<0 0 0 1>}
 middle_01_l {type=Curve, loc=<0 0 0>, rot=<0 0 0 1>}
 middle_02_l {type=Curve, loc=<0 0 0>, rot=<0 0 0 1>}
 middle_03_l {type=Curve, loc=<0 0 0>, rot=<0 0 0 1>}
 pinky_01_l {type=Curve, loc=<0 0 0>, rot=<0 0 0 1>}
 pinky_02_l {type=Curve, loc=<0 0 0>, rot=<0 0 0 1>}
 pinky_03_l {type=Curve, loc=<0 0 0>, rot=<0 0 0 1>}
 ring_01_l {type=Curve, loc=<0 0 0>, rot=<0 0 0 1>}
 ring_02_l {type=Curve, loc=<0 0 0>, rot=<0 0 0 1>}
 ring_03_l {type=Curve, loc=<0 0 0>, rot=<0 0 0 1>}
 index_01_r {type=Curve, loc=<0 0 0>, rot=<0 0 0 1>}
 index_02_r {type=Curve, loc=<0 0 0>, rot=<0 0 0 1>}
 index_03_r {type=Curve, loc=<0 0 0>, rot=<0 0 0 1>}
 lowerarm_twist_01_r {type=Curve, loc=<0 0 0>, rot=<0 0 0 1>}
 lowerarm_twist_02_r {type=Curve, loc=<0 0 0>, rot=<0 0 0 1>}
 pinky_metacarpal_r {type=Curve, loc=<0 0 0>, rot=<0 0 0 1>}
 ring_01_r {type=Curve, loc=<0 0 0>, rot=<0 0 0 1>}
 ring_02_r {type=Curve, loc=<0 0 0>, rot=<0 0 0 1>}
 thumb_01_r {type=Curve, loc=<0 0 0>, rot=<0 0 0 1>}
 thumb_02_r {type=Curve, loc=<0 0 0>, rot=<0 0 0 1>}
 thumb_03_r {type=Curve, loc=<0 0 0>, rot=<0 0 0 1>}
 foot_r {type=Curve, loc=<0 0 0>, rot=<0 0 0 1>}
 upperarm_twist_01_l {type=Curve, loc=<0 0 0>, rot=<0 0 0 1>}
 upperarm_twist_01_r {type=Curve, loc=<0 0 0>, rot=<0 0 0 1>}
 root {type=Curve, loc=<0 0 0>, rot=<0 0 0 1>}
 foot_l {type=Curve, loc=<0 0 0>, rot=<0 0 0 1>}
 HandAttachLinFloorSpace {type=Curve, loc=<0 0 0>, rot=<0 0 0 1>}
 HandAttachRinFloorSpace {type=Curve, loc=<0 0 0>, rot=<0 0 0 1>}
 CanBypassSlideInitiate {type=Curve, loc=<0 0 0>, rot=<0 0 0 1>}
 AttachEmote1InPelvisSpace {type=Curve, loc=<0 0 0>, rot=<0 0 0 1>}
 AttachEmote2InPelvisSpace {type=Curve, loc=<0 0 0>, rot=<0 0 0 1>}
 AttachEmote3InPelvisSpace {type=Curve, loc=<0 0 0>, rot=<0 0 0 1>}
 LeftLegFK {type=Curve, loc=<0 0 0>, rot=<0 0 0 1>}
 RightLegFK {type=Curve, loc=<0 0 0>, rot=<0 0 0 1>}
]
