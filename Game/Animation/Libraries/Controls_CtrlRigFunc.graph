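# Control Rig function library with one function, Add Pose Mirroring Metadata, which tags controls with settings for pose mirroring.
# The function loops over a control list and writes its mirror axis, behavioral and no-flip inputs as named metadata on each item, in the Self namespace.
# It only stores these settings and does no mirroring itself; the main rig graph holds just an unconnected Begin Execution event.
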ControlRigBlueprint /Game/Animation/Libraries/Controls_CtrlRigFunc
graphs [
 New Function_ContainedGraph {nodes [
   n1 = RigVMFunctionEntryNode()
   > DISPATCH_RigVMDispatch_ArrayIterator = RigVMDispatchNode(Array=@n1.Control List) {template="DISPATCH_RigVMDispatch_ArrayIterator(in Array,out Element,out Index,out Count,out Ratio)"}
   > RigVMDispatchNode(Name=Mirror Axis, NameSpace=Self, Item=@DISPATCH_RigVMDispatch_ArrayIterator.Element, Value=@n1.Mirror Axis) {template="DISPATCH_RigDispatch_SetMetadata(in Item,in Name,in NameSpace,in Value,out Success)"}
   > DISPATCH_RigDispatch_SetMetadata = RigVMDispatchNode(Name=Mirror behavioral, NameSpace=Self, Item=@DISPATCH_RigVMDispatch_ArrayIterator.Element, Value=@n1.Mirror Behavioral) {template="DISPATCH_RigDispatch_SetMetadata(in Item,in Name,in NameSpace,in Value,out Success)"}
   > RigVMDispatchNode(Name=Mirror No Flip, NameSpace=Self, Item=@DISPATCH_RigVMDispatch_ArrayIterator.Element, Value=@n1.Mirror No Flip) {template="DISPATCH_RigDispatch_SetMetadata(in Item,in Name,in NameSpace,in Value,out Success)"}
   RigVMFunctionReturnNode(ExecuteContext=@DISPATCH_RigVMDispatch_ArrayIterator.Completed)
  ]}
 RigVMFunctionLibrary {nodes [Add Pose Mirroring Metadata = RigVMCollapseNode(Control List=(), ExecuteContext=(), Mirror Axis=(X=0,Y=1,Z=1), Mirror Behavioral=true, Mirror No Flip=true)]}
 RigVMModel {nodes [RigUnit_BeginExecution = &RigUnit_BeginExecution()]}
]
