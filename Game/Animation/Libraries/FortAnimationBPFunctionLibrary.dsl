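; Blueprint function library of Fortnite animation helpers, including emote vehicle wheel rotation, cape-driven pelvis rotation, transformation montage clearing and character cloth control.
; The deprecated cloth controller switches on character state (ground, sprint and others) to select cloth and wind gust parameters.
; It scales linear velocity by player movement and cloth up axis, then passes settings through wind gust and setter helpers alongside a Chaos clothing interactor.
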(fn AdjustEmoteVehicleWheelRotation
  (params (Owning%20Actor :type Actor%20Object%20Reference) (CurrentWheelRotation :type Float%20%28double-precision%29) (DeltaTime :type Float%20%28double-precision%29) (CurrentActorRotation :type Float%20%28double-precision%29))
  (bind (_returnvalue_roll _returnvalue_pitch _returnvalue_yaw) (Math|Rotator|BreakRotator (Transformation|GetActorRotation Owning%20Actor)))
  (return _returnvalue_yaw (Math|Interpolation|FInterpTo CurrentWheelRotation (Math|Float|Clamp(Float) (* (- _returnvalue_yaw CurrentActorRotation) 5.0) -45.0 45.0) DeltaTime 1e+01)))

(fn CapeSpine05PelvisRotationFunction
  (params (SkeletalMeshComponent :type Skinned%20Mesh%20Component%20Object%20Reference))
  ;[0.3217 0.4631 1] Rotate the Root Bone on Extreme Spine_05 Angles
  (if (> (AlphaUtility|GetAlphafromReferenceTransformDelta SkeletalMeshComponent "spine_05" "pelvis" "QuaternionTwist" "(X=0.000000,Y=30.000000)" "(X=0.000000,Y=1.000000)" "0.000000,0.000000,1.000000") 0.01)
    (LocalVariables|SetrotateZ%20output (Math|Rotator|MakeRotator (Math|Float|MapRangeClamped (.yaw (.rotation (Components|SkinnedMesh|GetDeltaTransformfromRefPose SkeletalMeshComponent "spine_05" "pelvis"))) 0.0 -25.0 0.0 5e+01)))
    (else
      (LocalVariables|SetrotateZ%20output "0, 0, 0")))
  (return (LocalVariables|GetrotateZ%20output)))

(fn ClearTransformationMontage
  (params (FortPlayerPawn :type Fort%20Player%20Pawn%20Object%20Reference))
  (Default|SetTransformationMontage FortPlayerPawn) ; Clear out the Transformation Montage
)

(fn ClothControllerFunctionDEPRECATED
  (params (CharacterPartAnimInstance :type Custom%20Character%20Part%20Anim%20Instance%20Object%20Reference) (ClothingAssetName :type String :default "Name of your Clothing Asset from the SkelMesh goes here") (Cloth%20Paramaters :type Cloth%20Params%20Struct%20Structure) (Cloth%20Core%20Settings :type Cloth%20Core%20Setting%20Struct%20Structure) (ClothUpAxis :type Vector :default "0,0,1.000000") (OnGround_Moving_SpeedLimit :type Float%20%28double-precision%29 :default 2.0) (OnGround_Moving_AngularVel_Max :type Float%20%28double-precision%29 :default 0.05) (OnGround_Moving_FictitiousVel_Max :type Float%20%28double-precision%29 :default 0.6) (LinVelScale_FwdBwd_Clamp :type Float%20%28double-precision%29) (LinVelScale_UpDwn_Clamp :type Float%20%28double-precision%29 :default 0.7) (DeltaTime :type Float%20%28double-precision%29) (GameTimeInSeconds :type Float%20%28double-precision%29) (WindGustsParams :type Wind%20Gust%20Params%20Struct%20Structure) (WindGustLocalSettings :type Wind%20Gust%20Local%20Setting%20Struct%20Structure))
  (bind _cloth_settings_local (LocalVariables|GetCloth%20Settings%20Local))
  (bind (_linearvelocityscale_7_9cbe863e417fc965c07809a2fe55adfd _angularvelocityscale_6_6a3dfe4a481ad4d7cb18d79f95cc88e8 _fictitiousangularscale_10_c5f4d16f48d0bedc3f0fa38b2648511b _gravityoverride_15_9c3cf2b14934a1ac917c4c99f03afc6e _dragcoefficient_23_fe68f666411ff9f29ba84cb4be726a52 _liftcoefficient_24_bb285d994a5b3aff167f4793d7eca837) (Utilities|Struct|BreakClothSettingStruct _cloth_settings_local))
  (bind _cloth_params_local (LocalVariables|GetCloth%20Params%20Local))
  (bind (_skydive_goingup_12_c026bda74bfbaba6672dfe8ffdaafc1b _skydive_floatingdown_10_2cbf3d3f421785f8994502931454a81f _skydive_divingdown_11_ec7bec4d406a8a98e60d7b9374c5e457 _skydive_parachuting_13_a88c4a0a407f497561d31287cb20250f _onground_standing_15_267db1464d0d3a29d99e94adfcb4afec _onground_moving_16_807a66b64d7b3b2fcf080b833b8a6254 _onground_crouch_moving_27_ce0b8bfb4e3417c0f790f7ae38dd7735 _emoteormelee_25_cfcf66b5492987cbd638ec98eb34d572 _swimming_3_33ce87ce4d4f712b1384e29115845962 _front_end_20_c7a4562a48e9bc9b4cd34d963cfc6490 _dbno_30_2d70d4004d0e9b117f6ce9a09c362b74 _dbno_carried_32_99829b99495fcfe3488a9a86b549ece5 _onground_falling_36_cdaa0f4d476157e4137548bdce8776f2 _ziplining_39_bccdb7e1431a4c174f1040abc81fd8a8 _skydive_surfing_42_408e65ff484aa54111e1ed8a69d8240d _riding_standing_46_952a2a43408c37c44e4cbfae3f3001b0 _riding_moving_47_ee7a01794460b9753e5076a5a3b3080a _riding_moving_targeting_52_2f57fd9e477c515d70b3da9d75453357 _tactical_sprint_51_5b2273a1499d34bdec52e6a8e3f274d4) (Utilities|Struct|BreakClothParamsStruct _cloth_params_local))
  (bind (_onground_moving_16_807a66b64d7b3b2fcf080b833b8a6254_linearvelocityscale_7_9cbe863e417fc965c07809a2fe55adfd _onground_moving_16_807a66b64d7b3b2fcf080b833b8a6254_angularvelocityscale_6_6a3dfe4a481ad4d7cb18d79f95cc88e8 _onground_moving_16_807a66b64d7b3b2fcf080b833b8a6254_fictitiousangularscale_10_c5f4d16f48d0bedc3f0fa38b2648511b _onground_moving_16_807a66b64d7b3b2fcf080b833b8a6254_gravityoverride_15_9c3cf2b14934a1ac917c4c99f03afc6e _onground_moving_16_807a66b64d7b3b2fcf080b833b8a6254_dragcoefficient_23_fe68f666411ff9f29ba84cb4be726a52 _onground_moving_16_807a66b64d7b3b2fcf080b833b8a6254_liftcoefficient_24_bb285d994a5b3aff167f4793d7eca837 _onground_moving_16_807a66b64d7b3b2fcf080b833b8a6254_jointname_27_c74f3b2245ea319de82a629ff3c17cca) (Utilities|Struct|BreakClothSettingStruct _onground_moving_16_807a66b64d7b3b2fcf080b833b8a6254))
  (bind _windgustsparamslocal (LocalVariables|GetWindGustsParamsLocal))
  (bind (_skydive_goingup_28_c026bda74bfbaba6672dfe8ffdaafc1b _skydive_floatingdown_29_2cbf3d3f421785f8994502931454a81f _skydive_divingdown_30_ec7bec4d406a8a98e60d7b9374c5e457 _skydive_parachuting_31_a88c4a0a407f497561d31287cb20250f _onground_standing_32_267db1464d0d3a29d99e94adfcb4afec _onground_moving_33_807a66b64d7b3b2fcf080b833b8a6254 _onground_crouch_moving_34_ce0b8bfb4e3417c0f790f7ae38dd7735 _emoteormelee_35_cfcf66b5492987cbd638ec98eb34d572 _swimming_36_33ce87ce4d4f712b1384e29115845962 _front_end_37_c7a4562a48e9bc9b4cd34d963cfc6490 _dbno_41_0c046bf1406ab4b752db3a99949fe3a3 _dbno_carried_42_3e9fadc247359fee88c7b4a3453449e6 _onground_falling_46_2cc5cd2149d12b7c53d91e8d6aa90c03 _ziplining_47_b8bbea804f8c515ea80fe7897bb99cf6 _skydive_surfing_50_2b64886e4e6f0e8cfc180e81fdf37f62 _riding_standing_55_206a781845eac0a66d2bbc8981041dbd _riding_moving_56_963fd3424b0d070c81120599a5aed02a _riding_moving_targeting_59_93868b48485365007e3bd8bd8a527b39 _tactical_sprint_58_7e4bd4244308ee1199aa48acf8871e68) (Utilities|Struct|BreakWindGustParamsStruct _windgustsparamslocal))
  (bind _cloth_params_local_2 (LocalVariables|GetCloth%20Params%20Local))
  (bind (_skydive_goingup_12_c026bda74bfbaba6672dfe8ffdaafc1b_2 _skydive_floatingdown_10_2cbf3d3f421785f8994502931454a81f_2 _skydive_divingdown_11_ec7bec4d406a8a98e60d7b9374c5e457_2 _skydive_parachuting_13_a88c4a0a407f497561d31287cb20250f_2 _onground_standing_15_267db1464d0d3a29d99e94adfcb4afec_2 _onground_moving_16_807a66b64d7b3b2fcf080b833b8a6254_2) (Utilities|Struct|BreakClothParamsStruct _cloth_params_local_2))
  (bind (_onground_moving_16_807a66b64d7b3b2fcf080b833b8a6254_linearvelocityscale_7_9cbe863e417fc965c07809a2fe55adfd_2 _onground_moving_16_807a66b64d7b3b2fcf080b833b8a6254_angularvelocityscale_6_6a3dfe4a481ad4d7cb18d79f95cc88e8_2 _onground_moving_16_807a66b64d7b3b2fcf080b833b8a6254_fictitiousangularscale_10_c5f4d16f48d0bedc3f0fa38b2648511b_2 _onground_moving_16_807a66b64d7b3b2fcf080b833b8a6254_gravityoverride_15_9c3cf2b14934a1ac917c4c99f03afc6e_2 _onground_moving_16_807a66b64d7b3b2fcf080b833b8a6254_dragcoefficient_23_fe68f666411ff9f29ba84cb4be726a52_2 _onground_moving_16_807a66b64d7b3b2fcf080b833b8a6254_liftcoefficient_24_bb285d994a5b3aff167f4793d7eca837_2) (Utilities|Struct|BreakClothSettingStruct _onground_moving_16_807a66b64d7b3b2fcf080b833b8a6254_2))
  (bind _cloth_params_local_3 (LocalVariables|GetCloth%20Params%20Local))
  (bind (_skydive_goingup_12_c026bda74bfbaba6672dfe8ffdaafc1b_3 _skydive_floatingdown_10_2cbf3d3f421785f8994502931454a81f_3 _skydive_divingdown_11_ec7bec4d406a8a98e60d7b9374c5e457_3 _skydive_parachuting_13_a88c4a0a407f497561d31287cb20250f_3 _onground_standing_15_267db1464d0d3a29d99e94adfcb4afec_3 _onground_moving_16_807a66b64d7b3b2fcf080b833b8a6254_3 _onground_crouch_moving_27_ce0b8bfb4e3417c0f790f7ae38dd7735_2 _emoteormelee_25_cfcf66b5492987cbd638ec98eb34d572_2 _swimming_3_33ce87ce4d4f712b1384e29115845962_2) (Utilities|Struct|BreakClothParamsStruct _cloth_params_local_3))
  (bind (_linearvelocityscale_7_9cbe863e417fc965c07809a2fe55adfd_2 _angularvelocityscale_6_6a3dfe4a481ad4d7cb18d79f95cc88e8_2 _fictitiousangularscale_10_c5f4d16f48d0bedc3f0fa38b2648511b_2 _gravityoverride_15_9c3cf2b14934a1ac917c4c99f03afc6e_2 _dragcoefficient_23_fe68f666411ff9f29ba84cb4be726a52_2 _liftcoefficient_24_bb285d994a5b3aff167f4793d7eca837_2 _jointname_27_c74f3b2245ea319de82a629ff3c17cca) (Utilities|Struct|BreakClothSettingStruct _swimming_3_33ce87ce4d4f712b1384e29115845962_2))
  (bind _cloth_params_local_4 (LocalVariables|GetCloth%20Params%20Local))
  (bind (_skydive_goingup_12_c026bda74bfbaba6672dfe8ffdaafc1b_4 _skydive_floatingdown_10_2cbf3d3f421785f8994502931454a81f_4 _skydive_divingdown_11_ec7bec4d406a8a98e60d7b9374c5e457_4 _skydive_parachuting_13_a88c4a0a407f497561d31287cb20250f_4 _onground_standing_15_267db1464d0d3a29d99e94adfcb4afec_4 _onground_moving_16_807a66b64d7b3b2fcf080b833b8a6254_4 _onground_crouch_moving_27_ce0b8bfb4e3417c0f790f7ae38dd7735_3 _emoteormelee_25_cfcf66b5492987cbd638ec98eb34d572_3 _swimming_3_33ce87ce4d4f712b1384e29115845962_3 _front_end_20_c7a4562a48e9bc9b4cd34d963cfc6490_2 _dbno_30_2d70d4004d0e9b117f6ce9a09c362b74_2 _dbno_carried_32_99829b99495fcfe3488a9a86b549ece5_2 _onground_falling_36_cdaa0f4d476157e4137548bdce8776f2_2 _ziplining_39_bccdb7e1431a4c174f1040abc81fd8a8_2 _skydive_surfing_42_408e65ff484aa54111e1ed8a69d8240d_2) (Utilities|Struct|BreakClothParamsStruct _cloth_params_local_4))
  (bind (_linearvelocityscale_7_9cbe863e417fc965c07809a2fe55adfd_3 _angularvelocityscale_6_6a3dfe4a481ad4d7cb18d79f95cc88e8_3 _fictitiousangularscale_10_c5f4d16f48d0bedc3f0fa38b2648511b_3 _gravityoverride_15_9c3cf2b14934a1ac917c4c99f03afc6e_3 _dragcoefficient_23_fe68f666411ff9f29ba84cb4be726a52_3 _liftcoefficient_24_bb285d994a5b3aff167f4793d7eca837_3 _jointname_27_c74f3b2245ea319de82a629ff3c17cca_2) (Utilities|Struct|BreakClothSettingStruct _skydive_goingup_12_c026bda74bfbaba6672dfe8ffdaafc1b_4))
  (bind (_skydive_goingup_12_c026bda74bfbaba6672dfe8ffdaafc1b_5 _skydive_floatingdown_10_2cbf3d3f421785f8994502931454a81f_5 _skydive_divingdown_11_ec7bec4d406a8a98e60d7b9374c5e457_5 _skydive_parachuting_13_a88c4a0a407f497561d31287cb20250f_5 _onground_standing_15_267db1464d0d3a29d99e94adfcb4afec_5 _onground_moving_16_807a66b64d7b3b2fcf080b833b8a6254_5) (Utilities|Struct|BreakClothParamsStruct (LocalVariables|GetCloth%20Params%20Local)))
  (bind _onground_moving_16_807a66b64d7b3b2fcf080b833b8a6254_linearvelocityscale_7_9cbe863e417fc965c07809a2fe55adfd_3 (Utilities|Struct|BreakClothSettingStruct _onground_moving_16_807a66b64d7b3b2fcf080b833b8a6254_5))
  (bind (_linearvelocityscale_7_9cbe863e417fc965c07809a2fe55adfd_4 _angularvelocityscale_6_6a3dfe4a481ad4d7cb18d79f95cc88e8_4 _fictitiousangularscale_10_c5f4d16f48d0bedc3f0fa38b2648511b_4 _gravityoverride_15_9c3cf2b14934a1ac917c4c99f03afc6e_4 _dragcoefficient_23_fe68f666411ff9f29ba84cb4be726a52_4 _liftcoefficient_24_bb285d994a5b3aff167f4793d7eca837_4 _jointname_27_c74f3b2245ea319de82a629ff3c17cca_3) (Utilities|Struct|BreakClothSettingStruct _skydive_floatingdown_10_2cbf3d3f421785f8994502931454a81f_4))
  (bind (_linearvelocityscale_7_9cbe863e417fc965c07809a2fe55adfd_5 _angularvelocityscale_6_6a3dfe4a481ad4d7cb18d79f95cc88e8_5 _fictitiousangularscale_10_c5f4d16f48d0bedc3f0fa38b2648511b_5 _gravityoverride_15_9c3cf2b14934a1ac917c4c99f03afc6e_5 _dragcoefficient_23_fe68f666411ff9f29ba84cb4be726a52_5 _liftcoefficient_24_bb285d994a5b3aff167f4793d7eca837_5 _jointname_27_c74f3b2245ea319de82a629ff3c17cca_4) (Utilities|Struct|BreakClothSettingStruct _skydive_divingdown_11_ec7bec4d406a8a98e60d7b9374c5e457_4))
  (bind (_linearvelocityscale_7_9cbe863e417fc965c07809a2fe55adfd_6 _angularvelocityscale_6_6a3dfe4a481ad4d7cb18d79f95cc88e8_6 _fictitiousangularscale_10_c5f4d16f48d0bedc3f0fa38b2648511b_6 _gravityoverride_15_9c3cf2b14934a1ac917c4c99f03afc6e_6 _dragcoefficient_23_fe68f666411ff9f29ba84cb4be726a52_6 _liftcoefficient_24_bb285d994a5b3aff167f4793d7eca837_6 _jointname_27_c74f3b2245ea319de82a629ff3c17cca_5) (Utilities|Struct|BreakClothSettingStruct _skydive_parachuting_13_a88c4a0a407f497561d31287cb20250f_4))
  (bind _cloth_params_local_5 (LocalVariables|GetCloth%20Params%20Local))
  (bind (_skydive_goingup_12_c026bda74bfbaba6672dfe8ffdaafc1b_6 _skydive_floatingdown_10_2cbf3d3f421785f8994502931454a81f_6 _skydive_divingdown_11_ec7bec4d406a8a98e60d7b9374c5e457_6 _skydive_parachuting_13_a88c4a0a407f497561d31287cb20250f_6 _onground_standing_15_267db1464d0d3a29d99e94adfcb4afec_6 _onground_moving_16_807a66b64d7b3b2fcf080b833b8a6254_6 _onground_crouch_moving_27_ce0b8bfb4e3417c0f790f7ae38dd7735_4 _emoteormelee_25_cfcf66b5492987cbd638ec98eb34d572_4 _swimming_3_33ce87ce4d4f712b1384e29115845962_4 _front_end_20_c7a4562a48e9bc9b4cd34d963cfc6490_3) (Utilities|Struct|BreakClothParamsStruct _cloth_params_local_5))
  (bind (_linearvelocityscale_7_9cbe863e417fc965c07809a2fe55adfd_7 _angularvelocityscale_6_6a3dfe4a481ad4d7cb18d79f95cc88e8_7 _fictitiousangularscale_10_c5f4d16f48d0bedc3f0fa38b2648511b_7 _gravityoverride_15_9c3cf2b14934a1ac917c4c99f03afc6e_7 _dragcoefficient_23_fe68f666411ff9f29ba84cb4be726a52_7 _liftcoefficient_24_bb285d994a5b3aff167f4793d7eca837_7 _jointname_27_c74f3b2245ea319de82a629ff3c17cca_6) (Utilities|Struct|BreakClothSettingStruct _front_end_20_c7a4562a48e9bc9b4cd34d963cfc6490_3))
  (bind (_linearvelocityscale_7_9cbe863e417fc965c07809a2fe55adfd_8 _angularvelocityscale_6_6a3dfe4a481ad4d7cb18d79f95cc88e8_8 _fictitiousangularscale_10_c5f4d16f48d0bedc3f0fa38b2648511b_8 _gravityoverride_15_9c3cf2b14934a1ac917c4c99f03afc6e_8 _dragcoefficient_23_fe68f666411ff9f29ba84cb4be726a52_8 _liftcoefficient_24_bb285d994a5b3aff167f4793d7eca837_8 _jointname_27_c74f3b2245ea319de82a629ff3c17cca_7) (Utilities|Struct|BreakClothSettingStruct _emoteormelee_25_cfcf66b5492987cbd638ec98eb34d572_4))
  (bind (_linearvelocityscale_7_9cbe863e417fc965c07809a2fe55adfd_9 _angularvelocityscale_6_6a3dfe4a481ad4d7cb18d79f95cc88e8_9 _fictitiousangularscale_10_c5f4d16f48d0bedc3f0fa38b2648511b_9 _gravityoverride_15_9c3cf2b14934a1ac917c4c99f03afc6e_9 _dragcoefficient_23_fe68f666411ff9f29ba84cb4be726a52_9 _liftcoefficient_24_bb285d994a5b3aff167f4793d7eca837_9 _jointname_27_c74f3b2245ea319de82a629ff3c17cca_8) (Utilities|Struct|BreakClothSettingStruct _onground_standing_15_267db1464d0d3a29d99e94adfcb4afec))
  (bind _cloth_params_local_6 (LocalVariables|GetCloth%20Params%20Local))
  (bind (_skydive_goingup_12_c026bda74bfbaba6672dfe8ffdaafc1b_7 _skydive_floatingdown_10_2cbf3d3f421785f8994502931454a81f_7 _skydive_divingdown_11_ec7bec4d406a8a98e60d7b9374c5e457_7 _skydive_parachuting_13_a88c4a0a407f497561d31287cb20250f_7 _onground_standing_15_267db1464d0d3a29d99e94adfcb4afec_7 _onground_moving_16_807a66b64d7b3b2fcf080b833b8a6254_7 _onground_crouch_moving_27_ce0b8bfb4e3417c0f790f7ae38dd7735_5 _emoteormelee_25_cfcf66b5492987cbd638ec98eb34d572_5 _swimming_3_33ce87ce4d4f712b1384e29115845962_5 _front_end_20_c7a4562a48e9bc9b4cd34d963cfc6490_4 _dbno_30_2d70d4004d0e9b117f6ce9a09c362b74_3 _dbno_carried_32_99829b99495fcfe3488a9a86b549ece5_3) (Utilities|Struct|BreakClothParamsStruct _cloth_params_local_6))
  (bind (_linearvelocityscale_7_9cbe863e417fc965c07809a2fe55adfd_10 _angularvelocityscale_6_6a3dfe4a481ad4d7cb18d79f95cc88e8_10 _fictitiousangularscale_10_c5f4d16f48d0bedc3f0fa38b2648511b_10 _gravityoverride_15_9c3cf2b14934a1ac917c4c99f03afc6e_10 _dragcoefficient_23_fe68f666411ff9f29ba84cb4be726a52_10 _liftcoefficient_24_bb285d994a5b3aff167f4793d7eca837_10 _jointname_27_c74f3b2245ea319de82a629ff3c17cca_9) (Utilities|Struct|BreakClothSettingStruct _dbno_30_2d70d4004d0e9b117f6ce9a09c362b74_3))
  (bind (_linearvelocityscale_7_9cbe863e417fc965c07809a2fe55adfd_11 _angularvelocityscale_6_6a3dfe4a481ad4d7cb18d79f95cc88e8_11 _fictitiousangularscale_10_c5f4d16f48d0bedc3f0fa38b2648511b_11 _gravityoverride_15_9c3cf2b14934a1ac917c4c99f03afc6e_11 _dragcoefficient_23_fe68f666411ff9f29ba84cb4be726a52_11 _liftcoefficient_24_bb285d994a5b3aff167f4793d7eca837_11 _jointname_27_c74f3b2245ea319de82a629ff3c17cca_10) (Utilities|Struct|BreakClothSettingStruct _dbno_carried_32_99829b99495fcfe3488a9a86b549ece5_3))
  (bind _windgustsparamslocal_2 (LocalVariables|GetWindGustsParamsLocal))
  (bind (_skydive_goingup_28_c026bda74bfbaba6672dfe8ffdaafc1b_2 _skydive_floatingdown_29_2cbf3d3f421785f8994502931454a81f_2 _skydive_divingdown_30_ec7bec4d406a8a98e60d7b9374c5e457_2 _skydive_parachuting_31_a88c4a0a407f497561d31287cb20250f_2 _onground_standing_32_267db1464d0d3a29d99e94adfcb4afec_2 _onground_moving_33_807a66b64d7b3b2fcf080b833b8a6254_2 _onground_crouch_moving_34_ce0b8bfb4e3417c0f790f7ae38dd7735_2 _emoteormelee_35_cfcf66b5492987cbd638ec98eb34d572_2 _swimming_36_33ce87ce4d4f712b1384e29115845962_2 _front_end_37_c7a4562a48e9bc9b4cd34d963cfc6490_2 _dbno_41_0c046bf1406ab4b752db3a99949fe3a3_2 _dbno_carried_42_3e9fadc247359fee88c7b4a3453449e6_2) (Utilities|Struct|BreakWindGustParamsStruct _windgustsparamslocal_2))
  (bind _windgustsparamslocal_3 (LocalVariables|GetWindGustsParamsLocal))
  (bind (_skydive_goingup_28_c026bda74bfbaba6672dfe8ffdaafc1b_3 _skydive_floatingdown_29_2cbf3d3f421785f8994502931454a81f_3 _skydive_divingdown_30_ec7bec4d406a8a98e60d7b9374c5e457_3 _skydive_parachuting_31_a88c4a0a407f497561d31287cb20250f_3 _onground_standing_32_267db1464d0d3a29d99e94adfcb4afec_3 _onground_moving_33_807a66b64d7b3b2fcf080b833b8a6254_3 _onground_crouch_moving_34_ce0b8bfb4e3417c0f790f7ae38dd7735_3 _emoteormelee_35_cfcf66b5492987cbd638ec98eb34d572_3 _swimming_36_33ce87ce4d4f712b1384e29115845962_3 _front_end_37_c7a4562a48e9bc9b4cd34d963cfc6490_3 _dbno_41_0c046bf1406ab4b752db3a99949fe3a3_3 _dbno_carried_42_3e9fadc247359fee88c7b4a3453449e6_3 _onground_falling_46_2cc5cd2149d12b7c53d91e8d6aa90c03_2 _ziplining_47_b8bbea804f8c515ea80fe7897bb99cf6_2 _skydive_surfing_50_2b64886e4e6f0e8cfc180e81fdf37f62_2) (Utilities|Struct|BreakWindGustParamsStruct _windgustsparamslocal_3))
  (bind _windgustsparamslocal_4 (LocalVariables|GetWindGustsParamsLocal))
  (bind (_skydive_goingup_28_c026bda74bfbaba6672dfe8ffdaafc1b_4 _skydive_floatingdown_29_2cbf3d3f421785f8994502931454a81f_4 _skydive_divingdown_30_ec7bec4d406a8a98e60d7b9374c5e457_4 _skydive_parachuting_31_a88c4a0a407f497561d31287cb20250f_4 _onground_standing_32_267db1464d0d3a29d99e94adfcb4afec_4 _onground_moving_33_807a66b64d7b3b2fcf080b833b8a6254_4 _onground_crouch_moving_34_ce0b8bfb4e3417c0f790f7ae38dd7735_4 _emoteormelee_35_cfcf66b5492987cbd638ec98eb34d572_4 _swimming_36_33ce87ce4d4f712b1384e29115845962_4) (Utilities|Struct|BreakWindGustParamsStruct _windgustsparamslocal_4))
  (bind _cloth_params_local_7 (LocalVariables|GetCloth%20Params%20Local))
  (bind (_skydive_goingup_12_c026bda74bfbaba6672dfe8ffdaafc1b_8 _skydive_floatingdown_10_2cbf3d3f421785f8994502931454a81f_8 _skydive_divingdown_11_ec7bec4d406a8a98e60d7b9374c5e457_8 _skydive_parachuting_13_a88c4a0a407f497561d31287cb20250f_8 _onground_standing_15_267db1464d0d3a29d99e94adfcb4afec_8 _onground_moving_16_807a66b64d7b3b2fcf080b833b8a6254_8 _onground_crouch_moving_27_ce0b8bfb4e3417c0f790f7ae38dd7735_6 _emoteormelee_25_cfcf66b5492987cbd638ec98eb34d572_6 _swimming_3_33ce87ce4d4f712b1384e29115845962_6 _front_end_20_c7a4562a48e9bc9b4cd34d963cfc6490_5 _dbno_30_2d70d4004d0e9b117f6ce9a09c362b74_4 _dbno_carried_32_99829b99495fcfe3488a9a86b549ece5_4 _onground_falling_36_cdaa0f4d476157e4137548bdce8776f2_3 _ziplining_39_bccdb7e1431a4c174f1040abc81fd8a8_3 _skydive_surfing_42_408e65ff484aa54111e1ed8a69d8240d_3 _riding_standing_46_952a2a43408c37c44e4cbfae3f3001b0_2 _riding_moving_47_ee7a01794460b9753e5076a5a3b3080a_2 _riding_moving_targeting_52_2f57fd9e477c515d70b3da9d75453357_2 _tactical_sprint_51_5b2273a1499d34bdec52e6a8e3f274d4_2 _rail_grinding_55_deeb00a04db992df5eeb40bd6518ea4e) (Utilities|Struct|BreakClothParamsStruct _cloth_params_local_7))
  (bind _windgustsparamslocal_5 (LocalVariables|GetWindGustsParamsLocal))
  (bind (_skydive_goingup_28_c026bda74bfbaba6672dfe8ffdaafc1b_5 _skydive_floatingdown_29_2cbf3d3f421785f8994502931454a81f_5 _skydive_divingdown_30_ec7bec4d406a8a98e60d7b9374c5e457_5 _skydive_parachuting_31_a88c4a0a407f497561d31287cb20250f_5 _onground_standing_32_267db1464d0d3a29d99e94adfcb4afec_5 _onground_moving_33_807a66b64d7b3b2fcf080b833b8a6254_5 _onground_crouch_moving_34_ce0b8bfb4e3417c0f790f7ae38dd7735_5 _emoteormelee_35_cfcf66b5492987cbd638ec98eb34d572_5 _swimming_36_33ce87ce4d4f712b1384e29115845962_5 _front_end_37_c7a4562a48e9bc9b4cd34d963cfc6490_4 _dbno_41_0c046bf1406ab4b752db3a99949fe3a3_4 _dbno_carried_42_3e9fadc247359fee88c7b4a3453449e6_4 _onground_falling_46_2cc5cd2149d12b7c53d91e8d6aa90c03_3 _ziplining_47_b8bbea804f8c515ea80fe7897bb99cf6_3 _skydive_surfing_50_2b64886e4e6f0e8cfc180e81fdf37f62_3 _riding_standing_55_206a781845eac0a66d2bbc8981041dbd_2 _riding_moving_56_963fd3424b0d070c81120599a5aed02a_2 _riding_moving_targeting_59_93868b48485365007e3bd8bd8a527b39_2 _tactical_sprint_58_7e4bd4244308ee1199aa48acf8871e68_2 _rail_grinding_62_ff88fd5f4a9ab308e2f095b9f0657d82) (Utilities|Struct|BreakWindGustParamsStruct _windgustsparamslocal_5))
  (bind _windgustsparamslocal_6 (LocalVariables|GetWindGustsParamsLocal))
  (bind (_skydive_goingup_28_c026bda74bfbaba6672dfe8ffdaafc1b_6 _skydive_floatingdown_29_2cbf3d3f421785f8994502931454a81f_6 _skydive_divingdown_30_ec7bec4d406a8a98e60d7b9374c5e457_6 _skydive_parachuting_31_a88c4a0a407f497561d31287cb20250f_6 _onground_standing_32_267db1464d0d3a29d99e94adfcb4afec_6 _onground_moving_33_807a66b64d7b3b2fcf080b833b8a6254_6 _onground_crouch_moving_34_ce0b8bfb4e3417c0f790f7ae38dd7735_6 _emoteormelee_35_cfcf66b5492987cbd638ec98eb34d572_6 _swimming_36_33ce87ce4d4f712b1384e29115845962_6 _front_end_37_c7a4562a48e9bc9b4cd34d963cfc6490_5) (Utilities|Struct|BreakWindGustParamsStruct _windgustsparamslocal_6))
  (bind (_linearvelocityscale_7_9cbe863e417fc965c07809a2fe55adfd_12 _angularvelocityscale_6_6a3dfe4a481ad4d7cb18d79f95cc88e8_12 _fictitiousangularscale_10_c5f4d16f48d0bedc3f0fa38b2648511b_12 _gravityoverride_15_9c3cf2b14934a1ac917c4c99f03afc6e_12 _dragcoefficient_23_fe68f666411ff9f29ba84cb4be726a52_12 _liftcoefficient_24_bb285d994a5b3aff167f4793d7eca837_12 _jointname_27_c74f3b2245ea319de82a629ff3c17cca_11) (Utilities|Struct|BreakClothSettingStruct _skydive_surfing_42_408e65ff484aa54111e1ed8a69d8240d_2))
  (bind _cloth_params_local_8 (LocalVariables|GetCloth%20Params%20Local))
  (bind (_skydive_goingup_12_c026bda74bfbaba6672dfe8ffdaafc1b_9 _skydive_floatingdown_10_2cbf3d3f421785f8994502931454a81f_9 _skydive_divingdown_11_ec7bec4d406a8a98e60d7b9374c5e457_9 _skydive_parachuting_13_a88c4a0a407f497561d31287cb20250f_9 _onground_standing_15_267db1464d0d3a29d99e94adfcb4afec_9 _onground_moving_16_807a66b64d7b3b2fcf080b833b8a6254_9 _onground_crouch_moving_27_ce0b8bfb4e3417c0f790f7ae38dd7735_7 _emoteormelee_25_cfcf66b5492987cbd638ec98eb34d572_7 _swimming_3_33ce87ce4d4f712b1384e29115845962_7 _front_end_20_c7a4562a48e9bc9b4cd34d963cfc6490_6 _dbno_30_2d70d4004d0e9b117f6ce9a09c362b74_5 _dbno_carried_32_99829b99495fcfe3488a9a86b549ece5_5 _onground_falling_36_cdaa0f4d476157e4137548bdce8776f2_4 _ziplining_39_bccdb7e1431a4c174f1040abc81fd8a8_4 _skydive_surfing_42_408e65ff484aa54111e1ed8a69d8240d_4 _riding_standing_46_952a2a43408c37c44e4cbfae3f3001b0_3 _riding_moving_47_ee7a01794460b9753e5076a5a3b3080a_3 _riding_moving_targeting_52_2f57fd9e477c515d70b3da9d75453357_3) (Utilities|Struct|BreakClothParamsStruct _cloth_params_local_8))
  (bind _windgustsparamslocal_7 (LocalVariables|GetWindGustsParamsLocal))
  (bind (_skydive_goingup_28_c026bda74bfbaba6672dfe8ffdaafc1b_7 _skydive_floatingdown_29_2cbf3d3f421785f8994502931454a81f_7 _skydive_divingdown_30_ec7bec4d406a8a98e60d7b9374c5e457_7 _skydive_parachuting_31_a88c4a0a407f497561d31287cb20250f_7 _onground_standing_32_267db1464d0d3a29d99e94adfcb4afec_7 _onground_moving_33_807a66b64d7b3b2fcf080b833b8a6254_7 _onground_crouch_moving_34_ce0b8bfb4e3417c0f790f7ae38dd7735_7 _emoteormelee_35_cfcf66b5492987cbd638ec98eb34d572_7 _swimming_36_33ce87ce4d4f712b1384e29115845962_7 _front_end_37_c7a4562a48e9bc9b4cd34d963cfc6490_6 _dbno_41_0c046bf1406ab4b752db3a99949fe3a3_5 _dbno_carried_42_3e9fadc247359fee88c7b4a3453449e6_5 _onground_falling_46_2cc5cd2149d12b7c53d91e8d6aa90c03_4 _ziplining_47_b8bbea804f8c515ea80fe7897bb99cf6_4 _skydive_surfing_50_2b64886e4e6f0e8cfc180e81fdf37f62_4 _riding_standing_55_206a781845eac0a66d2bbc8981041dbd_3 _riding_moving_56_963fd3424b0d070c81120599a5aed02a_3 _riding_moving_targeting_59_93868b48485365007e3bd8bd8a527b39_3) (Utilities|Struct|BreakWindGustParamsStruct _windgustsparamslocal_7))
  (bind (_linearvelocityscale_7_9cbe863e417fc965c07809a2fe55adfd_13 _angularvelocityscale_6_6a3dfe4a481ad4d7cb18d79f95cc88e8_13 _fictitiousangularscale_10_c5f4d16f48d0bedc3f0fa38b2648511b_13 _gravityoverride_15_9c3cf2b14934a1ac917c4c99f03afc6e_13 _dragcoefficient_23_fe68f666411ff9f29ba84cb4be726a52_13 _liftcoefficient_24_bb285d994a5b3aff167f4793d7eca837_13 _jointname_27_c74f3b2245ea319de82a629ff3c17cca_12) (Utilities|Struct|BreakClothSettingStruct _riding_standing_46_952a2a43408c37c44e4cbfae3f3001b0_3))
  (bind _cloth_core_settings_local (LocalVariables|GetCloth%20Core%20Settings%20Local))
  (bind (_edgestiffness_28_9a24da0842e21d5736699499f1de1171 _bendingstiffness_29_6cd0864b4078d8378c061197e9b1b094 _areastiffness_30_0d981b4e4039d5a0c6a4c78d6f09d510 _tetherstiffness_24_a5c4b017498d3154d76cecab9ae7bdf2 _collisionthickness_14_dacf320b4c03c6bd248a8e9b7b867fc7 _frictioncoefficient_15_e5761ce84e8e52eebd663b9dc3a0906f _selfcollisionthickness_16_98974b1c4b5abe4ea1cff7a791319392 _dampingcoefficient_17_a343d5a54d704f5b249c8e907da80101 _localdampingcoefficient_33_993afaea430eebf424a7d0b8bb57cfb5 _animdrivestiffness_27_e54c28b348b4c35a1d0817afa9ff0423 _animdrivedamping_26_f9edbf5240e75a4905afa1a8240922c8 _numberiterations_21_8e04a8e94b2040c7dbf93194733ac36a _numbersubsteps_23_5108d5fa40140a1aad38459100b14261) (Utilities|Struct|BreakClothCoreSettingStruct _cloth_core_settings_local))
  (bind (_linearvelocityscale_7_9cbe863e417fc965c07809a2fe55adfd_14 _angularvelocityscale_6_6a3dfe4a481ad4d7cb18d79f95cc88e8_14 _fictitiousangularscale_10_c5f4d16f48d0bedc3f0fa38b2648511b_14 _gravityoverride_15_9c3cf2b14934a1ac917c4c99f03afc6e_14 _dragcoefficient_23_fe68f666411ff9f29ba84cb4be726a52_14 _liftcoefficient_24_bb285d994a5b3aff167f4793d7eca837_14 _jointname_27_c74f3b2245ea319de82a629ff3c17cca_13) (Utilities|Struct|BreakClothSettingStruct _riding_moving_47_ee7a01794460b9753e5076a5a3b3080a_3))
  (bind (_linearvelocityscale_7_9cbe863e417fc965c07809a2fe55adfd_15 _angularvelocityscale_6_6a3dfe4a481ad4d7cb18d79f95cc88e8_15 _fictitiousangularscale_10_c5f4d16f48d0bedc3f0fa38b2648511b_15 _gravityoverride_15_9c3cf2b14934a1ac917c4c99f03afc6e_15 _dragcoefficient_23_fe68f666411ff9f29ba84cb4be726a52_15 _liftcoefficient_24_bb285d994a5b3aff167f4793d7eca837_15 _jointname_27_c74f3b2245ea319de82a629ff3c17cca_14) (Utilities|Struct|BreakClothSettingStruct _riding_moving_targeting_52_2f57fd9e477c515d70b3da9d75453357_3))
  (bind (_linearvelocityscale_7_9cbe863e417fc965c07809a2fe55adfd_16 _angularvelocityscale_6_6a3dfe4a481ad4d7cb18d79f95cc88e8_16 _fictitiousangularscale_10_c5f4d16f48d0bedc3f0fa38b2648511b_16 _gravityoverride_15_9c3cf2b14934a1ac917c4c99f03afc6e_16 _dragcoefficient_23_fe68f666411ff9f29ba84cb4be726a52_16 _liftcoefficient_24_bb285d994a5b3aff167f4793d7eca837_16 _jointname_27_c74f3b2245ea319de82a629ff3c17cca_15) (Utilities|Struct|BreakClothSettingStruct _tactical_sprint_51_5b2273a1499d34bdec52e6a8e3f274d4))
  (bind _cloth_params_local_9 (LocalVariables|GetCloth%20Params%20Local))
  (bind (_skydive_goingup_12_c026bda74bfbaba6672dfe8ffdaafc1b_10 _skydive_floatingdown_10_2cbf3d3f421785f8994502931454a81f_10 _skydive_divingdown_11_ec7bec4d406a8a98e60d7b9374c5e457_10 _skydive_parachuting_13_a88c4a0a407f497561d31287cb20250f_10 _onground_standing_15_267db1464d0d3a29d99e94adfcb4afec_10 _onground_moving_16_807a66b64d7b3b2fcf080b833b8a6254_10 _onground_crouch_moving_27_ce0b8bfb4e3417c0f790f7ae38dd7735_8 _emoteormelee_25_cfcf66b5492987cbd638ec98eb34d572_8 _swimming_3_33ce87ce4d4f712b1384e29115845962_8 _front_end_20_c7a4562a48e9bc9b4cd34d963cfc6490_7 _dbno_30_2d70d4004d0e9b117f6ce9a09c362b74_6 _dbno_carried_32_99829b99495fcfe3488a9a86b549ece5_6 _onground_falling_36_cdaa0f4d476157e4137548bdce8776f2_5 _ziplining_39_bccdb7e1431a4c174f1040abc81fd8a8_5 _skydive_surfing_42_408e65ff484aa54111e1ed8a69d8240d_5 _riding_standing_46_952a2a43408c37c44e4cbfae3f3001b0_4 _riding_moving_47_ee7a01794460b9753e5076a5a3b3080a_4 _riding_moving_targeting_52_2f57fd9e477c515d70b3da9d75453357_4 _tactical_sprint_51_5b2273a1499d34bdec52e6a8e3f274d4_3 _rail_grinding_55_deeb00a04db992df5eeb40bd6518ea4e_2 _dbno_tumbling_61_cfa82d294539fa5a621a62bc2e7a8cd9 _dbno_rollfwd_62_8493f23840908e76a35cabb794befe86 _dbno_rollbwd_63_2cf54c094abe5911a4dd0587754bd23f _dbno_impulse_64_dc5f9eeb407dd4ff2de0ac812dd29451) (Utilities|Struct|BreakClothParamsStruct _cloth_params_local_9))
  (bind (_linearvelocityscale_7_9cbe863e417fc965c07809a2fe55adfd_17 _angularvelocityscale_6_6a3dfe4a481ad4d7cb18d79f95cc88e8_17 _fictitiousangularscale_10_c5f4d16f48d0bedc3f0fa38b2648511b_17 _gravityoverride_15_9c3cf2b14934a1ac917c4c99f03afc6e_17 _dragcoefficient_23_fe68f666411ff9f29ba84cb4be726a52_17 _liftcoefficient_24_bb285d994a5b3aff167f4793d7eca837_17 _jointname_27_c74f3b2245ea319de82a629ff3c17cca_16) (Utilities|Struct|BreakClothSettingStruct _dbno_tumbling_61_cfa82d294539fa5a621a62bc2e7a8cd9))
  (bind (_linearvelocityscale_7_9cbe863e417fc965c07809a2fe55adfd_18 _angularvelocityscale_6_6a3dfe4a481ad4d7cb18d79f95cc88e8_18 _fictitiousangularscale_10_c5f4d16f48d0bedc3f0fa38b2648511b_18 _gravityoverride_15_9c3cf2b14934a1ac917c4c99f03afc6e_18 _dragcoefficient_23_fe68f666411ff9f29ba84cb4be726a52_18 _liftcoefficient_24_bb285d994a5b3aff167f4793d7eca837_18 _jointname_27_c74f3b2245ea319de82a629ff3c17cca_17) (Utilities|Struct|BreakClothSettingStruct _dbno_rollfwd_62_8493f23840908e76a35cabb794befe86))
  (bind (_linearvelocityscale_7_9cbe863e417fc965c07809a2fe55adfd_19 _angularvelocityscale_6_6a3dfe4a481ad4d7cb18d79f95cc88e8_19 _fictitiousangularscale_10_c5f4d16f48d0bedc3f0fa38b2648511b_19 _gravityoverride_15_9c3cf2b14934a1ac917c4c99f03afc6e_19 _dragcoefficient_23_fe68f666411ff9f29ba84cb4be726a52_19 _liftcoefficient_24_bb285d994a5b3aff167f4793d7eca837_19 _jointname_27_c74f3b2245ea319de82a629ff3c17cca_18) (Utilities|Struct|BreakClothSettingStruct _dbno_rollbwd_63_2cf54c094abe5911a4dd0587754bd23f))
  (bind (_linearvelocityscale_7_9cbe863e417fc965c07809a2fe55adfd_20 _angularvelocityscale_6_6a3dfe4a481ad4d7cb18d79f95cc88e8_20 _fictitiousangularscale_10_c5f4d16f48d0bedc3f0fa38b2648511b_20 _gravityoverride_15_9c3cf2b14934a1ac917c4c99f03afc6e_20 _dragcoefficient_23_fe68f666411ff9f29ba84cb4be726a52_20 _liftcoefficient_24_bb285d994a5b3aff167f4793d7eca837_20 _jointname_27_c74f3b2245ea319de82a629ff3c17cca_19) (Utilities|Struct|BreakClothSettingStruct _dbno_impulse_64_dc5f9eeb407dd4ff2de0ac812dd29451))
  (bind _windgustsparamslocal_8 (LocalVariables|GetWindGustsParamsLocal))
  (bind (_skydive_goingup_28_c026bda74bfbaba6672dfe8ffdaafc1b_8 _skydive_floatingdown_29_2cbf3d3f421785f8994502931454a81f_8 _skydive_divingdown_30_ec7bec4d406a8a98e60d7b9374c5e457_8 _skydive_parachuting_31_a88c4a0a407f497561d31287cb20250f_8 _onground_standing_32_267db1464d0d3a29d99e94adfcb4afec_8 _onground_moving_33_807a66b64d7b3b2fcf080b833b8a6254_8 _onground_crouch_moving_34_ce0b8bfb4e3417c0f790f7ae38dd7735_8 _emoteormelee_35_cfcf66b5492987cbd638ec98eb34d572_8 _swimming_36_33ce87ce4d4f712b1384e29115845962_8 _front_end_37_c7a4562a48e9bc9b4cd34d963cfc6490_7 _dbno_41_0c046bf1406ab4b752db3a99949fe3a3_6 _dbno_carried_42_3e9fadc247359fee88c7b4a3453449e6_6 _onground_falling_46_2cc5cd2149d12b7c53d91e8d6aa90c03_5 _ziplining_47_b8bbea804f8c515ea80fe7897bb99cf6_5 _skydive_surfing_50_2b64886e4e6f0e8cfc180e81fdf37f62_5 _riding_standing_55_206a781845eac0a66d2bbc8981041dbd_4 _riding_moving_56_963fd3424b0d070c81120599a5aed02a_4 _riding_moving_targeting_59_93868b48485365007e3bd8bd8a527b39_4 _tactical_sprint_58_7e4bd4244308ee1199aa48acf8871e68_3 _rail_grinding_62_ff88fd5f4a9ab308e2f095b9f0657d82_2 _dbno_tumbling_71_23aaa3624a4776d875f8049a17406a81 _dbno_rollfwd_72_387c1f514be9313deb6e17b3acab3fba _dbno_rollbwd_73_b4165acb4840a5d88fccf28ca56463a1 _dbno_impulse_74_3b793bd343babd1a82c40abcc2d883e4) (Utilities|Struct|BreakWindGustParamsStruct _windgustsparamslocal_8))
  (bind _characterpartaniminstance (Variables|GetCharacterPartAnimInstance))
  (bind _returnvalue (Animation|GetOwningComponent _characterpartaniminstance))
  (bind _returnvalue_2 (Clothing|GetClothingSimulationInteractor _returnvalue "None"))
  (bind _characterpartaniminstance_2 (Variables|GetCharacterPartAnimInstance))
  (bind _characterpartaniminstance_3 (Variables|GetCharacterPartAnimInstance))
  (bind _characterpartaniminstance_4 (Variables|GetCharacterPartAnimInstance))
  (bind _characterpartaniminstance_5 (Variables|GetCharacterPartAnimInstance))
  (bind _characterpartaniminstance_6 (Variables|GetCharacterPartAnimInstance))
  (bind _characterpartaniminstance_7 (Variables|GetCharacterPartAnimInstance))
  (bind _characterpartaniminstance_8 (Variables|GetCharacterPartAnimInstance))
  (bind _characterpartaniminstance_9 (Variables|GetCharacterPartAnimInstance))
  (bind _characterpartaniminstance_10 (Variables|GetCharacterPartAnimInstance))
  (bind _characterpartaniminstance_11 (Variables|GetCharacterPartAnimInstance))
  (bind _deltatime (Variables|GetDeltaTime))
  (bind _gametimeinseconds (Variables|GetGameTimeInSeconds))
  (bind _characterpartaniminstance_12 (Variables|GetCharacterPartAnimInstance))
  (bind _characterpartaniminstance_13 (Variables|GetCharacterPartAnimInstance))
  (bind _characterpartaniminstance_14 (Variables|GetCharacterPartAnimInstance))
  (bind _characterpartaniminstance_15 (Variables|GetCharacterPartAnimInstance))
  (bind _characterpartaniminstance_16 (Variables|GetCharacterPartAnimInstance))
  (bind _chaosclothinginteractor (LocalVariables|GetChaosClothingInteractor))
  (bind _characterpartaniminstance_17 (Variables|GetCharacterPartAnimInstance))
  (bind _characterpartaniminstance_18 (Variables|GetCharacterPartAnimInstance))
  (bind _characterpartaniminstance_19 (Variables|GetCharacterPartAnimInstance))
  (bind _characterpartaniminstance_20 (Variables|GetCharacterPartAnimInstance))
  (bind _chaosclothinginteractor_2 (LocalVariables|GetChaosClothingInteractor))
  ;[1 1 1] Check to see if cloth is enabled at all.  If not just break out.
  (if (Variables|/Script/FortniteGame.FortAnimInstance|GetbClothEnabled (Variables|GetCharacterPartAnimInstance))
    (LocalVariables|SetCloth%20Params%20Local Cloth%20Paramaters)
    (LocalVariables|SetCloth%20Core%20Settings%20Local Cloth%20Core%20Settings)
    (LocalVariables|SetWindGustLocalSettings%20Local WindGustLocalSettings)
    (LocalVariables|SetWindGustsParamsLocal WindGustsParams)
    (Utilities|FlowControl|Sequence
      (:then_0
        ;[1 1 1] Define the Chaos Clothing Interactor
        (Utilities|IsValid _returnvalue_2
          (:Is%20Valid
            (bind _aschaos_clothing_interactor (Utilities|Casting|CastToChaosClothingInteractor (ClothingSimulation|GetClothingInteractor _returnvalue_2 (Utilities|String|StringToName (Variables|GetClothingAssetName)))))
            (LocalVariables|SetChaosClothingInteractor _aschaos_clothing_interactor))))
      (:then_1
        ;[1 1 1] Detect the players fwd/bwd movement and create variables to store them
        (bind (_xmovement _ymovement) (Animation|Utilities|PlayerMovement (Variables|GetCharacterPartAnimInstance)))
        (LocalVariables|SetPlayer%20Forward%20Vector%20Local (Math|Float|MapRangeClamped _ymovement 1.0 (Variables|GetLinVelScale_FwdBwd_Clamp) 1.0 (Variables|GetLinVelScale_FwdBwd_Clamp))) ; This is just the result of the players forward and backward (Y) movement clamped using the MapRange Clamp - this will be used to scale the fwd/bwd axis later
        (LocalVariables|SetPlayer%20Forward%20Vector%20scaled%20for%20jump%20Local (Math|Float|MapRangeClamped _ymovement -1.0 1.0 (Variables|GetLinVelScale_UpDwn_Clamp) 1.0)) ; This is just the result of the players forward and backward (Y) movement clamped using the MapRange Clamp - this will be used to scale the up axis later.
      )
      (:then_2
        ;[1 1 1] State Logic - Switch order causes messy lines but this is an optimization - most common states should be at the top
        (switch name (Animation|CharacterDynamicsControl|GetCharacterStateforLayer (Variables|GetCharacterPartAnimInstance))
          (:Default
            _)
          (:OnGround_Standing
            _)
          (:OnGround_Moving
            ;[1 1 1] Check the Rotation Rate of the pawn when moving.  Sets the on ground moving fictitious velocity to the Max value if true.
            (LocalVariables|SetOnGround_Moving_AngularVel _onground_moving_16_807a66b64d7b3b2fcf080b833b8a6254_angularvelocityscale_6_6a3dfe4a481ad4d7cb18d79f95cc88e8_2)
            (LocalVariables|SetOnGround_Moving_FictitiousVel _onground_moving_16_807a66b64d7b3b2fcf080b833b8a6254_fictitiousangularscale_10_c5f4d16f48d0bedc3f0fa38b2648511b_2)
            (LocalVariables|SetCurrentState "OnGround_Moving")
            (LocalVariables|SetOnGroundMoving_DragCoefficientLocal _onground_moving_16_807a66b64d7b3b2fcf080b833b8a6254_dragcoefficient_23_fe68f666411ff9f29ba84cb4be726a52_2)
            (LocalVariables|SetOnGroundMoving_LiftCoefficientLocal _onground_moving_16_807a66b64d7b3b2fcf080b833b8a6254_liftcoefficient_24_bb285d994a5b3aff167f4793d7eca837_2)
            ;[1 1 1] Scale LinearVelocityScale
            (bind _linearvelvectorout (Default|ClothControllerHelperScaleforPlayerMovementDEPRECATED _onground_moving_16_807a66b64d7b3b2fcf080b833b8a6254_linearvelocityscale_7_9cbe863e417fc965c07809a2fe55adfd_3 (LocalVariables|GetPlayer%20Forward%20Vector%20Local) (LocalVariables|GetPlayer%20Forward%20Vector%20scaled%20for%20jump%20Local) (.x (Variables|GetClothUpAxis)) (.z (Variables|GetClothUpAxis)))) ; Depending on which axis is up in the ClothUpAxis var, this will scale the fwd/bwd axis using the PlayerForwardVectoLocal.  And it will scale the up/dwn axis using PlayerForwardVectorScaledForJumpLocal.
            (LocalVariables|SetMovementLinearVelVector%20Local _linearvelvectorout)
            ;[1 1 1] On Ground Moving
            (bind (_guststrengthwindvelocity_out _windgustlocalsettingsout) (Default|WindGustControlFunctionDEPRECATED _characterpartaniminstance_10 (Variables|GetDeltaTime) (Variables|GetGameTimeInSeconds) _onground_moving_16_807a66b64d7b3b2fcf080b833b8a6254_jointname_27_c74f3b2245ea319de82a629ff3c17cca _onground_moving_33_807a66b64d7b3b2fcf080b833b8a6254 (LocalVariables|GetWindGustLocalSettings%20Local)))
            (LocalVariables|SetGustStrength%20-%20WindVelocityLocal _guststrengthwindvelocity_out)
            (LocalVariables|SetWindGustLocalSettings%20Local _windgustlocalsettingsout)
            (bind _cloth_settings_out (Default|ClothControllerSetterFunctionDEPRECATED _characterpartaniminstance_10 (Utilities|Struct|MakeClothSettingStruct (LocalVariables|GetMovementLinearVelVector%20Local) (LocalVariables|GetOnGround_Moving_AngularVel) (LocalVariables|GetOnGround_Moving_FictitiousVel) _onground_moving_16_807a66b64d7b3b2fcf080b833b8a6254_gravityoverride_15_9c3cf2b14934a1ac917c4c99f03afc6e (LocalVariables|GetOnGroundMoving_DragCoefficientLocal) (LocalVariables|GetOnGroundMoving_LiftCoefficientLocal) _onground_moving_16_807a66b64d7b3b2fcf080b833b8a6254_jointname_27_c74f3b2245ea319de82a629ff3c17cca) 0)) ; OnGround Moving
            (LocalVariables|SetCloth%20Settings%20Local _cloth_settings_out)
            (LocalVariables|SetWindGustSettingsLocal _onground_moving_33_807a66b64d7b3b2fcf080b833b8a6254))
          (:Tactical_Sprint
            ;[1 1 1] Tactical Sprinting
            (LocalVariables|SetCurrentState "Tactical Sprinting")
            (bind (_guststrengthwindvelocity_out_2 _windgustlocalsettingsout_2) (Default|WindGustControlFunctionDEPRECATED _characterpartaniminstance_15 (Variables|GetDeltaTime) (Variables|GetGameTimeInSeconds) _jointname_27_c74f3b2245ea319de82a629ff3c17cca_15 _tactical_sprint_58_7e4bd4244308ee1199aa48acf8871e68 (LocalVariables|GetWindGustLocalSettings%20Local)))
            (LocalVariables|SetGustStrength%20-%20WindVelocityLocal _guststrengthwindvelocity_out_2)
            (LocalVariables|SetWindGustLocalSettings%20Local _windgustlocalsettingsout_2)
            (bind _cloth_settings_out_2 (Default|ClothControllerSetterFunctionDEPRECATED _characterpartaniminstance_15 _tactical_sprint_51_5b2273a1499d34bdec52e6a8e3f274d4)) ; OnGround Standing
            (LocalVariables|SetCloth%20Settings%20Local _cloth_settings_out_2)
            (LocalVariables|SetWindGustSettingsLocal _tactical_sprint_58_7e4bd4244308ee1199aa48acf8871e68))
          (:OnGround_Falling
            ;[1 1 1] OnGround Falling
            (LocalVariables|SetCurrentState "Falling")
            (bind (_guststrengthwindvelocity_out_3 _windgustlocalsettingsout_3) (Default|WindGustControlFunctionDEPRECATED _characterpartaniminstance_6 (Variables|GetDeltaTime) (Variables|GetGameTimeInSeconds) "root" _onground_falling_46_2cc5cd2149d12b7c53d91e8d6aa90c03_3 (LocalVariables|GetWindGustLocalSettings%20Local)))
            (LocalVariables|SetGustStrength%20-%20WindVelocityLocal _guststrengthwindvelocity_out_3)
            (LocalVariables|SetWindGustLocalSettings%20Local _windgustlocalsettingsout_3)
            (bind _cloth_settings_out_3 (Default|ClothControllerSetterFunctionDEPRECATED _characterpartaniminstance_6 _onground_falling_36_cdaa0f4d476157e4137548bdce8776f2_3)) ; Falling
            (LocalVariables|SetCloth%20Settings%20Local _cloth_settings_out_3)
            (LocalVariables|SetWindGustSettingsLocal _onground_falling_46_2cc5cd2149d12b7c53d91e8d6aa90c03_3))
          (:OnGround_Moving_HighRotationRate
            (LocalVariables|SetOnGround_Moving_AngularVel (Variables|GetOnGround_Moving_AngularVel_Max))
            (LocalVariables|SetOnGround_Moving_FictitiousVel (Variables|GetOnGround_Moving_FictitiousVel_Max))
            (LocalVariables|SetCurrentState "OnGround_Moving_HighRotationRate"))
          (:Ziplining
            ;[1 1 1] Ziplining
            (LocalVariables|SetCurrentState "Ziplining")
            (bind (_guststrengthwindvelocity_out_4 _windgustlocalsettingsout_4) (Default|WindGustControlFunctionDEPRECATED _characterpartaniminstance_7 (Variables|GetDeltaTime) (Variables|GetGameTimeInSeconds) "root" _ziplining_47_b8bbea804f8c515ea80fe7897bb99cf6_3 (LocalVariables|GetWindGustLocalSettings%20Local)))
            (LocalVariables|SetGustStrength%20-%20WindVelocityLocal _guststrengthwindvelocity_out_4)
            (LocalVariables|SetWindGustLocalSettings%20Local _windgustlocalsettingsout_4)
            (bind _cloth_settings_out_4 (Default|ClothControllerSetterFunctionDEPRECATED _characterpartaniminstance_7 _ziplining_39_bccdb7e1431a4c174f1040abc81fd8a8_3)) ; Ziplining
            (LocalVariables|SetCloth%20Settings%20Local _cloth_settings_out_4)
            (LocalVariables|SetWindGustSettingsLocal _ziplining_47_b8bbea804f8c515ea80fe7897bb99cf6_3))
          (:Swimming
            ;[1 1 1] Swimming
            (LocalVariables|SetCurrentState "Swimming")
            (bind (_guststrengthwindvelocity_out_5 _windgustlocalsettingsout_5) (Default|WindGustControlFunctionDEPRECATED _characterpartaniminstance_3 (Variables|GetDeltaTime) (Variables|GetGameTimeInSeconds) _jointname_27_c74f3b2245ea319de82a629ff3c17cca _swimming_36_33ce87ce4d4f712b1384e29115845962_4 (LocalVariables|GetWindGustLocalSettings%20Local)))
            (LocalVariables|SetGustStrength%20-%20WindVelocityLocal _guststrengthwindvelocity_out_5)
            (LocalVariables|SetWindGustLocalSettings%20Local _windgustlocalsettingsout_5)
            (bind _cloth_settings_out_5 (Default|ClothControllerSetterFunctionDEPRECATED _characterpartaniminstance_3 _swimming_3_33ce87ce4d4f712b1384e29115845962_2)) ; Swimming
            (LocalVariables|SetCloth%20Settings%20Local _cloth_settings_out_5)
            (LocalVariables|SetWindGustSettingsLocal _swimming_36_33ce87ce4d4f712b1384e29115845962_4))
          (:Rail_Grinding
            ;[1 1 1] Rail Grinding
            (LocalVariables|SetCurrentState "Rail Grinding")
            (bind (_guststrengthwindvelocity_out_6 _windgustlocalsettingsout_6) (Default|WindGustControlFunctionDEPRECATED _characterpartaniminstance_16 (Variables|GetDeltaTime) (Variables|GetGameTimeInSeconds) "root" _rail_grinding_62_ff88fd5f4a9ab308e2f095b9f0657d82 (LocalVariables|GetWindGustLocalSettings%20Local)))
            (LocalVariables|SetGustStrength%20-%20WindVelocityLocal _guststrengthwindvelocity_out_6)
            (LocalVariables|SetWindGustLocalSettings%20Local _windgustlocalsettingsout_6)
            (bind _cloth_settings_out_6 (Default|ClothControllerSetterFunctionDEPRECATED _characterpartaniminstance_16 _rail_grinding_55_deeb00a04db992df5eeb40bd6518ea4e)) ; Ziplining
            (LocalVariables|SetCloth%20Settings%20Local _cloth_settings_out_6)
            (LocalVariables|SetWindGustSettingsLocal _rail_grinding_62_ff88fd5f4a9ab308e2f095b9f0657d82))
          (:EmoteOrMelee
            ;[1 1 1] Emote or Melee
            (LocalVariables|SetCurrentState "MeleeOrEmote")
            (bind (_guststrengthwindvelocity_out_7 _windgustlocalsettingsout_7) (Default|WindGustControlFunctionDEPRECATED _characterpartaniminstance_4 (Variables|GetDeltaTime) (Variables|GetGameTimeInSeconds) _jointname_27_c74f3b2245ea319de82a629ff3c17cca_7 _emoteormelee_35_cfcf66b5492987cbd638ec98eb34d572_6 (LocalVariables|GetWindGustLocalSettings%20Local)))
            (LocalVariables|SetGustStrength%20-%20WindVelocityLocal _guststrengthwindvelocity_out_7)
            (LocalVariables|SetWindGustLocalSettings%20Local _windgustlocalsettingsout_7)
            (bind _cloth_settings_out_7 (Default|ClothControllerSetterFunctionDEPRECATED _characterpartaniminstance_4 _emoteormelee_25_cfcf66b5492987cbd638ec98eb34d572_4)) ; Melee or Emote
            (LocalVariables|SetCloth%20Settings%20Local _cloth_settings_out_7)
            (LocalVariables|SetWindGustSettingsLocal _emoteormelee_35_cfcf66b5492987cbd638ec98eb34d572_6))
          (:Skydive_GoingUp
            ;[1 1 1] SKYDIVING
            (LocalVariables|SetCurrentState "SkyDiving_GoingUp")
            (bind (_guststrengthwindvelocity_out_8 _windgustlocalsettingsout_8) (Default|WindGustControlFunctionDEPRECATED _characterpartaniminstance_2 _deltatime _gametimeinseconds _jointname_27_c74f3b2245ea319de82a629ff3c17cca_2 _skydive_goingup_28_c026bda74bfbaba6672dfe8ffdaafc1b_3 (LocalVariables|GetWindGustLocalSettings%20Local))) ; Skydiving Going Up
            (LocalVariables|SetGustStrength%20-%20WindVelocityLocal _guststrengthwindvelocity_out_8)
            (LocalVariables|SetWindGustLocalSettings%20Local _windgustlocalsettingsout_8)
            (bind _cloth_settings_out_8 (Default|ClothControllerSetterFunctionDEPRECATED _characterpartaniminstance_2 _skydive_goingup_12_c026bda74bfbaba6672dfe8ffdaafc1b_4)) ; Skydiving Going Up
            (LocalVariables|SetCloth%20Settings%20Local _cloth_settings_out_8)
            (LocalVariables|SetWindGustSettingsLocal _skydive_goingup_28_c026bda74bfbaba6672dfe8ffdaafc1b_3))
          (:Skydive_DivingDown
            (LocalVariables|SetCurrentState "SkyDiving_DivingDown")
            (bind (_guststrengthwindvelocity_out_9 _windgustlocalsettingsout_9) (Default|WindGustControlFunctionDEPRECATED _characterpartaniminstance_2 _deltatime _gametimeinseconds _jointname_27_c74f3b2245ea319de82a629ff3c17cca_4 _skydive_divingdown_30_ec7bec4d406a8a98e60d7b9374c5e457_3 (LocalVariables|GetWindGustLocalSettings%20Local))) ; Skydiving Diving Down
            (LocalVariables|SetGustStrength%20-%20WindVelocityLocal _guststrengthwindvelocity_out_9)
            (LocalVariables|SetWindGustLocalSettings%20Local _windgustlocalsettingsout_9)
            (bind _cloth_settings_out_9 (Default|ClothControllerSetterFunctionDEPRECATED _characterpartaniminstance_2 _skydive_divingdown_11_ec7bec4d406a8a98e60d7b9374c5e457_4)) ; Skydiving Diving Down
            (LocalVariables|SetCloth%20Settings%20Local _cloth_settings_out_9)
            (LocalVariables|SetWindGustSettingsLocal _skydive_divingdown_30_ec7bec4d406a8a98e60d7b9374c5e457_3))
          (:Skydive_FloatingDown
            (LocalVariables|SetCurrentState "SkyDiving_FloatingDown")
            (bind (_guststrengthwindvelocity_out_10 _windgustlocalsettingsout_10) (Default|WindGustControlFunctionDEPRECATED _characterpartaniminstance_2 _deltatime _gametimeinseconds _jointname_27_c74f3b2245ea319de82a629ff3c17cca_3 _skydive_floatingdown_29_2cbf3d3f421785f8994502931454a81f_3 (LocalVariables|GetWindGustLocalSettings%20Local))) ; Skydiving Floating Down
            (LocalVariables|SetGustStrength%20-%20WindVelocityLocal _guststrengthwindvelocity_out_10)
            (LocalVariables|SetWindGustLocalSettings%20Local _windgustlocalsettingsout_10)
            (bind _cloth_settings_out_10 (Default|ClothControllerSetterFunctionDEPRECATED _characterpartaniminstance_2 _skydive_floatingdown_10_2cbf3d3f421785f8994502931454a81f_4)) ; Skydiving Floating Down
            (LocalVariables|SetCloth%20Settings%20Local _cloth_settings_out_10)
            (LocalVariables|SetWindGustSettingsLocal _skydive_floatingdown_29_2cbf3d3f421785f8994502931454a81f_3))
          (:Skydive_Parachuting
            (LocalVariables|SetCurrentState "SkyDiving_Parachuting")
            (bind (_guststrengthwindvelocity_out_11 _windgustlocalsettingsout_11) (Default|WindGustControlFunctionDEPRECATED _characterpartaniminstance_2 _deltatime _gametimeinseconds _jointname_27_c74f3b2245ea319de82a629ff3c17cca_5 _skydive_parachuting_31_a88c4a0a407f497561d31287cb20250f_3 (LocalVariables|GetWindGustLocalSettings%20Local))) ; Skydiving Parachuting
            (LocalVariables|SetGustStrength%20-%20WindVelocityLocal _guststrengthwindvelocity_out_11)
            (LocalVariables|SetWindGustLocalSettings%20Local _windgustlocalsettingsout_11)
            (bind _cloth_settings_out_11 (Default|ClothControllerSetterFunctionDEPRECATED _characterpartaniminstance_2 _skydive_parachuting_13_a88c4a0a407f497561d31287cb20250f_4)) ; Skydiving Parachuting
            (LocalVariables|SetCloth%20Settings%20Local _cloth_settings_out_11)
            (LocalVariables|SetWindGustSettingsLocal _skydive_parachuting_31_a88c4a0a407f497561d31287cb20250f_3))
          (:Skydive_ParachutingWithPBA
            _)
          (:Skydive_Umbrella
            _)
          (:Riding_Moving_Targeting
            ;[1 1 1] Riding Moving Targeting
            (LocalVariables|SetCurrentState "Riding_Moving_Targeting")
            (bind (_guststrengthwindvelocity_out_12 _windgustlocalsettingsout_12) (Default|WindGustControlFunctionDEPRECATED _characterpartaniminstance_14 (Variables|GetDeltaTime) (Variables|GetGameTimeInSeconds) _jointname_27_c74f3b2245ea319de82a629ff3c17cca_14 _riding_moving_targeting_59_93868b48485365007e3bd8bd8a527b39_3 (LocalVariables|GetWindGustLocalSettings%20Local)))
            (LocalVariables|SetGustStrength%20-%20WindVelocityLocal _guststrengthwindvelocity_out_12)
            (LocalVariables|SetWindGustLocalSettings%20Local _windgustlocalsettingsout_12)
            (bind _cloth_settings_out_12 (Default|ClothControllerSetterFunctionDEPRECATED _characterpartaniminstance_14 _riding_moving_targeting_52_2f57fd9e477c515d70b3da9d75453357_3)) ; OnGround Standing
            (LocalVariables|SetCloth%20Settings%20Local _cloth_settings_out_12)
            (LocalVariables|SetWindGustSettingsLocal _riding_moving_targeting_59_93868b48485365007e3bd8bd8a527b39_3))
          (:Riding_Standing
            ;[1 1 1] Riding Standing
            (LocalVariables|SetCurrentState "Riding_Standing")
            (bind (_guststrengthwindvelocity_out_13 _windgustlocalsettingsout_13) (Default|WindGustControlFunctionDEPRECATED _characterpartaniminstance_12 (Variables|GetDeltaTime) (Variables|GetGameTimeInSeconds) _jointname_27_c74f3b2245ea319de82a629ff3c17cca_12 _riding_standing_55_206a781845eac0a66d2bbc8981041dbd_3 (LocalVariables|GetWindGustLocalSettings%20Local)))
            (LocalVariables|SetGustStrength%20-%20WindVelocityLocal _guststrengthwindvelocity_out_13)
            (LocalVariables|SetWindGustLocalSettings%20Local _windgustlocalsettingsout_13)
            (bind _cloth_settings_out_13 (Default|ClothControllerSetterFunctionDEPRECATED _characterpartaniminstance_12 _riding_standing_46_952a2a43408c37c44e4cbfae3f3001b0_3)) ; OnGround Standing
            (LocalVariables|SetCloth%20Settings%20Local _cloth_settings_out_13)
            (LocalVariables|SetWindGustSettingsLocal _riding_standing_55_206a781845eac0a66d2bbc8981041dbd_3))
          (:Riding_Moving
            ;[1 1 1] Riding Moving
            (LocalVariables|SetCurrentState "Riding_Moving")
            (bind (_guststrengthwindvelocity_out_14 _windgustlocalsettingsout_14) (Default|WindGustControlFunctionDEPRECATED _characterpartaniminstance_13 (Variables|GetDeltaTime) (Variables|GetGameTimeInSeconds) _jointname_27_c74f3b2245ea319de82a629ff3c17cca_13 _riding_moving_56_963fd3424b0d070c81120599a5aed02a_3 (LocalVariables|GetWindGustLocalSettings%20Local)))
            (LocalVariables|SetGustStrength%20-%20WindVelocityLocal _guststrengthwindvelocity_out_14)
            (LocalVariables|SetWindGustLocalSettings%20Local _windgustlocalsettingsout_14)
            (bind _cloth_settings_out_14 (Default|ClothControllerSetterFunctionDEPRECATED _characterpartaniminstance_13 _riding_moving_47_ee7a01794460b9753e5076a5a3b3080a_3)) ; OnGround Standing
            (LocalVariables|SetCloth%20Settings%20Local _cloth_settings_out_14)
            (LocalVariables|SetWindGustSettingsLocal _riding_moving_56_963fd3424b0d070c81120599a5aed02a_3))
          (:DBNO
            ;[1 1 1] DBNO
            (LocalVariables|SetCurrentState "DBNO")
            (bind (_guststrengthwindvelocity_out_15 _windgustlocalsettingsout_15) (Default|WindGustControlFunctionDEPRECATED :CharacterPartAnimInstance _characterpartaniminstance_9 :DeltaTime (Variables|GetDeltaTime) :GameTimeInSeconds (Variables|GetGameTimeInSeconds) :JointName _jointname_27_c74f3b2245ea319de82a629ff3c17cca_9 :WindGustLocalSettings (LocalVariables|GetWindGustLocalSettings%20Local)))
            (LocalVariables|SetGustStrength%20-%20WindVelocityLocal _guststrengthwindvelocity_out_15)
            (LocalVariables|SetWindGustLocalSettings%20Local _windgustlocalsettingsout_15)
            (bind _cloth_settings_out_15 (Default|ClothControllerSetterFunctionDEPRECATED _characterpartaniminstance_9 _dbno_30_2d70d4004d0e9b117f6ce9a09c362b74_3)) ; DBNO
            (LocalVariables|SetCloth%20Settings%20Local _cloth_settings_out_15)
            (LocalVariables|SetWindGustSettingsLocal _dbno_41_0c046bf1406ab4b752db3a99949fe3a3_2))
          (:DBNO_Carried
            ;[1 1 1] DBNO Carried
            (LocalVariables|SetCurrentState "DBNO Carried")
            (bind (_guststrengthwindvelocity_out_16 _windgustlocalsettingsout_16) (Default|WindGustControlFunctionDEPRECATED :CharacterPartAnimInstance _characterpartaniminstance_8 :DeltaTime (Variables|GetDeltaTime) :GameTimeInSeconds (Variables|GetGameTimeInSeconds) :JointName _jointname_27_c74f3b2245ea319de82a629ff3c17cca_10 :WindGustLocalSettings (LocalVariables|GetWindGustLocalSettings%20Local)))
            (LocalVariables|SetGustStrength%20-%20WindVelocityLocal _guststrengthwindvelocity_out_16)
            (LocalVariables|SetWindGustLocalSettings%20Local _windgustlocalsettingsout_16)
            (bind _cloth_settings_out_16 (Default|ClothControllerSetterFunctionDEPRECATED _characterpartaniminstance_8 _dbno_carried_32_99829b99495fcfe3488a9a86b549ece5_3)) ; DBNO Carried
            (LocalVariables|SetCloth%20Settings%20Local _cloth_settings_out_16)
            (LocalVariables|SetWindGustSettingsLocal _dbno_carried_42_3e9fadc247359fee88c7b4a3453449e6_2))
          (:DBNO_Tumbling
            ;[1 1 1] DBNO Tumbling
            (LocalVariables|SetCurrentState "DBNO Tumbling")
            (bind (_guststrengthwindvelocity_out_17 _windgustlocalsettingsout_17) (Default|WindGustControlFunctionDEPRECATED :CharacterPartAnimInstance _characterpartaniminstance_17 :DeltaTime (Variables|GetDeltaTime) :GameTimeInSeconds (Variables|GetGameTimeInSeconds) :JointName _jointname_27_c74f3b2245ea319de82a629ff3c17cca_16 :WindGustLocalSettings (LocalVariables|GetWindGustLocalSettings%20Local)))
            (LocalVariables|SetGustStrength%20-%20WindVelocityLocal _guststrengthwindvelocity_out_17)
            (LocalVariables|SetWindGustLocalSettings%20Local _windgustlocalsettingsout_17)
            (bind _cloth_settings_out_17 (Default|ClothControllerSetterFunctionDEPRECATED _characterpartaniminstance_17 _dbno_tumbling_61_cfa82d294539fa5a621a62bc2e7a8cd9)) ; DBNO Tumbling
            (LocalVariables|SetCloth%20Settings%20Local _cloth_settings_out_17)
            (LocalVariables|SetWindGustSettingsLocal _dbno_tumbling_71_23aaa3624a4776d875f8049a17406a81))
          (:DBNO_RollFwd
            ;[1 1 1] DBNO Roll Fwd
            (LocalVariables|SetCurrentState "DBNO Roll Fwd")
            (bind (_guststrengthwindvelocity_out_18 _windgustlocalsettingsout_18) (Default|WindGustControlFunctionDEPRECATED :CharacterPartAnimInstance _characterpartaniminstance_18 :DeltaTime (Variables|GetDeltaTime) :GameTimeInSeconds (Variables|GetGameTimeInSeconds) :JointName _jointname_27_c74f3b2245ea319de82a629ff3c17cca_17 :WindGustLocalSettings (LocalVariables|GetWindGustLocalSettings%20Local)))
            (LocalVariables|SetGustStrength%20-%20WindVelocityLocal _guststrengthwindvelocity_out_18)
            (LocalVariables|SetWindGustLocalSettings%20Local _windgustlocalsettingsout_18)
            (bind _cloth_settings_out_18 (Default|ClothControllerSetterFunctionDEPRECATED _characterpartaniminstance_18 _dbno_rollfwd_62_8493f23840908e76a35cabb794befe86)) ; DBNO Roll Fwd
            (LocalVariables|SetCloth%20Settings%20Local _cloth_settings_out_18)
            (LocalVariables|SetWindGustSettingsLocal _dbno_rollfwd_72_387c1f514be9313deb6e17b3acab3fba))
          (:DBNO_RollBwd
            ;[1 1 1] DBNO Roll Bwd
            (LocalVariables|SetCurrentState "DBNO Roll Bwd")
            (bind (_guststrengthwindvelocity_out_19 _windgustlocalsettingsout_19) (Default|WindGustControlFunctionDEPRECATED :CharacterPartAnimInstance _characterpartaniminstance_19 :DeltaTime (Variables|GetDeltaTime) :GameTimeInSeconds (Variables|GetGameTimeInSeconds) :JointName _jointname_27_c74f3b2245ea319de82a629ff3c17cca_18 :WindGustLocalSettings (LocalVariables|GetWindGustLocalSettings%20Local)))
            (LocalVariables|SetGustStrength%20-%20WindVelocityLocal _guststrengthwindvelocity_out_19)
            (LocalVariables|SetWindGustLocalSettings%20Local _windgustlocalsettingsout_19)
            (bind _cloth_settings_out_19 (Default|ClothControllerSetterFunctionDEPRECATED _characterpartaniminstance_19 _dbno_rollbwd_63_2cf54c094abe5911a4dd0587754bd23f)) ; DBNO Roll Bwd
            (LocalVariables|SetCloth%20Settings%20Local _cloth_settings_out_19)
            (LocalVariables|SetWindGustSettingsLocal _dbno_rollbwd_73_b4165acb4840a5d88fccf28ca56463a1))
          (:DBNO_Impulse
            ;[1 1 1] DBNO Impulse
            (LocalVariables|SetCurrentState "DBNO Impulse")
            (bind (_guststrengthwindvelocity_out_20 _windgustlocalsettingsout_20) (Default|WindGustControlFunctionDEPRECATED :CharacterPartAnimInstance _characterpartaniminstance_20 :DeltaTime (Variables|GetDeltaTime) :GameTimeInSeconds (Variables|GetGameTimeInSeconds) :JointName _jointname_27_c74f3b2245ea319de82a629ff3c17cca_19 :WindGustLocalSettings (LocalVariables|GetWindGustLocalSettings%20Local)))
            (LocalVariables|SetGustStrength%20-%20WindVelocityLocal _guststrengthwindvelocity_out_20)
            (LocalVariables|SetWindGustLocalSettings%20Local _windgustlocalsettingsout_20)
            (bind _cloth_settings_out_20 (Default|ClothControllerSetterFunctionDEPRECATED _characterpartaniminstance_20 _dbno_impulse_64_dc5f9eeb407dd4ff2de0ac812dd29451)) ; DBNO Roll Bwd
            (LocalVariables|SetCloth%20Settings%20Local _cloth_settings_out_20)
            (LocalVariables|SetWindGustSettingsLocal _dbno_impulse_74_3b793bd343babd1a82c40abcc2d883e4))
          (:DBNO_HighRotationRate
            _)
          (:Skydive_Surfing
            (LocalVariables|SetCurrentState "SkyDiving_Surfing")
            (bind (_guststrengthwindvelocity_out_21 _windgustlocalsettingsout_21) (Default|WindGustControlFunctionDEPRECATED _characterpartaniminstance_2 _deltatime _gametimeinseconds _jointname_27_c74f3b2245ea319de82a629ff3c17cca_11 _skydive_surfing_50_2b64886e4e6f0e8cfc180e81fdf37f62_2 (LocalVariables|GetWindGustLocalSettings%20Local))) ; Skydiving Surfing
            (LocalVariables|SetGustStrength%20-%20WindVelocityLocal _guststrengthwindvelocity_out_21)
            (LocalVariables|SetWindGustLocalSettings%20Local _windgustlocalsettingsout_21)
            (bind _cloth_settings_out_21 (Default|ClothControllerSetterFunctionDEPRECATED _characterpartaniminstance_2 _skydive_surfing_42_408e65ff484aa54111e1ed8a69d8240d_2)) ; Skydiving Surfing
            (LocalVariables|SetCloth%20Settings%20Local _cloth_settings_out_21)
            (LocalVariables|SetWindGustSettingsLocal _skydive_surfing_50_2b64886e4e6f0e8cfc180e81fdf37f62_2))
          (:Front_End
            ;[1 1 1] Front End
            (LocalVariables|SetCurrentState "FrontEnd")
            (bind (_guststrengthwindvelocity_out_22 _windgustlocalsettingsout_22) (Default|WindGustControlFunctionDEPRECATED _characterpartaniminstance_5 (Variables|GetDeltaTime) (Variables|GetGameTimeInSeconds) _jointname_27_c74f3b2245ea319de82a629ff3c17cca_6 _front_end_37_c7a4562a48e9bc9b4cd34d963cfc6490_5 (LocalVariables|GetWindGustLocalSettings%20Local)))
            (LocalVariables|SetGustStrength%20-%20WindVelocityLocal _guststrengthwindvelocity_out_22)
            (LocalVariables|SetWindGustLocalSettings%20Local _windgustlocalsettingsout_22)
            (bind _cloth_settings_out_22 (Default|ClothControllerSetterFunctionDEPRECATED _characterpartaniminstance_5 _front_end_20_c7a4562a48e9bc9b4cd34d963cfc6490_3)) ; Front End
            (LocalVariables|SetCloth%20Settings%20Local _cloth_settings_out_22)
            (LocalVariables|SetWindGustSettingsLocal _front_end_37_c7a4562a48e9bc9b4cd34d963cfc6490_5)))
        ;[1 1 1] On Ground Standing
        (LocalVariables|SetCurrentState "OnGround_Standing")
        (bind (_guststrengthwindvelocity_out_23 _windgustlocalsettingsout_23) (Default|WindGustControlFunctionDEPRECATED _characterpartaniminstance_11 (Variables|GetDeltaTime) (Variables|GetGameTimeInSeconds) _jointname_27_c74f3b2245ea319de82a629ff3c17cca_8 _onground_standing_32_267db1464d0d3a29d99e94adfcb4afec (LocalVariables|GetWindGustLocalSettings%20Local)))
        (LocalVariables|SetGustStrength%20-%20WindVelocityLocal _guststrengthwindvelocity_out_23)
        (LocalVariables|SetWindGustLocalSettings%20Local _windgustlocalsettingsout_23)
        (bind _cloth_settings_out_23 (Default|ClothControllerSetterFunctionDEPRECATED _characterpartaniminstance_11 _onground_standing_15_267db1464d0d3a29d99e94adfcb4afec)) ; OnGround Standing
        (LocalVariables|SetCloth%20Settings%20Local _cloth_settings_out_23)
        (LocalVariables|SetWindGustSettingsLocal _onground_standing_32_267db1464d0d3a29d99e94adfcb4afec))
      (:then_3
        ;[1 1 1] Set all of the final values after they have been determined.
        (ClothingSimulation|SetVelocityScale _chaosclothinginteractor _linearvelocityscale_7_9cbe863e417fc965c07809a2fe55adfd _angularvelocityscale_6_6a3dfe4a481ad4d7cb18d79f95cc88e8 _fictitiousangularscale_10_c5f4d16f48d0bedc3f0fa38b2648511b)
        (ClothingSimulation|SetGravity _chaosclothinginteractor 1.0 true _gravityoverride_15_9c3cf2b14934a1ac917c4c99f03afc6e)
        (ClothingSimulation|SetWind _chaosclothinginteractor _dragcoefficient_23_fe68f666411ff9f29ba84cb4be726a52 _liftcoefficient_24_bb285d994a5b3aff167f4793d7eca837 1e-06 (LocalVariables|GetGustStrength%20-%20WindVelocityLocal) "(X=0.070000,Y=0.500000)" "(X=0.070000,Y=0.500000)")
        ;[1 1 1] Set the Directly Driven Values
        (ClothingSimulation|SetMaterial _chaosclothinginteractor_2 _edgestiffness_28_9a24da0842e21d5736699499f1de1171 _bendingstiffness_29_6cd0864b4078d8378c061197e9b1b094 _areastiffness_30_0d981b4e4039d5a0c6a4c78d6f09d510)
        (ClothingSimulation|SetLongRangeAttachment _chaosclothinginteractor_2 _tetherstiffness_24_a5c4b017498d3154d76cecab9ae7bdf2 "(X=1.000000,Y=1.000000)")
        (ClothingSimulation|SetCollision _chaosclothinginteractor_2 _collisionthickness_14_dacf320b4c03c6bd248a8e9b7b867fc7 _frictioncoefficient_15_e5761ce84e8e52eebd663b9dc3a0906f false _selfcollisionthickness_16_98974b1c4b5abe4ea1cff7a791319392)
        (ClothingSimulation|SetDamping _chaosclothinginteractor_2 _dampingcoefficient_17_a343d5a54d704f5b249c8e907da80101 _localdampingcoefficient_33_993afaea430eebf424a7d0b8bb57cfb5)
        (ClothingSimulation|SetAnimDrive _chaosclothinginteractor_2 _animdrivestiffness_27_e54c28b348b4c35a1d0817afa9ff0423 _animdrivedamping_26_f9edbf5240e75a4905afa1a8240922c8)
        ;[1 1 1] Iterations and Substeps
        (ClothingSimulation|SetNumIterations _returnvalue_2 _numberiterations_21_8e04a8e94b2040c7dbf93194733ac36a)
        (ClothingSimulation|SetNumSubsteps _returnvalue_2 _numbersubsteps_23_5108d5fa40140a1aad38459100b14261)
        (return (LocalVariables|GetCurrentState) (LocalVariables|GetCloth%20Settings%20Local) (LocalVariables|GetCloth%20Core%20Settings%20Local) (LocalVariables|GetWindGustSettingsLocal) (LocalVariables|GetWindGustLocalSettings%20Local) (LocalVariables|GetGustStrength%20-%20WindVelocityLocal) (LocalVariables|GetMovementLinearVelVector%20Local))))))

(fn ClothControllerHelperScaleForPlayerMovementDEPRECATED
  (params (LinearVelScale :type Vector) (PlayerFwdVector :type Float%20%28double-precision%29) (PlayerFwdVectorJump :type Float%20%28double-precision%29) (ClothUpX :type Float%20%28double-precision%29) (ClothUpZ :type Float%20%28double-precision%29))
  ;[1 1 1] Moving 2d.  Uses movement to change the linear velocities based on movement.  Helps with how the cape looks when jumpign and moving bwd.  Also Angular Vel and Fictitious Scales are different here than standing still.
  (if (> ClothUpX 0.5)
    (return (Math|Vector|MakeVector (* (.x LinearVelScale) PlayerFwdVectorJump) (* (.y LinearVelScale) PlayerFwdVector) (.z LinearVelScale)))
    (elif (> ClothUpZ 0.5)
      (return (Math|Vector|MakeVector (.x LinearVelScale) (* (.y LinearVelScale) PlayerFwdVector) (* (.z LinearVelScale) PlayerFwdVectorJump))))))

(fn ClothControllerSetterFunctionDEPRECATED
  (params (CharacterPartAnimInstance :type Custom%20Character%20Part%20Anim%20Instance%20Object%20Reference) (Cloth%20Settings :type Cloth%20Setting%20Struct%20Structure))
  (bind (_linearvelocityscale_7_9cbe863e417fc965c07809a2fe55adfd _angularvelocityscale_6_6a3dfe4a481ad4d7cb18d79f95cc88e8 _fictitiousangularscale_10_c5f4d16f48d0bedc3f0fa38b2648511b _gravityoverride_15_9c3cf2b14934a1ac917c4c99f03afc6e _dragcoefficient_23_fe68f666411ff9f29ba84cb4be726a52 _liftcoefficient_24_bb285d994a5b3aff167f4793d7eca837 _jointname_27_c74f3b2245ea319de82a629ff3c17cca) (Utilities|Struct|BreakClothSettingStruct Cloth%20Settings))
  (bind _returnvalue (CustomCharacter|GetOwnerSkeletalMeshComponent CharacterPartAnimInstance))
  (Utilities|IsValid _returnvalue
    (:Is%20Valid
      ;[1 1 1] Combine and apply the values for all of the settings before adding them to the Struct
      (bind _output_get (LocalVariables|SetLinear%20Velocity%20Scale%20Local _linearvelocityscale_7_9cbe863e417fc965c07809a2fe55adfd))
      (bind _output_get_2 (LocalVariables|SetAngular%20Velocity%20Scale%20Local _angularvelocityscale_6_6a3dfe4a481ad4d7cb18d79f95cc88e8))
      (bind _output_get_3 (LocalVariables|SetFictitious%20Angular%20Scale%20Local _fictitiousangularscale_10_c5f4d16f48d0bedc3f0fa38b2648511b))
      (bind _returnvalue_2 (Animation|Utilities|AlignVectorwithJoint _gravityoverride_15_9c3cf2b14934a1ac917c4c99f03afc6e _returnvalue 0 _jointname_27_c74f3b2245ea319de82a629ff3c17cca))
      (bind _output_get_4 (LocalVariables|SetGravity%20Override%20Local _returnvalue_2))
      (bind _output_get_5 (LocalVariables|SetDrag%20Coefficient%20Local _dragcoefficient_23_fe68f666411ff9f29ba84cb4be726a52))
      (bind _output_get_6 (LocalVariables|SetLift%20Coefficient%20Local _liftcoefficient_24_bb285d994a5b3aff167f4793d7eca837))
      (return (Utilities|Struct|MakeClothSettingStruct _output_get _output_get_2 _output_get_3 _output_get_4 _output_get_5 _output_get_6 _jointname_27_c74f3b2245ea319de82a629ff3c17cca)))))

(fn ClothAssetControllerFunctionDEPRECATED
  (params (CharacterPartAnimInstance :type Custom%20Character%20Part%20Anim%20Instance%20Object%20Reference) (SkeletalMeshComponent :type Skeletal%20Mesh%20Component%20Object%20Reference) (ClothingAssetName :type String :default "Name of your Clothing Asset from the SkelMesh goes here") (DeltaTime :type Float%20%28double-precision%29) (Debug :type Boolean) (TrackedBone :type Name :default "pelvis") (ClothAssetProperties_Default :type Cloth%20Asset%20Controller%20Properties%20Struct%20Structure) (ClothAssetControllerParams :type Cloth%20Asset%20Controller%20Params%20Struct%20Structure) (isJumpTimingIN :type Boolean) (JumpTimerIN :type Float%20%28double-precision%29))
  (bind (_gravityoverride_29_9cbe863e417fc965c07809a2fe55adfd _windvelocity_28_9c3cf2b14934a1ac917c4c99f03afc6e _linearvelocityscale_34_7b4eb54144b7b9bc735052af800f52bc _maxlinearvelocity_35_db31169f4de195bda020009507428ae7) (Utilities|Struct|BreakClothAssetControllerPropertiesStruct ClothAssetProperties_Default))
  (bind (_goingdwn_divedwn_75_3b988791442f1e014430ef972980e2ed _goingup_facedwn_76_80a1d8be411f9623986840bcc1be0c04 _goingup_headfirst_77_541cf9d54cbbe9cd9252b297a71a7f58 _fallingfast_whileupright_78_27047d3c4308e02195be32941c8303bf) (Utilities|Struct|BreakClothAssetControllerParamsStruct ClothAssetControllerParams))
  (bind (_gravityoverride_29_9cbe863e417fc965c07809a2fe55adfd_2 _windvelocity_28_9c3cf2b14934a1ac917c4c99f03afc6e_2 _linearvelocityscale_34_7b4eb54144b7b9bc735052af800f52bc_2 _maxlinearvelocity_35_db31169f4de195bda020009507428ae7_2) (Utilities|Struct|BreakClothAssetControllerPropertiesStruct _goingdwn_divedwn_75_3b988791442f1e014430ef972980e2ed))
  (bind (_gravityoverride_29_9cbe863e417fc965c07809a2fe55adfd_3 _windvelocity_28_9c3cf2b14934a1ac917c4c99f03afc6e_3 _linearvelocityscale_34_7b4eb54144b7b9bc735052af800f52bc_3 _maxlinearvelocity_35_db31169f4de195bda020009507428ae7_3) (Utilities|Struct|BreakClothAssetControllerPropertiesStruct _goingup_facedwn_76_80a1d8be411f9623986840bcc1be0c04))
  (bind (_gravityoverride_29_9cbe863e417fc965c07809a2fe55adfd_4 _windvelocity_28_9c3cf2b14934a1ac917c4c99f03afc6e_4 _linearvelocityscale_34_7b4eb54144b7b9bc735052af800f52bc_4 _maxlinearvelocity_35_db31169f4de195bda020009507428ae7_4) (Utilities|Struct|BreakClothAssetControllerPropertiesStruct _goingup_headfirst_77_541cf9d54cbbe9cd9252b297a71a7f58))
  (bind (_gravityoverride_29_9cbe863e417fc965c07809a2fe55adfd_5 _windvelocity_28_9c3cf2b14934a1ac917c4c99f03afc6e_5 _linearvelocityscale_34_7b4eb54144b7b9bc735052af800f52bc_5 _maxlinearvelocity_35_db31169f4de195bda020009507428ae7_5) (Utilities|Struct|BreakClothAssetControllerPropertiesStruct _fallingfast_whileupright_78_27047d3c4308e02195be32941c8303bf))
  (bind _trackedbonefacingdirection (LocalVariables|GetTrackedBoneFacingDirection))
  (bind _returnvalue (Math|Float|Absolute(Float) _trackedbonefacingdirection))
  (bind _trackedbonefacingdirection_2 (LocalVariables|GetTrackedBoneFacingDirection))
  (bind _returnvalue_2 (* _trackedbonefacingdirection_2 -1.0))
  (bind _returnvalue_3 (Math|Float|Clamp(Float) _returnvalue_2))
  (bind _trackedbonefacingdirection_3 (LocalVariables|GetTrackedBoneFacingDirection))
  (bind _returnvalue_4 (Math|Float|Clamp(Float) _trackedbonefacingdirection_3))
  (bind _characterpartaniminstance (Variables|GetCharacterPartAnimInstance))
  (bind _pawnvelocityz (Variables|/Script/FortniteGame.FortAnimInstance|GetPawnVelocityZ _characterpartaniminstance))
  (bind _returnvalue_5 (Math|Float|MapRangeClamped _pawnvelocityz 1e+03 8e+03 0.0 1.0))
  (bind _trackedbonefacingdirection_4 (LocalVariables|GetTrackedBoneFacingDirection))
  (bind _returnvalue_6 (Math|Float|Clamp(Float) _trackedbonefacingdirection_4))
  (bind _characterpartaniminstance_2 (Variables|GetCharacterPartAnimInstance))
  (bind _pawnvelocityz_2 (Variables|/Script/FortniteGame.FortAnimInstance|GetPawnVelocityZ _characterpartaniminstance_2))
  (bind _returnvalue_7 (Math|Float|MapRangeClamped _pawnvelocityz_2 -1e+02 -5e+02 0.0 1.0))
  (bind _characterpartaniminstance_3 (Variables|GetCharacterPartAnimInstance))
  (bind _pawnvelocityz_3 (Variables|/Script/FortniteGame.FortAnimInstance|GetPawnVelocityZ _characterpartaniminstance_3))
  (bind _returnvalue_8 (Math|Float|MapRangeClamped _pawnvelocityz_3 8e+02 6e+03 0.0 1.0))
  (bind _characterpartaniminstance_4 (Variables|GetCharacterPartAnimInstance))
  (bind _pawnvelocityz_4 (Variables|/Script/FortniteGame.FortAnimInstance|GetPawnVelocityZ _characterpartaniminstance_4))
  (bind _returnvalue_9 (Math|Float|MapRangeClamped _pawnvelocityz_4 -8e+02 -3e+03 0.0 1.0))
  (bind _alphasum (LocalVariables|GetAlphaSum))
  (bind _goingdwndivedwn_alpha (LocalVariables|GetGoingDwnDiveDwn_Alpha))
  (bind _returnvalue_10 (/ _goingdwndivedwn_alpha _alphasum))
  (bind _returnvalue_11 (- 1.0 _returnvalue))
  (bind _returnvalue_12 (* _returnvalue_8 _returnvalue_11))
  (bind _returnvalue_13 (* _returnvalue_9 _returnvalue_3))
  (bind _alphasum_2 (LocalVariables|GetAlphaSum))
  (bind _goingupfacedwn_alpha (LocalVariables|GetGoingUpFaceDwn_Alpha))
  (bind _returnvalue_14 (/ _goingupfacedwn_alpha _alphasum_2))
  (bind _alphasum_3 (LocalVariables|GetAlphaSum))
  (bind _fallingfastwhileupright_alpha (LocalVariables|GetFallingFastWhileUpright_Alpha))
  (bind _returnvalue_15 (/ _fallingfastwhileupright_alpha _alphasum_3))
  (bind _alphasum_4 (LocalVariables|GetAlphaSum))
  (bind _goingupheadfirst_alpha (LocalVariables|GetGoingUpHeadFirst_Alpha))
  (bind _returnvalue_16 (/ _goingupheadfirst_alpha _alphasum_4))
  (bind _returnvalue_17 (Utilities|Operators|Add _returnvalue_16 _returnvalue_14 _returnvalue_10 _returnvalue_15))
  (bind _returnvalue_18 (* _returnvalue_7 _returnvalue_6))
  (bind _returnvalue_19 (* _returnvalue_5 _returnvalue_4))
  (bind _skeletalmeshcomponent (Variables|GetSkeletalMeshComponent))
  (bind _clothassetinteractor (LocalVariables|GetClothAssetInteractor))
  (bind _debug (Variables|GetDebug))
  ;[1 1 1] Check to see if cloth is enabled at all.  If not just break out.
  (if (Variables|/Script/FortniteGame.FortAnimInstance|GetbClothEnabled (Variables|GetCharacterPartAnimInstance))
    (LocalVariables|SetisJumpTiming%20Local isJumpTimingIN)
    (LocalVariables|SetJumpTimerLocal JumpTimerIN)
    ; Get the default values that are coming in from the Cloth Asset
    (LocalVariables|SetGravityOverrideDEFAULT _gravityoverride_29_9cbe863e417fc965c07809a2fe55adfd)
    (LocalVariables|SetWindVelocityDEFAULT _windvelocity_28_9c3cf2b14934a1ac917c4c99f03afc6e)
    (LocalVariables|SetLinearVelScaleDEFAULT _linearvelocityscale_34_7b4eb54144b7b9bc735052af800f52bc)
    (LocalVariables|SetMaxLinearVelDEFAULT _maxlinearvelocity_35_db31169f4de195bda020009507428ae7)
    ; Get the target values that are coming in from the AnimBP and set by the user.
    (LocalVariables|SetGoingDwnDiveDwn_Gravity_TARGET _gravityoverride_29_9cbe863e417fc965c07809a2fe55adfd_2)
    (LocalVariables|SetGoingDwnDiveDwn_Wind_TARGET _windvelocity_28_9c3cf2b14934a1ac917c4c99f03afc6e_2)
    (LocalVariables|SetGoingDwnDiveDwn_LinVel_TARGET _linearvelocityscale_34_7b4eb54144b7b9bc735052af800f52bc_2)
    (LocalVariables|SetGoingDwnDiveDwn_MaxLinVelClamp_TARGET _maxlinearvelocity_35_db31169f4de195bda020009507428ae7_2)
    (LocalVariables|SetGoingUpFaceDwn_Gravity_TARGET _gravityoverride_29_9cbe863e417fc965c07809a2fe55adfd_3)
    (LocalVariables|SetGoingUpFaceDwn_Wind_TARGET _windvelocity_28_9c3cf2b14934a1ac917c4c99f03afc6e_3)
    (LocalVariables|SetGoingUpFaceDwn_LinVel_TARGET _linearvelocityscale_34_7b4eb54144b7b9bc735052af800f52bc_3)
    (LocalVariables|SetGoingUpFaceDwn_MaxLinVelClamp_TARGET _maxlinearvelocity_35_db31169f4de195bda020009507428ae7_3)
    (LocalVariables|SetGoingUpHeadFirst_Gravity_TARGET _gravityoverride_29_9cbe863e417fc965c07809a2fe55adfd_4)
    (LocalVariables|SetGoingUpHeadFirst_Wind_TARGET _windvelocity_28_9c3cf2b14934a1ac917c4c99f03afc6e_4)
    (LocalVariables|SetGoingUpHeadFirst_LinVel_TARGET _linearvelocityscale_34_7b4eb54144b7b9bc735052af800f52bc_4)
    (LocalVariables|SetGoingUpHeadFirst_MaxLinVelClamp_TARGET _maxlinearvelocity_35_db31169f4de195bda020009507428ae7_4)
    (LocalVariables|SetFallingFastWhileUpright_Gravity_TARGET _gravityoverride_29_9cbe863e417fc965c07809a2fe55adfd_5)
    (LocalVariables|SetFallingFastWhileUpright_Wind_TARGET _windvelocity_28_9c3cf2b14934a1ac917c4c99f03afc6e_5)
    (LocalVariables|SetFallingFastWhileUpright_LinVel_TARGET _linearvelocityscale_34_7b4eb54144b7b9bc735052af800f52bc_5)
    (LocalVariables|SetFallingFastWhileUpright_MaxLinVelClamp_TARGET _maxlinearvelocity_35_db31169f4de195bda020009507428ae7_5)
    (Utilities|FlowControl|Sequence
      (:then_0
        ;[1 1 1] Use SkelMeshComponent to get ClothAssetInteractor
        (bind _aschaos_cloth_asset_interactor (Utilities|Casting|CastToChaosClothAssetInteractor (ClothingSimulation|GetClothingInteractor (Clothing|GetClothingSimulationInteractor (Variables|GetSkeletalMeshComponent) (Utilities|String|StringToName (Variables|GetClothingAssetName))) (Utilities|String|StringToName (Variables|GetClothingAssetName)))))
        (LocalVariables|SetClothAssetInteractor _aschaos_cloth_asset_interactor))
      (:then_1
        ;[1 1 1] Calculate the per axis velocity of a given bone.
        (LocalVariables|SetTrackedBoneLocation (Transformation|GetSocketLocation _skeletalmeshcomponent (Variables|GetTrackedBone)))
        (LocalVariables|SetTrackedBoneFacingDirection (Math|Vector|DotProduct (Math|Vector|GetForwardVector (.rotation (Transformation|GetSocketTransform _skeletalmeshcomponent (Variables|GetTrackedBone) "RTS_World"))) "0.000000,0.000000,1.000000")))
      (:then_2
        ; Jump Timer
        (if (Variables|/Script/FortniteGame.FortAnimInstance|GetbIsJumping (Variables|GetCharacterPartAnimInstance))
          (LocalVariables|SetJumpTimerLocal 0.0)
          (LocalVariables|SetisJumpTiming%20Local true)))
      (:then_3
        (if (LocalVariables|GetisJumpTiming%20Local)
          (LocalVariables|SetJumpTimerLocal (+ (LocalVariables|GetJumpTimerLocal) (Variables|GetDeltaTime)))
          (if (> (LocalVariables|GetJumpTimerLocal) (LocalVariables|GetJumpTimerMax))
            (LocalVariables|SetisJumpTiming%20Local false)
            (LocalVariables|SetJumpTimerLocal 0.0))))
      (:then_4
        ;[0 0.7618 1 0.5] GoingDwn DiveDwn - Moving down fast while the character is upside down/vertical (Skydive Diving Down)
        (LocalVariables|SetGoingDwnDiveDwn_Alpha _returnvalue_13)
        (LocalVariables|SetGoingDwnDiveDwn_Gravity_RESULT (Math|Vector|Lerp(Vector) (LocalVariables|GetGravityOverrideDEFAULT) (LocalVariables|GetGoingDwnDiveDwn_Gravity_TARGET) _returnvalue_13))
        (LocalVariables|SetGoingDwnDiveDwn_Wind_RESULT (Math|Vector|Lerp(Vector) (LocalVariables|GetWindVelocityDEFAULT) (LocalVariables|GetGoingDwnDiveDwn_Wind_TARGET) _returnvalue_13))
        (LocalVariables|SetGoingDwnDiveDwn_LinVel_RESULT (Math|Vector|Lerp(Vector) (LocalVariables|GetLinearVelScaleDEFAULT) (LocalVariables|GetGoingDwnDiveDwn_LinVel_TARGET) _returnvalue_13))
        (LocalVariables|SetGoingDwnDiveDwn_MaxLinVelClamp_RESULT (Math|Vector|Lerp(Vector) (LocalVariables|GetMaxLinearVelDEFAULT) (LocalVariables|GetGoingDwnDiveDwn_MaxLinVelClamp_TARGET) _returnvalue_13)))
      (:then_5
        ;[0 0.5938 0.000515] GoingUp FaceDwn - Going up fast while the character is horizontal and facing down
        (if true
          (LocalVariables|SetGoingUpFaceDwn_Alpha _returnvalue_12)
          (LocalVariables|SetGoingUpFaceDwn_Gravity_RESULT (Math|Vector|Lerp(Vector) (LocalVariables|GetGravityOverrideDEFAULT) (LocalVariables|GetGoingUpFaceDwn_Gravity_TARGET) _returnvalue_12))
          (LocalVariables|SetGoingUpFaceDwn_Wind_RESULT (Math|Vector|Lerp(Vector) (LocalVariables|GetWindVelocityDEFAULT) (LocalVariables|GetGoingUpFaceDwn_Wind_TARGET) _returnvalue_12))
          (LocalVariables|SetGoingUpFaceDwn_LinVel_RESULT (Math|Vector|Lerp(Vector) (LocalVariables|GetLinearVelScaleDEFAULT) (LocalVariables|GetGoingUpFaceDwn_LinVel_TARGET) _returnvalue_12))
          (LocalVariables|SetGoingUpFaceDwn_MaxLinVelClamp_RESULT (Math|Vector|Lerp(Vector) (LocalVariables|GetMaxLinearVelDEFAULT) (LocalVariables|GetGoingUpFaceDwn_MaxLinVelClamp_TARGET) _returnvalue_12))))
      (:then_6
        ;[0 0 1] GoingUp HeadFirst - Moving up fast while the character is vertical (Skydive GoingUp)
        (LocalVariables|SetGoingUpHeadFirst_Alpha _returnvalue_19)
        (LocalVariables|SetGoingUpHeadFirst_Gravity_RESULT (Math|Vector|Lerp(Vector) (LocalVariables|GetGravityOverrideDEFAULT) (LocalVariables|GetGoingUpHeadFirst_Gravity_TARGET) _returnvalue_19))
        (LocalVariables|SetGoingUpHeadFirst_Wind_RESULT (Math|Vector|Lerp(Vector) (LocalVariables|GetWindVelocityDEFAULT) (LocalVariables|GetGoingUpHeadFirst_Wind_TARGET) _returnvalue_19))
        (LocalVariables|SetGoingUpHeadFirst_LinVel_RESULT (Math|Vector|Lerp(Vector) (LocalVariables|GetLinearVelScaleDEFAULT) (LocalVariables|GetGoingUpHeadFirst_LinVel_TARGET) _returnvalue_19))
        (LocalVariables|SetGoingUpHeadFirst_MaxLinVelClamp_RESULT (Math|Vector|Lerp(Vector) (LocalVariables|GetMaxLinearVelDEFAULT) (LocalVariables|GetGoingUpHeadFirst_MaxLinVelClamp_TARGET) _returnvalue_19)))
      (:then_7
        ;[1 0 0] FallingFast WhileUpright - Falling down fast while the character is vertical (Parachuting or Falling)
        ;; Jumping on even ground fix.
        (if (Math|Float|InRange(Float) (LocalVariables|GetJumpTimerLocal) 0.01 (LocalVariables|GetJumpTimerMax))
          (LocalVariables|SetFallingFastWhileUpright_Wind_TARGET (LocalVariables|GetWindVelocityDEFAULT))
          (else
            (LocalVariables|SetFallingFastWhileUpright_Wind_TARGET (LocalVariables|GetFallingFastWhileUpright_Wind_TARGET))))
        (LocalVariables|SetFallingFastWhileUpright_Alpha _returnvalue_18)
        (LocalVariables|SetFallingFastWhileUpright_Gravity_RESULT (Math|Vector|Lerp(Vector) (LocalVariables|GetGravityOverrideDEFAULT) (LocalVariables|GetFallingFastWhileUpright_Gravity_TARGET) _returnvalue_18))
        (LocalVariables|SetFallingFastWhileUpright_Wind_RESULT (Math|Vector|Lerp(Vector) (LocalVariables|GetWindVelocityDEFAULT) (LocalVariables|GetFallingFastWhileUpright_Wind_TARGET) _returnvalue_18))
        (LocalVariables|SetFallingFastWhileUpright_LinVel_RESULT (Math|Vector|Lerp(Vector) (LocalVariables|GetLinearVelScaleDEFAULT) (LocalVariables|GetFallingFastWhileUpright_LinVel_TARGET) _returnvalue_18))
        (LocalVariables|SetFallingFastWhileUpright_MaxLinVelClamp_RESULT (Math|Vector|Lerp(Vector) (LocalVariables|GetMaxLinearVelDEFAULT) (LocalVariables|GetFallingFastWhileUpright_MaxLinVelClamp_TARGET) _returnvalue_18)))
      (:then_8
        ; Calculates all of the above states and blends everything together
        (LocalVariables|SetAlphaSum (Math|Float|SelectFloat 1.0 (Utilities|Operators|Add (LocalVariables|GetGoingUpHeadFirst_Alpha) (LocalVariables|GetGoingUpFaceDwn_Alpha) (LocalVariables|GetGoingDwnDiveDwn_Alpha) (LocalVariables|GetFallingFastWhileUpright_Alpha)) (== (Utilities|Operators|Add (LocalVariables|GetGoingUpHeadFirst_Alpha) (LocalVariables|GetGoingUpFaceDwn_Alpha) (LocalVariables|GetGoingDwnDiveDwn_Alpha) (LocalVariables|GetFallingFastWhileUpright_Alpha)) 0)))
        (LocalVariables|SetGravityOverride_OUTPUT (Math|Vector|Lerp(Vector) (LocalVariables|GetGravityOverrideDEFAULT) (Utilities|Operators|Add (* (LocalVariables|GetGoingUpHeadFirst_Gravity_RESULT) _returnvalue_16) (* (LocalVariables|GetGoingUpFaceDwn_Gravity_RESULT) _returnvalue_14) (* (LocalVariables|GetGoingDwnDiveDwn_Gravity_RESULT) _returnvalue_10) (* (LocalVariables|GetFallingFastWhileUpright_Gravity_RESULT) _returnvalue_15)) _returnvalue_17))
        (LocalVariables|SetWindVelocity_OUTPUT (Math|Vector|Lerp(Vector) (LocalVariables|GetWindVelocityDEFAULT) (Utilities|Operators|Add (* (LocalVariables|GetGoingUpHeadFirst_Wind_RESULT) _returnvalue_16) (* (LocalVariables|GetGoingUpFaceDwn_Wind_RESULT) _returnvalue_14) (* (LocalVariables|GetGoingDwnDiveDwn_Wind_RESULT) _returnvalue_10) (* (LocalVariables|GetFallingFastWhileUpright_Wind_RESULT) _returnvalue_15)) _returnvalue_17))
        (LocalVariables|SetLinearVelScale_OUTPUT (Math|Vector|Lerp(Vector) (LocalVariables|GetLinearVelScaleDEFAULT) (Utilities|Operators|Add (* (LocalVariables|GetGoingUpHeadFirst_LinVel_RESULT) _returnvalue_16) (* (LocalVariables|GetGoingUpFaceDwn_LinVel_RESULT) _returnvalue_14) (* (LocalVariables|GetGoingDwnDiveDwn_LinVel_RESULT) _returnvalue_10) (* (LocalVariables|GetFallingFastWhileUpright_LinVel_RESULT) _returnvalue_15)) _returnvalue_17))
        (LocalVariables|SetMaxLinearVel_OUTPUT (Math|Vector|Lerp(Vector) (LocalVariables|GetMaxLinearVelDEFAULT) (Utilities|Operators|Add (* (LocalVariables|GetGoingUpFaceDwn_MaxLinVelClamp_RESULT) _returnvalue_14) (* (LocalVariables|GetGoingUpHeadFirst_MaxLinVelClamp_RESULT) _returnvalue_16) (* (LocalVariables|GetGoingDwnDiveDwn_MaxLinVelClamp_RESULT) _returnvalue_10) (* (LocalVariables|GetFallingFastWhileUpright_MaxLinVelClamp_RESULT) _returnvalue_15)) _returnvalue_17)))
      (:then_9
        ;[1 1 1] Apply the OUTPUT values to the Cloth Interactor
        (bind _returnvalue_20 (Animation|Utilities|AlignVectorwithJoint (LocalVariables|GetGravityOverride_OUTPUT) (Variables|GetSkeletalMeshComponent) 0 "root")) ; Gravity has to be aligned with the joint here because cloth gravity is always in world space.  So the vector has to change instead.
        (bind _output_get (LocalVariables|SetGravityOverride_OUTPUT _returnvalue_20))
        (ClothProperty|SetVectorPropertyValue _clothassetinteractor "GravityOverride" -1 _output_get)
        (ClothProperty|SetVectorPropertyValue _clothassetinteractor "WindVelocity" -1 (LocalVariables|GetWindVelocity_OUTPUT))
        (ClothProperty|SetVectorPropertyValue _clothassetinteractor "LinearVelocityScale" -1 (LocalVariables|GetLinearVelScale_OUTPUT))
        (ClothProperty|SetVectorPropertyValue _clothassetinteractor "MaxLinearVelocity" -1 (LocalVariables|GetMaxLinearVel_OUTPUT)))
      (:then_10
        ; Debugging
        (Development|PrintString 0 (Utilities|String|ToString(Float) (LocalVariables|GetGoingUpHeadFirst_Alpha)) _debug false "(R=0.000000,G=0.660000,B=1.000000,A=1.000000)" 1.0)
        (Development|PrintString 0 (Utilities|String|ToString(Float) (LocalVariables|GetGoingUpFaceDwn_Alpha)) _debug false "(R=0.000000,G=0.593750,B=0.000515,A=1.000000)" 1.0 "")
        (Development|PrintString 0 (Utilities|String|ToString(Float) (LocalVariables|GetGoingDwnDiveDwn_Alpha)) _debug false "(R=0.000000,G=0.000000,B=1.000000,A=1.000000)" 1.0 "")
        (Development|PrintString 0 (Utilities|String|ToString(Float) (LocalVariables|GetFallingFastWhileUpright_Alpha)) _debug false "(R=1.000000,G=0.000000,B=0.000000,A=1.000000)" 1.0))
      (:then_11
        ; Apply final values to the Return Node
        (bind _output_get_2 (Variables|SetDebuggerStruct_GoingUpHeadFirst_2_B008E91B4A9E2F50B0924C93C132B3C2 (Utilities|Struct|MakeClothAssetControllerFunctionDebugger (LocalVariables|GetGoingUpHeadFirst_Alpha) (LocalVariables|GetGoingUpFaceDwn_Alpha) (LocalVariables|GetGoingDwnDiveDwn_Alpha) (LocalVariables|GetFallingFastWhileUpright_Alpha))))
        (return _output_get_2 (Variables|GetClothAssetProperties_Default) (Utilities|Struct|MakeClothAssetControllerPropertiesStruct (LocalVariables|GetGravityOverride_OUTPUT) (LocalVariables|GetWindVelocity_OUTPUT) (LocalVariables|GetLinearVelScale_OUTPUT) (LocalVariables|GetMaxLinearVel_OUTPUT)) (LocalVariables|GetisJumpTiming%20Local) (LocalVariables|GetJumpTimerLocal))))))

(fn ClothAssetControllerGetClothPropertiesFunctionDEPRECATED
  (params (SkeletalMeshComponent :type Skeletal%20Mesh%20Component%20Object%20Reference) (ClothingAssetName :type String))
  ;[1 1 1] Use SkelMeshComponent to get ClothAssetInteractor
  (bind _aschaos_cloth_asset_interactor (Utilities|Casting|CastToChaosClothAssetInteractor (ClothingSimulation|GetClothingInteractor (Clothing|GetClothingSimulationInteractor (Variables|GetSkeletalMeshComponent) (Utilities|String|StringToName (Variables|GetClothingAssetName))) (Utilities|String|StringToName (Variables|GetClothingAssetName)))))
  (LocalVariables|SetClothAssetInteractor _aschaos_cloth_asset_interactor)
  (return (Utilities|Struct|MakeClothAssetControllerPropertiesStruct (ClothProperty|GetVectorPropertyValue (LocalVariables|GetClothAssetInteractor) "GravityOverride" 0 "0.000000,0.000000,0.000000") (ClothProperty|GetVectorPropertyValue (LocalVariables|GetClothAssetInteractor) "WindVelocity" 0 "0.000000,0.000000,0.000000") (ClothProperty|GetVectorPropertyValue (LocalVariables|GetClothAssetInteractor) "LinearVelocityScale" 0 "0.000000,0.000000,0.000000") (ClothProperty|GetVectorPropertyValue (LocalVariables|GetClothAssetInteractor) "MaxLinearVelocity" 0 "0.000000,0.000000,0.000000"))))

(fn ComplexSimSpaceFunctionDEPRECATED
  (params (CharacterPartAnimInstance :type Custom%20Character%20Part%20Anim%20Instance%20Object%20Reference) (InterpSpeedMax :type Float%20%28double-precision%29 :default 9.0) (PawnSpeedMax :type Float%20%28double-precision%29 :default 2e+01) (RInterpErrorTolerance :type Float%20%28double-precision%29 :default 1.0) (DeltaTime :type Float%20%28double-precision%29) (DebugDraw :type Boolean) (sim_space_rotation_in :type Rotator) (inComponentSpaceIn :type Boolean) (inComponentSpaceLastFrameIn :type Boolean) (lockedOnSimSpaceIn :type Boolean) (InterpSpeedIn :type Float%20%28double-precision%29 :default 9.0))
  (bind _local_simspacerotation (LocalVariables|GetLocal_SimSpaceRotation))
  (bind _local_usingcomponentspace (LocalVariables|GetLocal_UsingComponentSpace))
  (bind _local_lockedonsimspace (LocalVariables|GetLocal_LockedOnSimSpace))
  (Animation|Utilities|ComplexSimSpaceFunction CharacterPartAnimInstance InterpSpeedMax PawnSpeedMax RInterpErrorTolerance DeltaTime DebugDraw sim_space_rotation_in inComponentSpaceIn lockedOnSimSpaceIn _local_simspacerotation _local_usingcomponentspace _local_lockedonsimspace)
  (return _local_simspacerotation _local_usingcomponentspace false _local_lockedonsimspace 0.0))

(fn DebugClothTeleportThresholdFunction
  (params (MinDistance :type Float%20%28double-precision%29 :default 1.5e+02) (MinRotation :type Float%20%28double-precision%29 :default 25.0) (SkelMeshComponent :type Scene%20Component%20Object%20Reference) (PrintToScreen :type Boolean) (BoneName :type Name :default "root") (InVector :type Vector%20%28by%20ref%29) (PawnDeltaRotation :type Rotator) (PrintToScreenDuration :type Float%20%28double-precision%29 :default 6.0))
  (bind _bonename (Variables|GetBoneName))
  (bind _skelmeshcomponent (Variables|GetSkelMeshComponent))
  (bind _returnvalue (Transformation|GetSocketLocation _skelmeshcomponent _bonename))
  (Utilities|IsValid SkelMeshComponent
    (:Is%20Valid
      (Utilities|FlowControl|Sequence
        (:then_0
          ;[1 1 1] Get the Distance and store it locally
          (LocalVariables|SetDistanceLocal (Math|Vector|Distance(Vector) _returnvalue (Variables|GetInVector)))
          (Variables|SetBy-RefVar (Variables|GetInVector) _returnvalue))
        (:then_1
          ;[1 1 1] Print the Distance
          (if (Variables|GetPrintToScreen)
            (if (> (LocalVariables|GetDistanceLocal) (Variables|GetMinDistance))
              (Development|PrintString 0 (Utilities|String|ToString(Float) (LocalVariables|GetDistanceLocal)) true true "(R=0.000000,G=0.492111,B=0.770833,A=1.000000)" (Variables|GetPrintToScreenDuration)))))
        (:then_2
          ;[1 1 1] Get the Rotation and store it locally
          (LocalVariables|SetRotationLocal (Math|Float|Absolute(Float) (.yaw (Variables|GetPawnDeltaRotation)))))
        (:then_3
          ;[1 1 1] Print the Rotation
          (if (Variables|GetPrintToScreen)
            (if (> (LocalVariables|GetRotationLocal) (Variables|GetMinRotation))
              (Development|PrintString 0 (Utilities|String|ToString(Float) (LocalVariables|GetRotationLocal)) true true "(R=0.562500,G=0.000000,B=0.021304,A=1.000000)" (Variables|GetPrintToScreenDuration)))))
        (:then_4
          ;[1 1 1] Return the rotation and distance values
          (return (LocalVariables|GetDistanceLocal) (LocalVariables|GetRotationLocal)))))))

(fn DebugPlayerVelocitiesFunction
  (params (CharacterPartAnimInstance :type Custom%20Character%20Part%20Anim%20Instance%20Object%20Reference) (SkelMeshComponent :type Scene%20Component%20Object%20Reference) (DeltaTime :type Float%20%28double-precision%29) (DebugSpace :type DebugDrawSpace%20Enum :default "NewEnumerator1") (BoneName :type Name :default "root") (DebugDrawOffset :type Vector :default "0.000000,0.000000,105.000000") (AngularDebugDrawOffset :type Vector :default "0.000000,0.000000,0.000000") (DrawDebug :type Boolean) (Debug%20Scalar :type Float%20%28double-precision%29 :default 5e+01) (Foreground :type EDrawDebugSceneDepthPriorityGroup%20Enum :default "Foreground") (TextLocationOffset :type Vector :default "20.000000,20.000000,20.000000"))
  (bind _characterpartaniminstance (Variables|GetCharacterPartAnimInstance))
  (bind _pawnrotationrate (Variables|/Script/FortniteGame.FortAnimInstance|GetPawnRotationRate _characterpartaniminstance))
  (bind _yaw (.yaw _pawnrotationrate))
  (bind _textlocationoffset (Variables|GetTextLocationOffset))
  (bind _x (.x _textlocationoffset))
  (bind _y (.y _textlocationoffset))
  (bind _z (.z _textlocationoffset))
  (bind _skelmeshcomponent (Variables|GetSkelMeshComponent))
  (bind _returnvalue (Transformation|GetWorldRotation _skelmeshcomponent))
  (bind _pawnvelocity (LocalVariables|GetPawnVelocity))
  (bind _skelmeshcomponent_upvector (LocalVariables|GetSkelMeshComponent_UpVector))
  (bind _returnvalue_2 (Math|Vector|DotProduct _skelmeshcomponent_upvector _pawnvelocity))
  (bind _pawnvelocity_2 (LocalVariables|GetPawnVelocity))
  (bind _skelmeshcomponent_fwdvector (LocalVariables|GetSkelMeshComponent_FwdVector))
  (bind _returnvalue_3 (Math|Vector|DotProduct _skelmeshcomponent_fwdvector _pawnvelocity_2))
  (bind _skelmeshcomponent_2 (Variables|GetSkelMeshComponent))
  (bind _returnvalue_4 (Transformation|GetWorldLocation _skelmeshcomponent_2))
  (bind _pawnvelocity_3 (LocalVariables|GetPawnVelocity))
  (bind _skelmeshcomponent_rightvector (LocalVariables|GetSkelMeshComponent_RightVector))
  (bind _returnvalue_5 (Math|Vector|DotProduct _skelmeshcomponent_rightvector _pawnvelocity_3))
  (bind _skelmeshcomponent_upvector_2 (LocalVariables|GetSkelMeshComponent_UpVector))
  (bind _returnvalue_6 (* _skelmeshcomponent_upvector_2 _z))
  (bind _debugdrawoffset (Variables|GetDebugDrawOffset))
  (bind _returnvalue_7 (+ _returnvalue_4 _debugdrawoffset))
  (bind _returnvalue_8 (+ _returnvalue_7 _returnvalue_6))
  (bind _skelmeshcomponent_fwdvector_2 (LocalVariables|GetSkelMeshComponent_FwdVector))
  (bind _returnvalue_9 (* _skelmeshcomponent_fwdvector_2 _y))
  (bind _returnvalue_10 (+ _returnvalue_7 _returnvalue_9))
  (bind _skelmeshcomponent_rightvector_2 (LocalVariables|GetSkelMeshComponent_RightVector))
  (bind _returnvalue_11 (* _skelmeshcomponent_rightvector_2 _x))
  (bind _returnvalue_12 (+ _returnvalue_7 _returnvalue_11))
  (bind _angulardebugdrawoffset (Variables|GetAngularDebugDrawOffset))
  (bind _debugsystemcurrentloc (LocalVariables|GetDebugSystemCurrentLoc))
  (bind _returnvalue_13 (+ _debugsystemcurrentloc _angulardebugdrawoffset))
  (bind _returnvalue_14 (+ _returnvalue_13 "0.000000,0.000000,-20.000000"))
  ;[1 1 1] Exit if Debug Draw is off
  (if (Variables|GetDrawDebug)
    (Utilities|FlowControl|Sequence
      (:then_0
        ;[1 1 1] Find the Elements we need for this function.  Pawn, Owning Component
        (bind _asfort_player_pawn (Utilities|Casting|CastToFortPlayerPawn (Animation|TryGetFortPawn (Variables|GetCharacterPartAnimInstance))))
        (LocalVariables|SetPawn _asfort_player_pawn))
      (:then_1
        ;[1 1 1] Get Directional Vectors and Pawn Velocity
        (LocalVariables|SetSkelMeshComponent_RightVector (Math|Vector|GetRightVector _returnvalue))
        (LocalVariables|SetSkelMeshComponent_FwdVector (Math|Vector|GetForwardVector _returnvalue))
        (LocalVariables|SetSkelMeshComponent_UpVector (Math|Vector|GetUpVector _returnvalue))
        (LocalVariables|SetPawnVelocity (Variables|/Script/FortniteGame.FortAnimInstance|GetPawnVelocity (Variables|GetCharacterPartAnimInstance))))
      (:then_2
        (LocalVariables|SetDebugLocAligned (Utilities|Operators|Add _returnvalue_12 _returnvalue_10 _returnvalue_8))
        ;[1 1 1] Figure out the placement of the debug system
        (LocalVariables|SetXAxisTextLoc _returnvalue_12)
        (LocalVariables|SetYAxisTextLoc _returnvalue_10)
        (LocalVariables|SetZAxisTextLoc _returnvalue_8)
        (LocalVariables|SetDebugSystemCurrentLoc _returnvalue_7)
        (LocalVariables|SetDebugSystemCurrentRotation _returnvalue))
      (:then_3
        ;[1 1 1] Dot Product of the X axis from the Right Vector
        (LocalVariables|SetLinearXAxisEnd (+ (LocalVariables|GetDebugSystemCurrentLoc) (* _skelmeshcomponent_rightvector (* _returnvalue_5 (Variables|GetDebug%20Scalar)))))
        (LocalVariables|SetComponentRightVectorDot _returnvalue_5)
        ;[1 1 1] Dot Product of the Y axis from the Forward Vector
        (LocalVariables|SetLinearYAxisEnd (+ (LocalVariables|GetDebugSystemCurrentLoc) (* _skelmeshcomponent_fwdvector (* _returnvalue_3 (Variables|GetDebug%20Scalar)))))
        (LocalVariables|SetComponentFwdVectorDot _returnvalue_3)
        ;[1 1 1] Dot Product of the Z axis from the Up Vector
        (LocalVariables|SetLinearZAxisEnd (+ (LocalVariables|GetDebugSystemCurrentLoc) (* _skelmeshcomponent_upvector (* _returnvalue_2 (Variables|GetDebug%20Scalar)))))
        (LocalVariables|SetComponentUpVectorDot _returnvalue_2))
      (:then_4
        ;[1 1 1] Space Switcher
        (switch Utilities|FlowControl|Switch|SwitchonDebugDrawSpace (Variables|GetDebugSpace)))
      (:then_5
        (Rendering|Debug|DrawDebugString 0 (LocalVariables|GetDebugSystemCurrentLoc) "Component") ; Draw the current space
      )
      (:then_6
        ;[1 1 1] Create Linear Debug Draw Arrows for Component Space
        (Rendering|Debug|DrawDebugArrow 0 (LocalVariables|GetDebugSystemCurrentLoc) (LocalVariables|GetLinearYAxisEnd) 0.0 "(R=1.000000,G=0.000000,B=0.000000,A=1.000000)" 0.0 0.5 (Variables|GetForeground)) ; Linear X Axis Lines
        (Rendering|Debug|DrawDebugArrow 0 (LocalVariables|GetDebugSystemCurrentLoc) (LocalVariables|GetLinearXAxisEnd) 0.0 "(R=0.174389,G=1.000000,B=0.000000,A=1.000000)" 0.0 0.5 (Variables|GetForeground)) ; Linear Y Axis Lines
        (Rendering|Debug|DrawDebugArrow 0 (LocalVariables|GetDebugSystemCurrentLoc) (LocalVariables|GetLinearZAxisEnd) 0.0 "(R=0.045509,G=0.000000,B=1.000000,A=1.000000)" 0.0 0.5 (Variables|GetForeground)) ; Linear Z Axis Lines
      )
      (:then_7
        ;[1 1 1] Draw the core coords system and the per axis values
        (Rendering|Debug|DrawDebugString 0 (LocalVariables|GetYAxisTextLoc) (Utilities|String|ToString(Float) (LocalVariables|GetComponentFwdVectorDot)) 0 "(R=1.000000,G=0.000000,B=0.000000,A=1.000000)") ; Pawn 2D Speed Text
        (Rendering|Debug|DrawDebugString 0 (LocalVariables|GetXAxisTextLoc) (Utilities|String|ToString(Float) (LocalVariables|GetComponentRightVectorDot)) 0 "(R=0.000000,G=1.000000,B=0.000867,A=1.000000)") ; Pawn 2D Speed Text
        (Rendering|Debug|DrawDebugString 0 (LocalVariables|GetZAxisTextLoc) (Utilities|String|ToString(Float) (LocalVariables|GetComponentUpVectorDot)) 0 "(R=0.000000,G=0.000000,B=1.000000,A=1.000000)") ; Pawn 2D Speed Text
        (Rendering|Debug|DrawDebugCoordinateSystem 0 (LocalVariables|GetDebugSystemCurrentLoc) (LocalVariables|GetDebugSystemCurrentRotation) 1e+01 0.0 0.0 "Foreground"))
      (:then_8
        ;[1 1 1] Create Angular Debug Draw
        (bind _output_get (LocalVariables|SetRotationRateEnd (* _yaw 3.6e+02)))
        (Rendering|Debug|DrawDebugString 0 (+ _returnvalue_13 "0.000000,0.000000,-10.000000") (Utilities|String|ToString(Float) _output_get) 0 "(R=0.766945,G=0.000000,B=0.833333,A=1.000000)") ; Pawn Rotation Rate in Degrees Text
        (Rendering|Debug|DrawDebugCircle 0 _returnvalue_14 (* _yaw 15.0) 24 "(R=0.786119,G=0.000000,B=0.854167,A=1.000000)" 0.0 0.5 "1.000000,0.000000,0.000000" "0.000000,1.000000,0.000000" false (Variables|GetForeground)) ; Pawn Rotation Rate Circle
        (Rendering|Debug|DrawDebugCircle 0 _returnvalue_14 15.0 24 "(R=0.138814,G=0.057726,B=0.145833,A=1.000000)" 0.0 0.3 "1.000000,0.000000,0.000000" "0.000000,1.000000,0.000000" false (Variables|GetForeground)) ; Pawn 360 deg Rotation Rate Reference Circle
      )
      (:then_9
        _))
    (else
      _))
  (return (LocalVariables|GetLinearXAxisEnd) (LocalVariables|GetLinearYAxisEnd) (LocalVariables|GetLinearZAxisEnd) (LocalVariables|GetRotationRateEnd) (LocalVariables|GetBoneY%20Axis) (LocalVariables|GetBoneX%20Axis) (LocalVariables|GetBoneZ%20Axis) (LocalVariables|GetBoneLocationPassed) (LocalVariables|GetSkelMeshComponent_FwdVector) (LocalVariables|GetSkelMeshComponent_RightVector) (LocalVariables|GetSkelMeshComponent_UpVector) (LocalVariables|GetComponentRightVectorDot) (LocalVariables|GetComponentFwdVectorDot) (LocalVariables|GetComponentUpVectorDot) (LocalVariables|GetPawnVelocity)))

(fn DrawJointDebugFunction
  (params (SkelMeshComponent :type Scene%20Component%20Object%20Reference) (JointOffset :type Vector) (JointName :type Array%20of%20Names) (Scale :type Float%20%28double-precision%29 :default 1e+01) (Duration :type Float%20%28double-precision%29) (Thickness :type Float%20%28double-precision%29 :default 0.7) (DebugDraw :type Boolean :default true) (Foreground :type EDrawDebugSceneDepthPriorityGroup%20Enum :default "Foreground"))
  (Utilities|IsValid SkelMeshComponent
    (:Is%20Valid
      (if DebugDraw
        (for _array_element JointName
          (Rendering|Debug|DrawDebugCoordinateSystem 0 (+ (Transformation|GetSocketLocation SkelMeshComponent _array_element) JointOffset) (Transformation|GetSocketRotation SkelMeshComponent _array_element) Scale Duration Thickness Foreground)
          (Rendering|Debug|DrawDebugString 0 (+ (Transformation|GetSocketLocation SkelMeshComponent _array_element) JointOffset) (Utilities|String|ToString(Name) _array_element)))))))

(fn DrawVectorFromJointFunction
  (params (SkelMeshComponent :type Scene%20Component%20Object%20Reference) (JointName :type Name) (LineEnd :type Vector) (JointOffset :type Vector) (LineColor :type Linear%20Color%20Structure :default "(R=1.000000,G=0.000000,B=0.000000,A=1.000000)") (Thickness :type Float%20%28double-precision%29 :default 2.0) (Foreground :type EDrawDebugSceneDepthPriorityGroup%20Enum :default "Foreground"))
  (Utilities|IsValid SkelMeshComponent
    (:Is%20Valid
      (Rendering|Debug|DrawDebugArrow 0 (+ (Transformation|GetSocketLocation SkelMeshComponent JointName) JointOffset) (+ (+ (Transformation|GetSocketLocation SkelMeshComponent JointName) JointOffset) LineEnd) 0.0 LineColor 0.0 Thickness Foreground))))

(fn FN_PSDPoseReaderFNC
  (params (startJnt :type String :default "upperarm_l") (endJnt :type String :default "lowerarm_l") (poseJnt :type String :default "pose_arm_up_l") (angle :type Float%20%28double-precision%29 :default 45.0) (debug :type Boolean) (skeletalMeshComponent :type Skeletal%20Mesh%20Component%20Object%20Reference))
  (bind _returnvalue (Utilities|String|StringToName poseJnt))
  (bind _returnvalue_2 (Utilities|String|StringToName startJnt))
  (bind _returnvalue_3 (Animation|Utilities|GetDirectionBetweenSockets skeletalMeshComponent _returnvalue_2 _returnvalue))
  (bind _returnvalue_4 (Utilities|String|StringToName startJnt))
  (bind _returnvalue_5 (Utilities|String|StringToName endJnt))
  (bind _returnvalue_6 (Utilities|String|StringToName startJnt))
  (bind _returnvalue_7 (Animation|Utilities|GetDirectionBetweenSockets skeletalMeshComponent _returnvalue_6 _returnvalue_5))
  (bind _angle__degrees (Class|AbilityFunctions|GetAngleBetweenTwoVectors _returnvalue_7 _returnvalue_3))
  (bind _returnvalue_8 (Transformation|GetSocketLocation skeletalMeshComponent _returnvalue_4))
  (bind _returnvalue_9 (/ _angle__degrees angle))
  (bind _returnvalue_10 (- 1.0 _returnvalue_9))
  (Utilities|FlowControl|Sequence
    (:then_0
      ;[1 1 1] Debug
      (if debug
        (Rendering|Debug|DrawDebugArrow 0 _returnvalue_8 (Transformation|GetSocketLocation skeletalMeshComponent (Utilities|String|StringToName poseJnt)) 0.0 "(R=0.169176,G=0.147041,B=1.000000,A=1.000000)" 0.0 1.0)
        (Rendering|Debug|DrawDebugCone 0 _returnvalue_8 _returnvalue_3 1e+02 angle angle 12 "(R=1.000000,G=1.000000,B=1.000000,A=1.000000)" 0.0 0.3)))
    (:then_1
      (if (>= _returnvalue_10 0.0)
        (bind _output_get (LocalVariables|Setpsd_alpha _returnvalue_10))
        (else
          (LocalVariables|Setpsd_alpha 0.0)))
      (return _output_get))))

(fn GetMotionVectorForTransformFunction
  (params (Skeletal%20Mesh :type Scene%20Component%20Object%20Reference) (Transform%20Name :type Name) (Previous%20Transform%20Location :type Vector%20%28by%20ref%29))
  ;[1 1 1] Nativized Function
  (bind (_outtransformmotionvector _outtransformlocation _outtransform) (Animation|Material|GetMotionVectorforTransform Skeletal%20Mesh Transform%20Name Previous%20Transform%20Location))
  (return _outtransformmotionvector _outtransformlocation _outtransform))

(fn GetJointLocationRelativetoBone
  (params (Target%20Joint :type Name :default "pelvis") (Relative%20Joint :type Name :default "pelvis") (Offset :type Vector) (SkelMeshComponent :type Mesh%20Component%20Object%20Reference) (Debug%20Draw :type Boolean))
  (bind _target_joint (Variables|GetTarget%20Joint))
  (bind _skelmeshcomponent (Variables|GetSkelMeshComponent))
  (bind _returnvalue (Transformation|GetSocketLocation _skelmeshcomponent _target_joint))
  (bind _returnvalue_2 (Animation|Utilities|AlignVectorwithJoint (Variables|GetOffset) (Variables|GetSkelMeshComponent) 0 (Variables|GetRelative%20Joint)))
  (if (Variables|GetDebug%20Draw)
    ;[1 1 1] Debug Draw
    (Rendering|Debug|DrawDebugCoordinateSystem 0 (+ _returnvalue _returnvalue_2) (Transformation|GetSocketRotation _skelmeshcomponent _target_joint) 5e+01 0.0 3.0)
    (Default|DrawJointDebugFunction (Variables|GetSkelMeshComponent) "0, 0, 0" (Utilities|Array|MakeArray (Variables|GetTarget%20Joint)) 1e+02))
  (return (+ _returnvalue _returnvalue_2)))

(fn GetWPORegionalParamNamesFunction
  (params (Region :type Integer%20%28by%20ref%29))
  ;[1 1 1] Nativized Function
  (bind (_outvelocityname _outbonepositionname _outbonevectorname) (Animation|Material|GetWPORegionalParamNames Region))
  (return _outvelocityname _outbonepositionname _outbonevectorname))

(fn GravityOverrideSetterFunctionDEPRECATED
  (params (CharacterPartAnimInstance :type Custom%20Character%20Part%20Anim%20Instance%20Object%20Reference) (SkelMeshComponent :type Mesh%20Component%20Object%20Reference) (GravityOverride%20Settings :type Gravity%20Override%20Setting%20Struct%20Structure) (GameTimeInSeconds :type Float%20%28double-precision%29))
  (bind (_windfrequency_28_9cbe863e417fc965c07809a2fe55adfd _windamplitude_31_6a3dfe4a481ad4d7cb18d79f95cc88e8 _windoffset_32_c5f4d16f48d0bedc3f0fa38b2648511b _gravityoverride_15_9c3cf2b14934a1ac917c4c99f03afc6e _jointname_27_c74f3b2245ea319de82a629ff3c17cca) (Utilities|Struct|BreakGravityOverrideSettingStruct GravityOverride%20Settings))
  ;[1 1 1] Combine and apply the values for all of the settings before adding them to the Struct
  (LocalVariables|SetGravityOverrideSettings%20Local GravityOverride%20Settings)
  (bind _vectorout (CallFunction|RandomVectorFunction 0 SkelMeshComponent _jointname_27_c74f3b2245ea319de82a629ff3c17cca GameTimeInSeconds 0.0 1.0 _windfrequency_28_9cbe863e417fc965c07809a2fe55adfd _windamplitude_31_6a3dfe4a481ad4d7cb18d79f95cc88e8 _windoffset_32_c5f4d16f48d0bedc3f0fa38b2648511b)) ; Create wind turbulence for freefalling
  (bind _returnvalue (Animation|Utilities|AlignVectorwithJoint _gravityoverride_15_9c3cf2b14934a1ac917c4c99f03afc6e SkelMeshComponent 0 _jointname_27_c74f3b2245ea319de82a629ff3c17cca))
  (bind _output_get (LocalVariables|SetGravity%20Override%20Local (+ _vectorout _returnvalue)))
  (bind _structout (Utilities|Struct|SetmembersinGravityOverrideSettingStruct (LocalVariables|GetGravityOverrideSettings%20Local) _output_get))
  (return _structout _output_get))

(fn GravityOverrideStateFunctionDEPRECATED
  (params (CharacterPartAnimInstance :type Custom%20Character%20Part%20Anim%20Instance%20Object%20Reference) (GameTimeInSeconds :type Float%20%28double-precision%29) (GravityOverrideParams :type Gravity%20Override%20Params%20Struct%20Structure%20%28by%20ref%29) (RBAN%20Params :type RBANParams%20Struct%20Structure%20%28by%20ref%29) (DebugDraw :type Boolean) (DebugJointName :type Name :default "root"))
  (bind _gravityoverridesettingslocal (LocalVariables|GetGravityOverrideSettingsLocal))
  (bind (_windfrequency_28_9cbe863e417fc965c07809a2fe55adfd _windamplitude_31_6a3dfe4a481ad4d7cb18d79f95cc88e8 _windoffset_32_c5f4d16f48d0bedc3f0fa38b2648511b _gravityoverride_15_9c3cf2b14934a1ac917c4c99f03afc6e _jointname_27_c74f3b2245ea319de82a629ff3c17cca) (Utilities|Struct|BreakGravityOverrideSettingStruct _gravityoverridesettingslocal))
  (bind _gravityoverridesettingslocal_2 (LocalVariables|GetGravityOverrideSettingsLocal))
  (bind (_windfrequency_28_9cbe863e417fc965c07809a2fe55adfd_2 _windamplitude_31_6a3dfe4a481ad4d7cb18d79f95cc88e8_2 _windoffset_32_c5f4d16f48d0bedc3f0fa38b2648511b_2 _gravityoverride_15_9c3cf2b14934a1ac917c4c99f03afc6e_2 _jointname_27_c74f3b2245ea319de82a629ff3c17cca_2) (Utilities|Struct|BreakGravityOverrideSettingStruct _gravityoverridesettingslocal_2))
  (bind (_skydive_goingup_28_c026bda74bfbaba6672dfe8ffdaafc1b _skydive_floatingdown_29_2cbf3d3f421785f8994502931454a81f _skydive_divingdown_30_ec7bec4d406a8a98e60d7b9374c5e457 _skydive_parachuting_31_a88c4a0a407f497561d31287cb20250f _onground_standing_32_267db1464d0d3a29d99e94adfcb4afec _onground_moving_33_807a66b64d7b3b2fcf080b833b8a6254 _emoteormelee_34_cfcf66b5492987cbd638ec98eb34d572 _swimming_35_33ce87ce4d4f712b1384e29115845962 _front_end_36_c7a4562a48e9bc9b4cd34d963cfc6490 _dbno_40_83703b55467f014d644d02b56709a2e6 _dbno_carried_41_2c3d52af4b7eb85dba7c108948e58b3a _onground_falling_45_dc16623b41084152717acbaf5dce8a91 _ziplining_46_0f05199d45821770f22e0b957551feb7 _skydive_surfing_49_5869aa5846bee7c351769c951d56a34f _riding_standing_54_c62e341c48cf5ba85acb799ee3c7653e _riding_moving_55_86cd46c94e5b465a175a6b939db193bd _riding_moving_targeting_58_38650d3547535fb293e61cbcf1b753bf _tactical_sprint_57_7b4700a94af4b34179554c8ec4eb52e0 _rail_grinding_61_97cb95534c5559baf29fe49c016359fa _dbno_tumbling_67_e1fc5a014f4491936d6121a8bbd4fbb5) (Utilities|Struct|BreakGravityOverrideParamsStruct (Variables|GetGravityOverrideParams)))
  (bind (_skydive_goingup_56_c026bda74bfbaba6672dfe8ffdaafc1b _skydive_floatingdown_57_2cbf3d3f421785f8994502931454a81f _skydive_divingdown_58_ec7bec4d406a8a98e60d7b9374c5e457 _skydive_parachuting_60_a88c4a0a407f497561d31287cb20250f _onground_standing_59_267db1464d0d3a29d99e94adfcb4afec _onground_moving_65_807a66b64d7b3b2fcf080b833b8a6254 _onground_crouch_moving_61_ce0b8bfb4e3417c0f790f7ae38dd7735 _emoteormelee_62_cfcf66b5492987cbd638ec98eb34d572 _swimming_64_33ce87ce4d4f712b1384e29115845962 _front_end_63_c7a4562a48e9bc9b4cd34d963cfc6490 _dbno_68_2d70d4004d0e9b117f6ce9a09c362b74 _dbno_carried_67_99829b99495fcfe3488a9a86b549ece5 _onground_falling_66_cdaa0f4d476157e4137548bdce8776f2 _ziplining_75_bccdb7e1431a4c174f1040abc81fd8a8 _skydive_surfing_74_408e65ff484aa54111e1ed8a69d8240d _riding_standing_69_952a2a43408c37c44e4cbfae3f3001b0 _riding_moving_72_ee7a01794460b9753e5076a5a3b3080a _riding_moving_targeting_73_2f57fd9e477c515d70b3da9d75453357 _tactical_sprint_70_5b2273a1499d34bdec52e6a8e3f274d4 _rail_grinding_71_deeb00a04db992df5eeb40bd6518ea4e _dbno_tumbling_81_3b07173d479217c14597d9820eeb8714) (Utilities|Struct|BreakRBANParamsStruct (Variables|GetRBAN%20Params)))
  (bind (_skydive_goingup_28_c026bda74bfbaba6672dfe8ffdaafc1b_2 _skydive_floatingdown_29_2cbf3d3f421785f8994502931454a81f_2 _skydive_divingdown_30_ec7bec4d406a8a98e60d7b9374c5e457_2 _skydive_parachuting_31_a88c4a0a407f497561d31287cb20250f_2 _onground_standing_32_267db1464d0d3a29d99e94adfcb4afec_2 _onground_moving_33_807a66b64d7b3b2fcf080b833b8a6254_2 _emoteormelee_34_cfcf66b5492987cbd638ec98eb34d572_2 _swimming_35_33ce87ce4d4f712b1384e29115845962_2 _front_end_36_c7a4562a48e9bc9b4cd34d963cfc6490_2 _dbno_40_83703b55467f014d644d02b56709a2e6_2 _dbno_carried_41_2c3d52af4b7eb85dba7c108948e58b3a_2 _onground_falling_45_dc16623b41084152717acbaf5dce8a91_2 _ziplining_46_0f05199d45821770f22e0b957551feb7_2 _skydive_surfing_49_5869aa5846bee7c351769c951d56a34f_2 _riding_standing_54_c62e341c48cf5ba85acb799ee3c7653e_2 _riding_moving_55_86cd46c94e5b465a175a6b939db193bd_2 _riding_moving_targeting_58_38650d3547535fb293e61cbcf1b753bf_2 _tactical_sprint_57_7b4700a94af4b34179554c8ec4eb52e0_2 _rail_grinding_61_97cb95534c5559baf29fe49c016359fa_2 _dbno_tumbling_67_e1fc5a014f4491936d6121a8bbd4fbb5_2 _dbno_rollfwd_68_b337f650476038e91401cab18ca91129) (Utilities|Struct|BreakGravityOverrideParamsStruct (Variables|GetGravityOverrideParams)))
  (bind (_skydive_goingup_56_c026bda74bfbaba6672dfe8ffdaafc1b_2 _skydive_floatingdown_57_2cbf3d3f421785f8994502931454a81f_2 _skydive_divingdown_58_ec7bec4d406a8a98e60d7b9374c5e457_2 _skydive_parachuting_60_a88c4a0a407f497561d31287cb20250f_2 _onground_standing_59_267db1464d0d3a29d99e94adfcb4afec_2 _onground_moving_65_807a66b64d7b3b2fcf080b833b8a6254_2 _onground_crouch_moving_61_ce0b8bfb4e3417c0f790f7ae38dd7735_2 _emoteormelee_62_cfcf66b5492987cbd638ec98eb34d572_2 _swimming_64_33ce87ce4d4f712b1384e29115845962_2 _front_end_63_c7a4562a48e9bc9b4cd34d963cfc6490_2 _dbno_68_2d70d4004d0e9b117f6ce9a09c362b74_2 _dbno_carried_67_99829b99495fcfe3488a9a86b549ece5_2 _onground_falling_66_cdaa0f4d476157e4137548bdce8776f2_2 _ziplining_75_bccdb7e1431a4c174f1040abc81fd8a8_2 _skydive_surfing_74_408e65ff484aa54111e1ed8a69d8240d_2 _riding_standing_69_952a2a43408c37c44e4cbfae3f3001b0_2 _riding_moving_72_ee7a01794460b9753e5076a5a3b3080a_2 _riding_moving_targeting_73_2f57fd9e477c515d70b3da9d75453357_2 _tactical_sprint_70_5b2273a1499d34bdec52e6a8e3f274d4_2 _rail_grinding_71_deeb00a04db992df5eeb40bd6518ea4e_2 _dbno_tumbling_81_3b07173d479217c14597d9820eeb8714_2 _dbno_rollfwd_82_df9da72b471974ea9d3aa9904059aeaa) (Utilities|Struct|BreakRBANParamsStruct (Variables|GetRBAN%20Params)))
  (bind (_skydive_goingup_28_c026bda74bfbaba6672dfe8ffdaafc1b_3 _skydive_floatingdown_29_2cbf3d3f421785f8994502931454a81f_3 _skydive_divingdown_30_ec7bec4d406a8a98e60d7b9374c5e457_3 _skydive_parachuting_31_a88c4a0a407f497561d31287cb20250f_3 _onground_standing_32_267db1464d0d3a29d99e94adfcb4afec_3 _onground_moving_33_807a66b64d7b3b2fcf080b833b8a6254_3 _emoteormelee_34_cfcf66b5492987cbd638ec98eb34d572_3 _swimming_35_33ce87ce4d4f712b1384e29115845962_3 _front_end_36_c7a4562a48e9bc9b4cd34d963cfc6490_3 _dbno_40_83703b55467f014d644d02b56709a2e6_3 _dbno_carried_41_2c3d52af4b7eb85dba7c108948e58b3a_3 _onground_falling_45_dc16623b41084152717acbaf5dce8a91_3 _ziplining_46_0f05199d45821770f22e0b957551feb7_3 _skydive_surfing_49_5869aa5846bee7c351769c951d56a34f_3 _riding_standing_54_c62e341c48cf5ba85acb799ee3c7653e_3 _riding_moving_55_86cd46c94e5b465a175a6b939db193bd_3 _riding_moving_targeting_58_38650d3547535fb293e61cbcf1b753bf_3 _tactical_sprint_57_7b4700a94af4b34179554c8ec4eb52e0_3 _rail_grinding_61_97cb95534c5559baf29fe49c016359fa_3 _dbno_tumbling_67_e1fc5a014f4491936d6121a8bbd4fbb5_3 _dbno_rollfwd_68_b337f650476038e91401cab18ca91129_2 _dbno_rollbwd_69_a313665b4e45cc3f8549e6a2fa8d3c2c) (Utilities|Struct|BreakGravityOverrideParamsStruct (Variables|GetGravityOverrideParams)))
  (bind (_skydive_goingup_56_c026bda74bfbaba6672dfe8ffdaafc1b_3 _skydive_floatingdown_57_2cbf3d3f421785f8994502931454a81f_3 _skydive_divingdown_58_ec7bec4d406a8a98e60d7b9374c5e457_3 _skydive_parachuting_60_a88c4a0a407f497561d31287cb20250f_3 _onground_standing_59_267db1464d0d3a29d99e94adfcb4afec_3 _onground_moving_65_807a66b64d7b3b2fcf080b833b8a6254_3 _onground_crouch_moving_61_ce0b8bfb4e3417c0f790f7ae38dd7735_3 _emoteormelee_62_cfcf66b5492987cbd638ec98eb34d572_3 _swimming_64_33ce87ce4d4f712b1384e29115845962_3 _front_end_63_c7a4562a48e9bc9b4cd34d963cfc6490_3 _dbno_68_2d70d4004d0e9b117f6ce9a09c362b74_3 _dbno_carried_67_99829b99495fcfe3488a9a86b549ece5_3 _onground_falling_66_cdaa0f4d476157e4137548bdce8776f2_3 _ziplining_75_bccdb7e1431a4c174f1040abc81fd8a8_3 _skydive_surfing_74_408e65ff484aa54111e1ed8a69d8240d_3 _riding_standing_69_952a2a43408c37c44e4cbfae3f3001b0_3 _riding_moving_72_ee7a01794460b9753e5076a5a3b3080a_3 _riding_moving_targeting_73_2f57fd9e477c515d70b3da9d75453357_3 _tactical_sprint_70_5b2273a1499d34bdec52e6a8e3f274d4_3 _rail_grinding_71_deeb00a04db992df5eeb40bd6518ea4e_3 _dbno_tumbling_81_3b07173d479217c14597d9820eeb8714_3 _dbno_rollfwd_82_df9da72b471974ea9d3aa9904059aeaa_2 _dbno_rollbwd_83_8f3a6ada4bfff38d49c5d480db2b3929) (Utilities|Struct|BreakRBANParamsStruct (Variables|GetRBAN%20Params)))
  (bind (_skydive_goingup_28_c026bda74bfbaba6672dfe8ffdaafc1b_4 _skydive_floatingdown_29_2cbf3d3f421785f8994502931454a81f_4 _skydive_divingdown_30_ec7bec4d406a8a98e60d7b9374c5e457_4 _skydive_parachuting_31_a88c4a0a407f497561d31287cb20250f_4 _onground_standing_32_267db1464d0d3a29d99e94adfcb4afec_4 _onground_moving_33_807a66b64d7b3b2fcf080b833b8a6254_4 _emoteormelee_34_cfcf66b5492987cbd638ec98eb34d572_4 _swimming_35_33ce87ce4d4f712b1384e29115845962_4 _front_end_36_c7a4562a48e9bc9b4cd34d963cfc6490_4 _dbno_40_83703b55467f014d644d02b56709a2e6_4 _dbno_carried_41_2c3d52af4b7eb85dba7c108948e58b3a_4 _onground_falling_45_dc16623b41084152717acbaf5dce8a91_4 _ziplining_46_0f05199d45821770f22e0b957551feb7_4 _skydive_surfing_49_5869aa5846bee7c351769c951d56a34f_4 _riding_standing_54_c62e341c48cf5ba85acb799ee3c7653e_4 _riding_moving_55_86cd46c94e5b465a175a6b939db193bd_4 _riding_moving_targeting_58_38650d3547535fb293e61cbcf1b753bf_4 _tactical_sprint_57_7b4700a94af4b34179554c8ec4eb52e0_4 _rail_grinding_61_97cb95534c5559baf29fe49c016359fa_4 _dbno_tumbling_67_e1fc5a014f4491936d6121a8bbd4fbb5_4 _dbno_rollfwd_68_b337f650476038e91401cab18ca91129_3 _dbno_rollbwd_69_a313665b4e45cc3f8549e6a2fa8d3c2c_2 _dbno_impulse_70_d4cf55bf4a37c2025b3653ba69764562) (Utilities|Struct|BreakGravityOverrideParamsStruct (Variables|GetGravityOverrideParams)))
  (bind (_skydive_goingup_56_c026bda74bfbaba6672dfe8ffdaafc1b_4 _skydive_floatingdown_57_2cbf3d3f421785f8994502931454a81f_4 _skydive_divingdown_58_ec7bec4d406a8a98e60d7b9374c5e457_4 _skydive_parachuting_60_a88c4a0a407f497561d31287cb20250f_4 _onground_standing_59_267db1464d0d3a29d99e94adfcb4afec_4 _onground_moving_65_807a66b64d7b3b2fcf080b833b8a6254_4 _onground_crouch_moving_61_ce0b8bfb4e3417c0f790f7ae38dd7735_4 _emoteormelee_62_cfcf66b5492987cbd638ec98eb34d572_4 _swimming_64_33ce87ce4d4f712b1384e29115845962_4 _front_end_63_c7a4562a48e9bc9b4cd34d963cfc6490_4 _dbno_68_2d70d4004d0e9b117f6ce9a09c362b74_4 _dbno_carried_67_99829b99495fcfe3488a9a86b549ece5_4 _onground_falling_66_cdaa0f4d476157e4137548bdce8776f2_4 _ziplining_75_bccdb7e1431a4c174f1040abc81fd8a8_4 _skydive_surfing_74_408e65ff484aa54111e1ed8a69d8240d_4 _riding_standing_69_952a2a43408c37c44e4cbfae3f3001b0_4 _riding_moving_72_ee7a01794460b9753e5076a5a3b3080a_4 _riding_moving_targeting_73_2f57fd9e477c515d70b3da9d75453357_4 _tactical_sprint_70_5b2273a1499d34bdec52e6a8e3f274d4_4 _rail_grinding_71_deeb00a04db992df5eeb40bd6518ea4e_4 _dbno_tumbling_81_3b07173d479217c14597d9820eeb8714_4 _dbno_rollfwd_82_df9da72b471974ea9d3aa9904059aeaa_3 _dbno_rollbwd_83_8f3a6ada4bfff38d49c5d480db2b3929_2 _dbno_impulse_84_d5c88e5b4e75537bf523069dc16a8de2) (Utilities|Struct|BreakRBANParamsStruct (Variables|GetRBAN%20Params)))
  (bind _characterpartaniminstance (Variables|GetCharacterPartAnimInstance))
  (bind _returnvalue (CustomCharacter|GetPartSkeletalMeshComponent _characterpartaniminstance))
  (bind _ownerskelmeshcomp (LocalVariables|GetOwnerSkelMeshComp))
  (bind _output_get (LocalVariables|SetOwnerSkelMeshComp (CustomCharacter|GetOwnerSkeletalMeshComponent (Variables|GetCharacterPartAnimInstance))))
  (Utilities|IsValid _output_get
    (:Is%20Valid
      (Utilities|FlowControl|Sequence
        (:then_0
          ;[1 1 1] State Logic - Switch order causes messy lines but this is an optimization - most common states should be at the top
          (switch name (Animation|CharacterDynamicsControl|GetCharacterStateforLayer (Variables|GetCharacterPartAnimInstance))
            (:Default
              _)
            (:OnGround_Standing
              _)
            (:OnGround_Moving
              ;[1 1 1] ON Ground Moving
              (LocalVariables|SetCurrentState "OnGround_Moving")
              (LocalVariables|SetGravityOverrideSettingsLocal (Utilities|Struct|BreakGravityOverrideParamsStruct (Variables|GetGravityOverrideParams)))
              (LocalVariables|SetRBAN%20SettingsLocal (Utilities|Struct|BreakRBANParamsStruct (Variables|GetRBAN%20Params))))
            (:Tactical_Sprint
              ;[1 1 1] Tactical Sprint
              (LocalVariables|SetCurrentState "Tactical Sprinting")
              (LocalVariables|SetGravityOverrideSettingsLocal (Utilities|Struct|BreakGravityOverrideParamsStruct (Variables|GetGravityOverrideParams)))
              (LocalVariables|SetRBAN%20SettingsLocal (Utilities|Struct|BreakRBANParamsStruct (Variables|GetRBAN%20Params))))
            (:OnGround_Falling
              ;[1 1 1] OnGround Falling
              (LocalVariables|SetCurrentState "OnGround Falling")
              (LocalVariables|SetGravityOverrideSettingsLocal (Utilities|Struct|BreakGravityOverrideParamsStruct (Variables|GetGravityOverrideParams)))
              (LocalVariables|SetRBAN%20SettingsLocal (Utilities|Struct|BreakRBANParamsStruct (Variables|GetRBAN%20Params))))
            (:OnGround_Moving_HighRotationRate
              _)
            (:Ziplining
              ;[1 1 1] Ziplining
              (LocalVariables|SetCurrentState "Ziplining")
              (LocalVariables|SetGravityOverrideSettingsLocal (Utilities|Struct|BreakGravityOverrideParamsStruct (Variables|GetGravityOverrideParams)))
              (LocalVariables|SetRBAN%20SettingsLocal (Utilities|Struct|BreakRBANParamsStruct (Variables|GetRBAN%20Params))))
            (:Swimming
              ;[1 1 1] Swimming
              (LocalVariables|SetCurrentState "Swimming")
              (LocalVariables|SetGravityOverrideSettingsLocal (Utilities|Struct|BreakGravityOverrideParamsStruct (Variables|GetGravityOverrideParams)))
              (LocalVariables|SetRBAN%20SettingsLocal (Utilities|Struct|BreakRBANParamsStruct (Variables|GetRBAN%20Params))))
            (:Rail_Grinding
              ;[1 1 1] Rail Grinding
              (LocalVariables|SetCurrentState "Rail Grinding")
              (LocalVariables|SetGravityOverrideSettingsLocal (Utilities|Struct|BreakGravityOverrideParamsStruct (Variables|GetGravityOverrideParams)))
              (LocalVariables|SetRBAN%20SettingsLocal (Utilities|Struct|BreakRBANParamsStruct (Variables|GetRBAN%20Params))))
            (:EmoteOrMelee
              ;[1 1 1] Emote or Melee
              (LocalVariables|SetCurrentState "MeleeOrEmote")
              (LocalVariables|SetGravityOverrideSettingsLocal (Utilities|Struct|BreakGravityOverrideParamsStruct (Variables|GetGravityOverrideParams)))
              (LocalVariables|SetRBAN%20SettingsLocal (Utilities|Struct|BreakRBANParamsStruct (Variables|GetRBAN%20Params))))
            (:Skydive_FloatingDown
              ;[1 1 1] SKYDIVING
              (LocalVariables|SetCurrentState "SkyDiving_FloatingDown")
              (LocalVariables|SetGravityOverrideSettingsLocal (Utilities|Struct|BreakGravityOverrideParamsStruct (Variables|GetGravityOverrideParams)))
              (LocalVariables|SetRBAN%20SettingsLocal (Utilities|Struct|BreakRBANParamsStruct (Variables|GetRBAN%20Params))))
            (:Skydive_DivingDown
              (LocalVariables|SetCurrentState "SkyDiving_DivingDown")
              (LocalVariables|SetGravityOverrideSettingsLocal (Utilities|Struct|BreakGravityOverrideParamsStruct (Variables|GetGravityOverrideParams)))
              (LocalVariables|SetRBAN%20SettingsLocal (Utilities|Struct|BreakRBANParamsStruct (Variables|GetRBAN%20Params))))
            (:Skydive_Parachuting
              (LocalVariables|SetCurrentState "SkyDiving_Parachuting")
              (LocalVariables|SetGravityOverrideSettingsLocal (Utilities|Struct|BreakGravityOverrideParamsStruct (Variables|GetGravityOverrideParams)))
              (LocalVariables|SetRBAN%20SettingsLocal (Utilities|Struct|BreakRBANParamsStruct (Variables|GetRBAN%20Params))))
            (:Skydive_ParachutingWithPBA
              _)
            (:Skydive_Umbrella
              _)
            (:Skydive_Surfing
              (LocalVariables|SetCurrentState "SkyDiving_Surfing")
              (LocalVariables|SetGravityOverrideSettingsLocal (Utilities|Struct|BreakGravityOverrideParamsStruct (Variables|GetGravityOverrideParams)))
              (LocalVariables|SetRBAN%20SettingsLocal (Utilities|Struct|BreakRBANParamsStruct (Variables|GetRBAN%20Params))))
            (:Skydive_GoingUp
              (LocalVariables|SetCurrentState "SkyDiving_GoingUp")
              (LocalVariables|SetGravityOverrideSettingsLocal (Utilities|Struct|BreakGravityOverrideParamsStruct (Variables|GetGravityOverrideParams)))
              (LocalVariables|SetRBAN%20SettingsLocal (Utilities|Struct|BreakRBANParamsStruct (Variables|GetRBAN%20Params))))
            (:Riding_Moving
              ;[1 1 1] Riding
              (LocalVariables|SetCurrentState "Riding_Moving")
              (LocalVariables|SetGravityOverrideSettingsLocal (Utilities|Struct|BreakGravityOverrideParamsStruct (Variables|GetGravityOverrideParams)))
              (LocalVariables|SetRBAN%20SettingsLocal (Utilities|Struct|BreakRBANParamsStruct (Variables|GetRBAN%20Params))))
            (:Riding_Moving_Targeting
              (LocalVariables|SetCurrentState "Riding_Moving_Targeting")
              (LocalVariables|SetGravityOverrideSettingsLocal (Utilities|Struct|BreakGravityOverrideParamsStruct (Variables|GetGravityOverrideParams)))
              (LocalVariables|SetRBAN%20SettingsLocal (Utilities|Struct|BreakRBANParamsStruct (Variables|GetRBAN%20Params))))
            (:Riding_Standing
              (LocalVariables|SetCurrentState "Riding_Standing")
              (LocalVariables|SetGravityOverrideSettingsLocal (Utilities|Struct|BreakGravityOverrideParamsStruct (Variables|GetGravityOverrideParams)))
              (LocalVariables|SetRBAN%20SettingsLocal (Utilities|Struct|BreakRBANParamsStruct (Variables|GetRBAN%20Params))))
            (:DBNO
              ;[1 1 1] DBNO
              (LocalVariables|SetCurrentState "DBNO")
              (LocalVariables|SetGravityOverrideSettingsLocal (Utilities|Struct|BreakGravityOverrideParamsStruct (Variables|GetGravityOverrideParams)))
              (LocalVariables|SetRBAN%20SettingsLocal (Utilities|Struct|BreakRBANParamsStruct (Variables|GetRBAN%20Params))))
            (:DBNO_Carried
              ;[1 1 1] DBNO Carried
              (LocalVariables|SetCurrentState "DBNO Carried")
              (LocalVariables|SetGravityOverrideSettingsLocal (Utilities|Struct|BreakGravityOverrideParamsStruct (Variables|GetGravityOverrideParams)))
              (LocalVariables|SetRBAN%20SettingsLocal (Utilities|Struct|BreakRBANParamsStruct (Variables|GetRBAN%20Params))))
            (:DBNO_Tumbling
              ;[1 1 1] DBNO Tumble
              (LocalVariables|SetCurrentState "DBNO Tumble")
              (LocalVariables|SetGravityOverrideSettingsLocal _dbno_tumbling_67_e1fc5a014f4491936d6121a8bbd4fbb5)
              ;[1 1 1] RBAN SimSpace Settings
              (LocalVariables|SetRBAN%20SettingsLocal _dbno_tumbling_81_3b07173d479217c14597d9820eeb8714))
            (:DBNO_RollFwd
              ;[1 1 1] DBNO Roll Fwd
              (LocalVariables|SetCurrentState "DBNO Roll Fwd")
              (LocalVariables|SetGravityOverrideSettingsLocal _dbno_rollfwd_68_b337f650476038e91401cab18ca91129)
              ;[1 1 1] RBAN SimSpace Settings
              (LocalVariables|SetRBAN%20SettingsLocal _dbno_rollfwd_82_df9da72b471974ea9d3aa9904059aeaa))
            (:DBNO_RollBwd
              ;[1 1 1] DBNO Roll Bwd
              (LocalVariables|SetCurrentState "DBNO Roll Bwd")
              (LocalVariables|SetGravityOverrideSettingsLocal _dbno_rollbwd_69_a313665b4e45cc3f8549e6a2fa8d3c2c)
              ;[1 1 1] RBAN SimSpace Settings
              (LocalVariables|SetRBAN%20SettingsLocal _dbno_rollbwd_83_8f3a6ada4bfff38d49c5d480db2b3929))
            (:DBNO_Impulse
              ;[1 1 1] DBNO Impulse
              (LocalVariables|SetCurrentState "DBNO Impulse")
              (LocalVariables|SetGravityOverrideSettingsLocal _dbno_impulse_70_d4cf55bf4a37c2025b3653ba69764562)
              ;[1 1 1] RBAN SimSpace Settings
              (LocalVariables|SetRBAN%20SettingsLocal _dbno_impulse_84_d5c88e5b4e75537bf523069dc16a8de2))
            (:DBNO_HighRotationRate
              _)
            (:Front_End
              ;[1 1 1] Front End
              (LocalVariables|SetCurrentState "FrontEnd")
              (LocalVariables|SetGravityOverrideSettingsLocal (Utilities|Struct|BreakGravityOverrideParamsStruct (Variables|GetGravityOverrideParams)))
              (LocalVariables|SetRBAN%20SettingsLocal (Utilities|Struct|BreakRBANParamsStruct (Variables|GetRBAN%20Params)))))
          ;[1 1 1] On Ground Standing
          (LocalVariables|SetCurrentState "OnGround_Standing")
          (LocalVariables|SetGravityOverrideSettingsLocal (Utilities|Struct|BreakGravityOverrideParamsStruct (Variables|GetGravityOverrideParams)))
          (LocalVariables|SetRBAN%20SettingsLocal (Utilities|Struct|BreakRBANParamsStruct (Variables|GetRBAN%20Params))))
        (:then_1
          ;[1 1 1] Cacluate the Gravity Override (inlined logic from GravityOverrideSetterFunction).
          (bind _vectorout (CallFunction|RandomVectorFunction 0 _ownerskelmeshcomp _jointname_27_c74f3b2245ea319de82a629ff3c17cca_2 (Variables|GetGameTimeInSeconds) 0.0 1.0 _windfrequency_28_9cbe863e417fc965c07809a2fe55adfd_2 _windamplitude_31_6a3dfe4a481ad4d7cb18d79f95cc88e8_2 _windoffset_32_c5f4d16f48d0bedc3f0fa38b2648511b_2)) ; Create wind turbulence for freefalling
          (bind _returnvalue_2 (Animation|Utilities|AlignVectorwithJoint _gravityoverride_15_9c3cf2b14934a1ac917c4c99f03afc6e_2 _ownerskelmeshcomp 0 _jointname_27_c74f3b2245ea319de82a629ff3c17cca_2))
          (bind _output_get_2 (LocalVariables|SetGravityOverrideLocal (+ _vectorout _returnvalue_2)))
          (Utilities|Struct|SetmembersinGravityOverrideSettingStruct (LocalVariables|GetGravityOverrideSettingsLocal) _output_get_2))
        (:then_2
          ;[1 1 1] Debug Draw
          (if (Variables|GetDebugDraw)
            (Default|DrawVectorfromJointFunction _returnvalue (Variables|GetDebugJointName) _gravityoverride_15_9c3cf2b14934a1ac917c4c99f03afc6e "0, 0, 0" "(R=1.000000,G=0.000000,B=0.000000,A=1.000000)" 1.0)
            (Default|DrawJointDebugFunction _returnvalue "0, 0, 0" (Utilities|Array|MakeArray _jointname_27_c74f3b2245ea319de82a629ff3c17cca))))
        (:then_3
          ;[1 1 1] Return
          (return (LocalVariables|GetGravityOverrideLocal) (LocalVariables|GetCurrentState) (LocalVariables|GetGravityOverrideSettingsLocal) (LocalVariables|GetRBAN%20SettingsLocal)))))))

(fn PlayTransformationMontage
  (params (FortPlayerPawn :type Fort%20Player%20Pawn%20Object%20Reference))
  (bind _mesh (Variables|/Script/Engine.Character|GetMesh FortPlayerPawn))
  (bind _returnvalue (Components|SkeletalMesh|GetAnimInstance _mesh))
  (bind _transformationmontage (Variables|/Script/FortniteGame.FortPlayerPawn|GetTransformationMontage FortPlayerPawn))
  (Utilities|IsValid FortPlayerPawn
    (:Is%20Valid
      (Utilities|IsValid _returnvalue
        (:Is%20Valid
          (Utilities|IsValid _transformationmontage
            (:Is%20Valid
              (Animation|Montage|MontagePlay _returnvalue _transformationmontage 1.0 "MontageLength" (Variables|/Script/FortniteGame.FortPlayerPawn|GetTransformationMontageStartTime FortPlayerPawn) true))))))))

(fn PlayerMovementFunctionDEPRECATED
  (params (CharacterPartAnimInstance :type Custom%20Character%20Part%20Anim%20Instance%20Object%20Reference) (X%20Dot%20Product :type Float%20%28double-precision%29 :default 1.0) (Y%20Dot%20Product :type Float%20%28double-precision%29 :default 1.0) (Z%20Dot%20Product :type Float%20%28double-precision%29 :default 1.0) (DrawDebug :type Boolean) (Debug%20Scalar :type Float%20%28double-precision%29 :default 5e+01))
  ;[1 1 1] Nativized Path (other code is left just for reference)
  (bind (_xmovement _ymovement _zmovement) (Animation|Utilities|PlayerMovement CharacterPartAnimInstance X%20Dot%20Product Y%20Dot%20Product Z%20Dot%20Product DrawDebug Debug%20Scalar))
  (return _xmovement _ymovement _zmovement))

(fn RandomVectorFunction
  (params (CharacterPartAnimInstance :type Custom%20Character%20Part%20Anim%20Instance%20Object%20Reference) (SkelMeshComponent :type Mesh%20Component%20Object%20Reference) (JointName :type Name :default "None") (GameTimeInSeconds :type Float%20%28double-precision%29) (TimeOffset :type Float%20%28double-precision%29) (PerlinScale :type Float%20%28double-precision%29 :default 1.0) (Frequency :type Vector :default "0,2.000000,0") (Amplitude :type Vector :default "0,20.000000,0") (Offset :type Vector :default "0,0.000000,0"))
  ;[1 1 1] Nativized Path (other code is just left for reference)
  (bind _returnvalue (Animation|Utilities|GetRandomWindVector GameTimeInSeconds TimeOffset PerlinScale Frequency Amplitude Offset))
  (bind _returnvalue_2 (Animation|Utilities|AlignVectorwithJoint _returnvalue SkelMeshComponent 0 JointName))
  (return _returnvalue_2))

(fn RandomRotatorFunction
  (params (GameTimeInSeconds :type Float%20%28double-precision%29) (PerlinScale :type Float%20%28double-precision%29 :default 1.0) (Frequency :type Rotator) (Amplitude :type Rotator) (Offset :type Rotator) (TimeOffset :type Float%20%28double-precision%29))
  ;[1 1 1] Nativized Path (other code is left just for reference)
  (bind _returnvalue (Animation|Utilities|RandomRotatorFunction GameTimeInSeconds TimeOffset PerlinScale Frequency Amplitude Offset))
  (return _returnvalue))

(fn ScaleJointRotationFunction
  (params (Scale :type Rotator :default "1.000000,1.000000,1.000000") (TargetBone :type Name :default "None") (ReferenceBone :type Name :default "None") (SkelMeshComponent :type Skinned%20Mesh%20Component%20Object%20Reference))
  (return (Math|Rotator|MakeRotator (* (.roll (.rotation (Components|SkinnedMesh|GetDeltaTransformfromRefPose SkelMeshComponent TargetBone ReferenceBone))) (.roll Scale)) (* (.pitch (.rotation (Components|SkinnedMesh|GetDeltaTransformfromRefPose SkelMeshComponent TargetBone ReferenceBone))) (.pitch Scale)) (* (.yaw (.rotation (Components|SkinnedMesh|GetDeltaTransformfromRefPose SkelMeshComponent TargetBone ReferenceBone))) (.yaw Scale))) (.rotation (Components|SkinnedMesh|GetDeltaTransformfromRefPose SkelMeshComponent TargetBone ReferenceBone))))

(fn SetTransformationMontage
  (params (FortPlayerPawn :type Fort%20Player%20Pawn%20Object%20Reference) (Montage :type Anim%20Montage%20Object%20Reference) (StartTime :type Float%20%28double-precision%29))
  (Utilities|IsValid FortPlayerPawn
    (:Is%20Valid
      (Variables|/Script/FortniteGame.FortPlayerPawn|SetTransformationMontage Montage FortPlayerPawn)
      (Variables|/Script/FortniteGame.FortPlayerPawn|SetTransformationMontageStartTime StartTime FortPlayerPawn))))

(fn SetWPOGlobalsFunction
  (params (Material :type Material%20Instance%20Dynamic%20Object%20Reference) (Overlay%20Material :type Material%20Instance%20Dynamic%20Object%20Reference) (Skeletal%20Mesh :type Skeletal%20Mesh%20Component%20Object%20Reference) (Local%20Noise%20Reference%20Transform :type Name :default "root") (Local%20Noise%20Offset%20Multiplier :type Float%20%28double-precision%29 :default 1.0) (Local%20Noise%20Transform%20%20Location :type Vector%20%28by%20ref%29) (Local%20Noise%20Offset :type Vector%20%28by%20ref%29) (Local%20Turbulence%20Vector%20Transform :type Name :default "root") (Local%20Turbulence%20Vector%20Transform%20Axis :type EAxis%20Enum :default "X"))
  ;[1 1 1] Nativized Function
  (Animation|Material|SetWPOGlobalsFunction Material Overlay%20Material Skeletal%20Mesh Local%20Noise%20Reference%20Transform Local%20Noise%20Offset%20Multiplier Local%20Noise%20Transform%20%20Location Local%20Noise%20Offset Local%20Turbulence%20Vector%20Transform Local%20Turbulence%20Vector%20Transform%20Axis)
  (return))

(fn SetWPORegionalFunction
  (params (Material :type Material%20Instance%20Dynamic%20Object%20Reference) (Overlay%20Material :type Material%20Instance%20Dynamic%20Object%20Reference) (Skeletal%20Mesh :type Skeletal%20Mesh%20Component%20Object%20Reference) (Velocity%20Transform :type Name :default "root") (Velocity%20Interp%20Speed :type Float%20%28double-precision%29 :default 1.0) (Collision%20Transform :type Name) (Collision%20Transform%20Axis :type EAxis%20Enum :default "X") (Invert%20Collision%20Transform%20Axis :type Boolean) (Delta%20Time%20X :type Float%20%28double-precision%29) (Previous%20Velocity%20Transform%20Location :type Vector%20%28by%20ref%29) (Previous%20Velocity%20Transform%20Velocity :type Vector%20%28by%20ref%29) (Region :type Integer :default 1))
  ;[1 1 1] Nativized Function
  (Animation|Material|SetWPORegionalFunction Material Overlay%20Material Skeletal%20Mesh Velocity%20Transform Velocity%20Interp%20Speed Collision%20Transform Collision%20Transform%20Axis Invert%20Collision%20Transform%20Axis Delta%20Time%20X Previous%20Velocity%20Transform%20Location Previous%20Velocity%20Transform%20Velocity Region)
  (return))

(fn UnclampedLerp(Local)
  (params (A :type Float%20%28double-precision%29) (B :type Float%20%28double-precision%29) (Alpha :type Float%20%28double-precision%29))
  ;[1 1 1] This function exists in the Athena Function Library and was being used from there.
  (return (+ A (* (- B A) Alpha))))

(fn VehicleDynamicsScalerFunction
  (params (CharacterPartAnimInstance :type Custom%20Character%20Part%20Anim%20Instance%20Object%20Reference) (Debug :type Boolean))
  (bind _characterpartaniminstance (Variables|GetCharacterPartAnimInstance))
  ;[1 1 1] THIS IS A TEST AREA FOR MAKING THE VEHICLE DYN CONTROL FUNCTION.  IT IS WIP.  DO NOT DELETE OR USE.
  (if (Variables|/Script/FortniteGame.CustomCharacterPartAnimInstance|GetbIsInVehicle _characterpartaniminstance)
    (if (or (Variables|/Script/FortniteGame.CustomCharacterPartAnimInstance|GetbIsGolfCartVehicleDriver _characterpartaniminstance) (Variables|/Script/FortniteGame.CustomCharacterPartAnimInstance|GetbIsGolfCartVehicleFrontPassenger _characterpartaniminstance))
      (Development|PrintString 0 "hair scaled, skirt not scaled" (Variables|GetDebug) (Variables|GetDebug))
      (return 1.0 0.0 0.0 1.0)
      (elif (Math|Boolean|ORBoolean (Variables|/Script/FortniteGame.CustomCharacterPartAnimInstance|GetbIsFerretVehicleDriver _characterpartaniminstance) (Variables|/Script/FortniteGame.CustomCharacterPartAnimInstance|GetbIsOctopusVehicleDriver _characterpartaniminstance) (Variables|/Script/FortniteGame.CustomCharacterPartAnimInstance|GetbIsOstrichVehicleDriver _characterpartaniminstance) (Variables|/Script/FortniteGame.CustomCharacterPartAnimInstance|GetbIsOstrichVehicleGunner _characterpartaniminstance) (Variables|/Script/FortniteGame.CustomCharacterPartAnimInstance|GetbIsMeatballDriver _characterpartaniminstance))
        (Development|PrintString 0 "hair not scaled, skirt scaled" (Variables|GetDebug) (Variables|GetDebug))
        (return 0.0 1.0 1.0 0.0)
        (elif (not (Math|Boolean|ORBoolean (Variables|/Script/FortniteGame.CustomCharacterPartAnimInstance|GetbVehicleSeatForcesCrouch _characterpartaniminstance) (Variables|/Script/FortniteGame.CustomCharacterPartAnimInstance|GetbIsGolfCartVehicleBackPassenger _characterpartaniminstance) (Variables|/Script/FortniteGame.CustomCharacterPartAnimInstance|GetbIsJackalVehicleDriver _characterpartaniminstance) (Variables|/Script/FortniteGame.CustomCharacterPartAnimInstance|GetbIsInShoppingCart _characterpartaniminstance) (Variables|/Script/FortniteGame.CustomCharacterPartAnimInstance|GetbIsInCannon _characterpartaniminstance) (Variables|/Script/FortniteGame.CustomCharacterPartAnimInstance|GetbIsMeatballPassenger _characterpartaniminstance) (Variables|/Script/FortniteGame.CustomCharacterPartAnimInstance|GetbIsMotorcycleDriver _characterpartaniminstance) (Variables|/Script/FortniteGame.CustomCharacterPartAnimInstance|GetbIsMotorcyclePassenger _characterpartaniminstance) (Variables|/Script/FortniteGame.CustomCharacterPartAnimInstance|GetbIsAntelopeVehicleDriver _characterpartaniminstance) (Variables|/Script/FortniteGame.CustomCharacterPartAnimInstance|GetbIsAntelopeVehiclePassenger _characterpartaniminstance)))
          (Development|PrintString 0 "hair scaled, skirt scaled" (Variables|GetDebug) (Variables|GetDebug))
          (return 1.0 0.0 1.0 0.0))))
    (else
      _))
  (Development|PrintString 0 "hair not scaled, skirt not scaled" (Variables|GetDebug) (Variables|GetDebug))
  (return 0.0 1.0 0.0 1.0))

(fn WindGustControlFunctionDEPRECATED
  (params (CharacterPartAnimInstance :type Custom%20Character%20Part%20Anim%20Instance%20Object%20Reference) (DeltaTime :type Float%20%28double-precision%29) (GameTimeInSeconds :type Float%20%28double-precision%29) (JointName :type Name :default "root") (WindGustSettings_MinTimeBetweenGusts_25_6CD0864B4078D8378C061197E9B1B094 :type Float%20%28single-precision%29) (WindGustSettings_MaxTimeBetweenGusts_28_0D981B4E4039D5A0C6A4C78D6F09D510 :type Float%20%28single-precision%29) (WindGustSettings_MinGustStrength_26_A5C4B017498D3154D76CECAB9AE7BDF2 :type Float%20%28single-precision%29) (WindGustSettings_MaxGustStrength_29_DACF320B4C03C6BD248A8E9B7B867FC7 :type Float%20%28single-precision%29) (WindGustSettings_MinGustDuration_27_E5761CE84E8E52EEBD663B9DC3A0906F :type Float%20%28single-precision%29) (WindGustSettings_MaxGustDuration_30_98974B1C4B5ABE4EA1CFF7A791319392 :type Float%20%28single-precision%29) (WindGustSettings_WindTurbulenceScalarA_48_A343D5A54D704F5B249C8E907DA80101 :type Float%20%28single-precision%29) (WindGustSettings_WindTurbulenceScalarB_47_E54C28B348B4C35A1D0817AFA9FF0423 :type Float%20%28single-precision%29) (WindGustSettings_MinUpInterpSpeed_36_8E04A8E94B2040C7DBF93194733AC36A :type Float%20%28single-precision%29) (WindGustSettings_MaxUpInterpSpeed_37_5108D5FA40140A1AAD38459100B14261 :type Float%20%28single-precision%29) (WindGustSettings_MinDownInterpSpeed_40_C3AD555846C2AEB9625C4E93B6758BB8 :type Float%20%28single-precision%29) (WindGustSettings_MaxDownInterpSpeed_41_394B2B66448D9D21249A288C7DEC98F1 :type Float%20%28single-precision%29) (WindGustSettings_WindFrequency_51_9F209DFA4F7AB96CE220578516F0B4CF :type Vector) (WindGustSettings_WindAmplitude_54_E174BDEC415CE55F6A1475B799695475 :type Vector) (WindGustSettings_WindOffset_55_91DD775940222EE47FDA3BA8695E8FE9 :type Vector) (WindGustLocalSettings :type Wind%20Gust%20Local%20Setting%20Struct%20Structure))
  (bind _returnvalue (CustomCharacter|GetOwnerSkeletalMeshComponent CharacterPartAnimInstance))
  (Utilities|IsValid _returnvalue
    (:Is%20Valid
      (bind _vectorout (CallFunction|RandomVectorFunction 0 _returnvalue JointName GameTimeInSeconds 0.0 1.0 WindGustSettings_WindFrequency_51_9F209DFA4F7AB96CE220578516F0B4CF WindGustSettings_WindAmplitude_54_E174BDEC415CE55F6A1475B799695475 WindGustSettings_WindOffset_55_91DD775940222EE47FDA3BA8695E8FE9))
      (bind (_guststrength___windvelocity _clothgustlocalsettingsout) (Default|WindGustGeneratorFunctionDEPRECATED DeltaTime (Utilities|Struct|MakeWindGustSettingStruct WindGustSettings_MinTimeBetweenGusts_25_6CD0864B4078D8378C061197E9B1B094 WindGustSettings_MaxTimeBetweenGusts_28_0D981B4E4039D5A0C6A4C78D6F09D510 WindGustSettings_MinGustStrength_26_A5C4B017498D3154D76CECAB9AE7BDF2 WindGustSettings_MaxGustStrength_29_DACF320B4C03C6BD248A8E9B7B867FC7 WindGustSettings_MinGustDuration_27_E5761CE84E8E52EEBD663B9DC3A0906F WindGustSettings_MaxGustDuration_30_98974B1C4B5ABE4EA1CFF7A791319392 WindGustSettings_WindTurbulenceScalarA_48_A343D5A54D704F5B249C8E907DA80101 WindGustSettings_WindTurbulenceScalarB_47_E54C28B348B4C35A1D0817AFA9FF0423 WindGustSettings_MinUpInterpSpeed_36_8E04A8E94B2040C7DBF93194733AC36A WindGustSettings_MaxUpInterpSpeed_37_5108D5FA40140A1AAD38459100B14261 WindGustSettings_MinDownInterpSpeed_40_C3AD555846C2AEB9625C4E93B6758BB8 WindGustSettings_MaxDownInterpSpeed_41_394B2B66448D9D21249A288C7DEC98F1) WindGustLocalSettings 0))
      (bind _output_get (LocalVariables|SetGustStrength%20-%20WindVelocityLocal (* _vectorout _guststrength___windvelocity)))
      (return _output_get _clothgustlocalsettingsout))))

(fn WindGustGeneratorFunctionDEPRECATED
  (params (DeltaTime :type Float%20%28double-precision%29) (GustSettings_MinTimeBetweenGusts_25_6CD0864B4078D8378C061197E9B1B094 :type Float%20%28single-precision%29) (GustSettings_MaxTimeBetweenGusts_28_0D981B4E4039D5A0C6A4C78D6F09D510 :type Float%20%28single-precision%29) (GustSettings_MinGustStrength_26_A5C4B017498D3154D76CECAB9AE7BDF2 :type Float%20%28single-precision%29) (GustSettings_MaxGustStrength_29_DACF320B4C03C6BD248A8E9B7B867FC7 :type Float%20%28single-precision%29) (GustSettings_MinGustDuration_27_E5761CE84E8E52EEBD663B9DC3A0906F :type Float%20%28single-precision%29) (GustSettings_MaxGustDuration_30_98974B1C4B5ABE4EA1CFF7A791319392 :type Float%20%28single-precision%29) (GustSettings_WindTurbulenceScalarA_48_A343D5A54D704F5B249C8E907DA80101 :type Float%20%28single-precision%29) (GustSettings_WindTurbulenceScalarB_47_E54C28B348B4C35A1D0817AFA9FF0423 :type Float%20%28single-precision%29) (GustSettings_MinUpInterpSpeed_36_8E04A8E94B2040C7DBF93194733AC36A :type Float%20%28single-precision%29) (GustSettings_MaxUpInterpSpeed_37_5108D5FA40140A1AAD38459100B14261 :type Float%20%28single-precision%29) (GustSettings_MinDownInterpSpeed_40_C3AD555846C2AEB9625C4E93B6758BB8 :type Float%20%28single-precision%29) (GustSettings_MaxDownInterpSpeed_41_394B2B66448D9D21249A288C7DEC98F1 :type Float%20%28single-precision%29) (GustSettings_WindFrequency_51_9F209DFA4F7AB96CE220578516F0B4CF :type Vector) (GustSettings_WindAmplitude_54_E174BDEC415CE55F6A1475B799695475 :type Vector) (GustSettings_WindOffset_55_91DD775940222EE47FDA3BA8695E8FE9 :type Vector) (ClothGustLocalSettings_ElapsedTime_48_9A24DA0842E21D5736699499F1DE1171 :type Float%20%28single-precision%29) (ClothGustLocalSettings_TimeBetweenGusts_54_E1D59EA94088FB83508636942434066D :type Float%20%28single-precision%29) (ClothGustLocalSettings_GustStrength_55_7788D1D04920E9BC2F96FCBDDCA7C33A :type Float%20%28single-precision%29) (ClothGustLocalSettings_GustDuration_56_E42A0C8341E10ACAC2016FA7F726B406 :type Float%20%28single-precision%29) (ClothGustLocalSettings_FinalGustStrengthLocal_57_BD76FBFF47687A7C38D40EAD653D5618 :type Float%20%28single-precision%29))
  (bind _returnvalue (Utilities|Time|GetGameTimeinSeconds 0))
  (bind _returnvalue_2 (Math|Random|PerlinNoise1D _returnvalue))
  (bind _returnvalue_3 (Math|Random|RandomFloatinRange GustSettings_MinDownInterpSpeed_40_C3AD555846C2AEB9625C4E93B6758BB8 GustSettings_MaxDownInterpSpeed_41_394B2B66448D9D21249A288C7DEC98F1))
  (bind _deltatime (Variables|GetDeltaTime))
  (bind _returnvalue_4 (Math|Random|PerlinNoise1D _returnvalue))
  (bind _returnvalue_5 (+ _returnvalue_4 1.0))
  (bind _finalguststrengthlocal (LocalVariables|GetFinalGustStrengthLocal))
  (bind _returnvalue_6 (Math|Interpolation|FInterpTo _finalguststrengthlocal _returnvalue_5 _deltatime _returnvalue_3))
  (bind _returnvalue_7 (Math|Random|RandomFloatinRange GustSettings_MinUpInterpSpeed_36_8E04A8E94B2040C7DBF93194733AC36A GustSettings_MaxUpInterpSpeed_37_5108D5FA40140A1AAD38459100B14261))
  (bind _deltatime_2 (Variables|GetDeltaTime))
  (bind _finalguststrengthlocal_2 (LocalVariables|GetFinalGustStrengthLocal))
  (bind _guststrengthlocal (LocalVariables|GetGustStrengthLocal))
  (bind _returnvalue_8 (+ _returnvalue_2 1.0))
  (bind _returnvalue_9 (* _returnvalue_8 _guststrengthlocal))
  (bind _returnvalue_10 (Math|Interpolation|FInterpTo _returnvalue_9 _finalguststrengthlocal_2 _deltatime_2 _returnvalue_7))
  (LocalVariables|SetElapsedTimeLocal ClothGustLocalSettings_ElapsedTime_48_9A24DA0842E21D5736699499F1DE1171)
  (LocalVariables|SetTimeBetweenGustsLocal ClothGustLocalSettings_TimeBetweenGusts_54_E1D59EA94088FB83508636942434066D)
  (LocalVariables|SetGustStrengthLocal ClothGustLocalSettings_GustStrength_55_7788D1D04920E9BC2F96FCBDDCA7C33A)
  (LocalVariables|SetGustDurationLocal ClothGustLocalSettings_GustDuration_56_E42A0C8341E10ACAC2016FA7F726B406)
  (LocalVariables|SetFinalGustStrengthLocal ClothGustLocalSettings_FinalGustStrengthLocal_57_BD76FBFF47687A7C38D40EAD653D5618)
  (Utilities|FlowControl|Sequence
    (:then_0
      ;[1 1 1] Increment 'Elapsed Time'.
      (LocalVariables|SetElapsedTimeLocal (+ (Variables|GetDeltaTime) (LocalVariables|GetElapsedTimeLocal)))
      ;[1 1 1] If elapsed time is > time between gusts, then create a new gust
      (if (>= (LocalVariables|GetElapsedTimeLocal) (LocalVariables|GetTimeBetweenGustsLocal))
        ;[1 1 1] Pick a new random time between gusts
        (LocalVariables|SetTimeBetweenGustsLocal (CallFunction|UnclampedLerp(Local) GustSettings_MinTimeBetweenGusts_25_6CD0864B4078D8378C061197E9B1B094 GustSettings_MaxTimeBetweenGusts_28_0D981B4E4039D5A0C6A4C78D6F09D510 (Math|Random|RandomFloat)))
        ;[1 1 1] Pick a new random gust strength
        (LocalVariables|SetGustStrengthLocal (CallFunction|UnclampedLerp(Local) GustSettings_MinGustStrength_26_A5C4B017498D3154D76CECAB9AE7BDF2 GustSettings_MaxGustStrength_29_DACF320B4C03C6BD248A8E9B7B867FC7 (Math|Random|RandomFloat)))
        ;[1 1 1] Pick a new random gust duration
        (LocalVariables|SetGustDurationLocal (CallFunction|UnclampedLerp(Local) GustSettings_MinGustDuration_27_E5761CE84E8E52EEBD663B9DC3A0906F GustSettings_MaxGustDuration_30_98974B1C4B5ABE4EA1CFF7A791319392 (Math|Random|RandomFloat)))
        (LocalVariables|SetElapsedTimeLocal 0.0) ; Reset the Elapsed time to 0
      ))
    (:then_1
      ;[1 1 1] If elapsed time is greater than Gust duration. Apply the gust.   If not blend back down to no gust
      (if (>= (LocalVariables|GetElapsedTimeLocal) (LocalVariables|GetGustDurationLocal))
        (LocalVariables|SetFinalGustStrengthLocal (Math|Interpolation|FInterpTo _returnvalue_10 _returnvalue_6 (Variables|GetDeltaTime) 5.0)) ; Blend back down to no gust
        (else
          (LocalVariables|SetFinalGustStrengthLocal (Math|Interpolation|FInterpTo _returnvalue_6 _returnvalue_10 (Variables|GetDeltaTime) 5.0)) ; Apply the gust
        )))
    (:then_2
      (return (Math|Float|Clamp(Float) (* (+ (* (LocalVariables|GetFinalGustStrengthLocal) GustSettings_WindTurbulenceScalarA_48_A343D5A54D704F5B249C8E907DA80101) 1.0) GustSettings_WindTurbulenceScalarB_47_E54C28B348B4C35A1D0817AFA9FF0423) 0.01 2.0) (LocalVariables|GetElapsedTimeLocal) (LocalVariables|GetTimeBetweenGustsLocal) (LocalVariables|GetGustStrengthLocal) (LocalVariables|GetGustDurationLocal) (LocalVariables|GetFinalGustStrengthLocal)))))
