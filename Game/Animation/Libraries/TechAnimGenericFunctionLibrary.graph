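# Control Rig function library of generic tech-animation helpers; its own rig graph is empty.
# Get Delta Ref returns local change from initial pose; Get Mirror Transform negates translation only, leaving rotation unmirrored.
# Get Normalized Distance Between outputs inverted remapped distance, optionally drawing colored debug lines.
# Spherical Collision clamps offset source around offset target, applying it by weight when displacement is within max radius; Modify Transforms with Mirror is an empty stub.
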ControlRigBlueprint /Game/Animation/Libraries/TechAnimGenericFunctionLibrary
graphs [
 ConstructionGraph {nodes [RigUnit_PrepareForExecution = &RigUnit_PrepareForExecution()]}
 ContainedGraph {nodes [
   n1 = RigVMFunctionEntryNode()
   n2 = /Script/RigVM.RigVMFunction_MathTransformToEulerTransform(Value=@n1.transform) {is=struct RigVMUnitNode, template="Cast::Execute(in Value,out Result)"}
   RigVMFunction_MathQuaternionFromRotatorV2 = /Script/RigVM.RigVMFunction_MathQuaternionFromRotatorV2(Value=@n2.Result.Rotation) {is=struct RigVMUnitNode, template="Cast::Execute(in Value,out Result)"}
   > RigVMFunctionReturnNode(
    mirrorTransform=(Rotation=(X=0,Y=0,Z=0,W=1),Translation=(X=0,Y=0,Z=0),Scale3D=(X=0,Y=0,Z=0))
    mirrorTransform.Translation=/Script/RigVM.RigVMFunction_MathVectorScale(Factor=-1, Value=@n1.transform.Translation) {is=struct RigVMUnitNode, template="Scale::Execute(in Value,in Factor,out Result)"}.Result
    mirrorTransform.Rotation=@RigVMFunction_MathQuaternionFromRotatorV2.Result
    mirrorTransform.Scale3D=@n2.Result.Scale
   )
  ]}
 New Function_ContainedGraph {nodes [
   n1 = RigVMFunctionEntryNode()
   n2 = &RigUnit_GetTransform(Space=LocalSpace, Item=@n1.Item)
   n3 = &RigUnit_GetTransform(Space=LocalSpace, bInitial=true, Item=@n1.Item)
   > RigVMFunctionReturnNode(Result=&RigVMFunction_MathTransformToEulerTransform(Value=&RigVMFunction_MathTransformMakeRelative(Global=@n2.Transform, Parent=@n3.Transform).Local) {template="Cast::Execute(in Value,out Result)"}.Result)
  ]}
 New Function_ContainedGraph {nodes [
   n1 = RigVMFunctionEntryNode()
   > RigVMFunction_Sequence = RigVMAggregateNode()
   RigVMFunction_ControlFlowBranch = &RigVMFunction_ControlFlowBranch(Condition=@n1.Debug, ExecuteContext=@RigVMFunction_Sequence.B)
   n2 = &RigUnit_GetTransform(bInitial=False, Item=@n1.Source)
   n3 = &RigUnit_GetTransform(bInitial=False, Item=@n1.Target)
   RigVMFunction_MathVectorDistance = &RigVMFunction_MathVectorDistance(A=@n2.Transform.Translation, B=@n3.Transform.Translation)
   Less = /Script/RigVM.RigVMFunction_MathDoubleLess(B=@n1.Max Distance, A=@RigVMFunction_MathVectorDistance.Result) {is=struct RigVMUnitNode, template="Less::Execute(in A,in B,out Result)"}
   n4 = &RigVMFunction_ControlFlowBranch(Condition=@Less.Result, ExecuteContext=@RigVMFunction_ControlFlowBranch.True)
   RigVMFunction_DebugLineNoSpace = /Script/RigVM.RigVMFunction_DebugLineNoSpace(
    Color=(R=1,G=0,B=0,A=1)
    DebugDrawSettings=(DepthPriority=SDPG_Foreground,Lifetime=-1)
    Thickness=2
    WorldOffset=(Rotation=(X=0,Y=0,Z=0,W=1),Translation=(X=0,Y=0,Z=0),Scale3D=(X=1,Y=1,Z=1))
    bEnabled=True
    ExecutePin=@n4.False
    B=@n3.Transform.Translation
    A=@n2.Transform.Translation
   ) {is=struct RigVMUnitNode}
   Remap = &RigVMFunction_MathDoubleRemap(
    TargetMaximum=1
    TargetMinimum=0
    Value=@RigVMFunction_MathVectorDistance.Result
    SourceMaximum=@n1.Max Distance
    SourceMinimum=@n1.Min Distance
   )
   Subtract = &RigVMFunction_MathDoubleSub(A=1, B=@Remap.Result)
   &RigVMFunction_DebugLineNoSpace(
    Color=(R=0,G=1,B=0,A=1)
    DebugDrawSettings=(DepthPriority=SDPG_Foreground,Lifetime=-1)
    Thickness=2
    WorldOffset=(Rotation=(X=0,Y=0,Z=0,W=1),Translation=(X=0,Y=0,Z=0),Scale3D=(X=1,Y=1,Z=1))
    bEnabled=True
    ExecutePin=@n4.True
    Color.R=@Subtract.Result
    B=@n3.Transform.Translation
    A=@n2.Transform.Translation
    Color.G=@Remap.Result
   )
   RigVMFunctionReturnNode(ExecuteContext=@RigVMFunction_Sequence.C, Result=RigVMVariableNode() {variable=Output}.Value)
   RigVMVariableNode(Value=&RigVMFunction_MathDoubleSub(A=1, B=@Remap.Result).Result, ExecuteContext=@RigVMFunction_Sequence.A) {variable=Output}
  ]}
 New Function_ContainedGraph {nodes [
   n1 = RigVMFunctionEntryNode()
   n2 = &RigUnit_GetTransform(bInitial=False, Item=@n1.Source)
   ClampSpatially = /Script/RigVM.RigVMFunction_MathTransformClampSpatially(
    Axis=None
    DebugColor=(R=1,G=0,B=0,A=1)
    DebugThickness=1
    Type=Sphere
    Space=&RigVMFunction_MathTransformMul(
     B=&RigUnit_GetTransform(bInitial=False, Item=@n1.Target).Transform
     A=/Script/RigVM.RigVMFunction_MathTransformMake(Rotation=(X=0,Y=0,Z=0,W=1), Scale=(X=1,Y=1,Z=1), Translation=@n1.Target Offset) {is=struct RigVMUnitNode}.Result
    ) {template="Multiply::Execute(in A,in B,out Result)"}.Result
    Value=&RigVMFunction_MathTransformMul(
     B=@n2.Transform
     A=&RigVMFunction_MathTransformMake(Rotation=(X=0,Y=0,Z=0,W=1), Scale=(X=1,Y=1,Z=1), Translation=@n1.Source Offset).Result
    ) {template="Multiply::Execute(in A,in B,out Result)"}.Result
    Maximum=@n1.Max Radius
    Minimum=@n1.Min Radius
    bDrawDebug=@n1.Debug
   ) {
    is=struct RigVMUnitNode
    template="ClampSpatially::Execute(in Value,in Axis,in Type,in Minimum,in Maximum,in Space,in bDrawDebug,in DebugColor,in DebugThickness,out Result)"
   }
   LessEqual = /Script/RigVM.RigVMFunction_MathDoubleLessEqual(B=@n1.Max Radius, A=&RigVMFunction_MathVectorDistance(B=@ClampSpatially.Result.Translation, A=@n2.Transform.Translation).Result) {is=struct RigVMUnitNode, template="LessEqual::Execute(in A,in B,out Result)"}
   n3 = &RigVMFunction_ControlFlowBranch(ExecuteContext=@n1.ExecuteContext, Condition=@LessEqual.Result)
   RigVMFunctionReturnNode(IsColliding=@LessEqual.Result, ExecuteContext=@n3.Completed)
   &RigUnit_ModifyTransforms(
    ItemToModify=((Item=(Type=Bone,Name="None"),Transform=(Rotation=(X=0,Y=0,Z=0,W=1),Translation=(X=0,Y=0,Z=0),Scale3D=(X=1,Y=1,Z=1))))
    Mode=OverrideGlobal
    WeightMinimum=0
    ItemToModify.0.Item=@n1.Source
    ExecutePin=@n3.True
    Weight=@n1.Weight
    ItemToModify.0.Transform=@ClampSpatially.Result
   )
  ]}
 New Function_ContainedGraph {nodes [
   RigVMFunctionEntryNode()
   > RigVMFunctionReturnNode()
  ]}
 RigVMFunctionLibrary {nodes [
   Get Delta Ref = RigVMCollapseNode(ExecuteContext=(), Item=(Type=Bone,Name="None"))
   Get Mirror Transform = RigVMCollapseNode(ExecuteContext=(), transform=(Rotation=(X=0,Y=0,Z=0,W=1),Translation=(X=0,Y=0,Z=0),Scale3D=(X=1,Y=1,Z=1)))
   Get Normalized Distance Between = RigVMCollapseNode(ExecuteContext=(), Min Distance=0, Source=(Type=None,Name="None"), Target=(Type=None,Name="None"))
   Modify Transforms with Mirror = RigVMCollapseNode(ExecuteContext=())
   Spherical Collision = RigVMCollapseNode(
    Debug=true
    ExecuteContext=()
    Max Radius=10
    Min Radius=9
    Source=(Type=None,Name="None")
    Source Offset=(X=0,Y=0,Z=0)
    Target=(Type=None,Name="None")
    Target Offset=(X=0,Y=0,Z=0)
    Weight=1
   )
  ]}
 RigVMFunction_Sequence_1_ContainedGraph {nodes [
   n1 = RigVMFunctionEntryNode()
   RigVMFunction_Sequence = &RigVMFunction_Sequence(ExecuteContext=@n1.ExecuteContext)
   n2 = &RigVMFunction_Sequence(ExecuteContext=@RigVMFunction_Sequence.B)
   RigVMFunctionReturnNode(A=@RigVMFunction_Sequence.A, B=@n2.A, C=@n2.B)
  ]}
 RigVMModel {nodes [RigUnit_BeginExecution = &RigUnit_BeginExecution()]}
]
